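; AreaVisualizerInterface is a Blueprint interface for area visualizers used by the Creative Capture Area device.
; It declares one function, SetupDynamicData, which passes an inner radius, outer radius and height to the implementer.
; The interface holds no logic, so implementers decide how these dimensions shape the displayed area.
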(fn SetupDynamicData
  (params (InnerRadius :type Float%20%28double-precision%29) (Outer%20Radius :type Float%20%28double-precision%29) (Height :type Float%20%28double-precision%29)))
