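; Blueprint interface for the Capture Area creative device's capture progress HUD.
; It declares a single function, GetDataFromBlueprint, which by its name lets the HUD pull data from an implementing Blueprint.
; The file holds only the function stub with a default return, so any actual data comes from implementers.
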(fn GetDataFromBlueprint
  (return 0))
