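; Generated dynamic mesh actor that draws a hollow cylinder as the area visualizer child actor for the Capture Area device.
; SetupDynamicData stores inner radius, outer radius and height, then marks the mesh for rebuild.
; On rebuild, Geometry Script appends a capped outer cylinder and boolean-subtracts an inner one built on a temporary compute mesh, which is then released.
; BeginPlay, Tick, overlap and construction script are empty.
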(event EventBeginPlay)

(event Collision|EventActorBeginOverlap
  (params (OtherActor :type Actor%20Object%20Reference)))

(event EventTick
  (params (DeltaSeconds :type Float%20%28single-precision%29)))

(event Events|EventOnRebuildGeneratedMesh
  (params (TargetMesh :type Dynamic%20Mesh%20Object%20Reference))
  (bind _returnvalue (GeometryScript|Primitives|AppendCylinder :TargetMesh TargetMesh :Radius (Variables|Default|GetOuter%20Radius) :Height (Variables|Default|GetHeight) :RadialSteps (Variables|Default|GetRadial%20Steps) :HeightSteps 0 :bCapped true :Origin "Base"))
  (bind _returnvalue_2 (DynamicMeshActor|AllocateComputeMesh))
  (bind _returnvalue_3 (GeometryScript|Primitives|AppendCylinder :TargetMesh _returnvalue_2 :Radius (Variables|Default|GetInner%20Radius) :Height (Variables|Default|GetHeight) :RadialSteps (Variables|Default|GetRadial%20Steps) :HeightSteps 0 :bCapped true :Origin "Base"))
  (bind _returnvalue_4 (GeometryScript|Booleans|ApplyMeshBoolean :TargetMesh _returnvalue :ToolMesh _returnvalue_3 :Operation "Subtract"))
  (DynamicMeshActor|ReleaseAllComputeMeshes))

(event EventSetupDynamicData
  (params (InnerRadius :type Float%20%28double-precision%29) (Outer%20Radius :type Float%20%28double-precision%29) (Height :type Float%20%28double-precision%29))
  (Variables|Default|SetInner%20Radius InnerRadius)
  (Variables|Default|SetOuter%20Radius Outer%20Radius)
  (Variables|Default|SetHeight Height)
  (DynamicMeshActor|MarkforMeshRebuild true))

(fn ConstructionScript)
