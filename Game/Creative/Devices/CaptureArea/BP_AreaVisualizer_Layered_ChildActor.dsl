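; Child-actor visualizer for the Capture Area device that procedurally builds a layered, hollow cylinder mesh with Geometry Script.
; SetupDynamicData receives inner radius, outer radius and height, then marks the mesh for rebuild.
; On rebuild, it derives slice height from slice count and spacing, unions stacked cylinders, then boolean-subtracts an inner cylinder to form spaced rings.
; Temporary compute meshes hold the boolean tools and are released afterwards.
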(event EventBeginPlay)

(event Collision|EventActorBeginOverlap
  (params (OtherActor :type Actor%20Object%20Reference)))

(event EventTick
  (params (DeltaSeconds :type Float%20%28single-precision%29)))

(event Events|EventOnRebuildGeneratedMesh
  (params (TargetMesh :type Dynamic%20Mesh%20Object%20Reference))
  (Variables|Default|SetSlice%20Height (- (/ (Variables|Default|GetTotal%20Height) (Variables|Default|GetTotal%20Slices)) (Variables|Default|GetSlice%20Spacing)))
  (bind _returnvalue (GeometryScript|Primitives|AppendCylinder :TargetMesh TargetMesh :Radius (Variables|Default|GetOuter%20Radius) :Height (Variables|Default|GetSlice%20Height) :RadialSteps (Variables|Default|GetRadial%20Steps) :HeightSteps 0 :bCapped true :Origin "Base"))
  (for _index (range 1 (Variables|Default|GetTotal%20Slices))
    ; Loop
    (bind _returnvalue_2 (DynamicMeshActor|AllocateComputeMesh))
    (bind _returnvalue_3 (GeometryScript|Primitives|AppendCylinder :TargetMesh _returnvalue_2 :Transform (Math|Transform|MakeTransform (Math|Vector|MakeVector 0.0 0.0 (* _index (+ (Variables|Default|GetSlice%20Height) (Variables|Default|GetSlice%20Spacing))))) :Radius (Variables|Default|GetOuter%20Radius) :Height (Variables|Default|GetSlice%20Height) :RadialSteps (Variables|Default|GetRadial%20Steps) :HeightSteps 0 :bCapped true :Origin "Base"))
    (GeometryScript|Booleans|ApplyMeshBoolean :TargetMesh _returnvalue :ToolMesh _returnvalue_3 :Operation "Union"))
  (bind _returnvalue_4 (DynamicMeshActor|AllocateComputeMesh))
  (bind _returnvalue_5 (GeometryScript|Primitives|AppendCylinder :TargetMesh _returnvalue_4 :Radius (Variables|Default|GetInner%20Radius) :Height (Variables|Default|GetTotal%20Height) :RadialSteps (Variables|Default|GetRadial%20Steps) :HeightSteps 0 :bCapped true :Origin "Base"))
  (GeometryScript|Booleans|ApplyMeshBoolean :TargetMesh _returnvalue :ToolMesh _returnvalue_5 :Operation "Subtract")
  (DynamicMeshActor|ReleaseAllComputeMeshes))

(event EventSetupDynamicData
  (params (InnerRadius :type Float%20%28double-precision%29) (Outer%20Radius :type Float%20%28double-precision%29) (Height :type Float%20%28double-precision%29))
  (Variables|Default|SetInner%20Radius InnerRadius)
  (Variables|Default|SetOuter%20Radius Outer%20Radius)
  (Variables|Default|SetTotal%20Height Height)
  (DynamicMeshActor|MarkforMeshRebuild true))

(fn ConstructionScript)
