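; Creative Capture Area device that lets players take team control of a zone, deliver items and earn periodic score within a minigame.
; Overlap, inventory, emote, minigame and receiver events drive a CaptureComponent state machine, gated by enabled and allowed flags.
; Capture state and ownership changes update objectives, map markers, progress widgets and zone events.
; Cosmetic-only branches handle audio loops, timeline-driven material effects, beacon particles and async-loaded item previews.
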(event EventEndPlay
  (params (EndPlayReason :type EEndPlayReason%20Enum :default "Destroyed"))
  ;[1 1 1] EndPlay / Destroyed
  (|CallParentFunction EndPlayReason)
  (Device|ResetDeviceCachedData true))

(event EventBeginPlay
  ;[1 1 1] Begin Play
  (Visibility|SetupVisibilityComponent)
  (Networking|CanExecuteCosmeticEvents
    (:True
      (Variables|BuildingActor|SetOverridePrimitivesToExcludeFoliage (Utilities|Array|MakeArray (Variables|Default|GetBaseMesh)))
      (Rendering|SetVisibility (Variables|Default|GetCaptureAreaLight) true false)
      (Rendering|Material|SetScalarParameterValue (Variables|Default|GetCaptureAreaMID) (Variables|MIDParameters|GetMIDParameterCaptureEffectOpacity) 0.0 :!class /Script/Engine.MaterialInstanceDynamic)
      (CallFunction|FXHandleEnable)
      (Device|UpdateDeviceColors)
      (ProgressWidget|InitInWorldWidget))
    (:False
      (CaptureComponent|SetCaptureState (Variables|CaptureArea|GetCaptureComponent) "Disabled" (Utilities|Enum|LiteralenumEFortTeam "Spectator")))))

(event EventDestroyed
  ;[1 1 1] EndPlay / Destroyed
  (Networking|CanExecuteCosmeticEvents
    (:False
      (Utilities|Time|ClearandInvalidateTimerbyHandle (Variables|PeriodicScoring|GetPeriodicScoringTimerHandle)))))

(event EditorPreview|EventSetupForEditorPreview
  ;[1 1 1] Begin Play
  (|CallParentFunction)
  (Class|CreativeItemPreviewComponent|EditorDisplayPreview (Variables|Default|GetEditorOnlyPreview) (Variables|Default|GetPreviewStaticMesh))
  (Networking|CanExecuteCosmeticEvents
    (:True
      (Variables|BuildingActor|SetOverridePrimitivesToExcludeFoliage (Utilities|Array|MakeArray (Variables|Default|GetBaseMesh)))
      (Rendering|SetVisibility (Variables|Default|GetCaptureAreaLight) true false)
      (Rendering|Material|SetScalarParameterValue (Variables|Default|GetCaptureAreaMID) (Variables|MIDParameters|GetMIDParameterCaptureEffectOpacity) 0.0 :!class /Script/Engine.MaterialInstanceDynamic)
      (CallFunction|FXHandleEnable)
      (Device|UpdateDeviceColors)
      (ProgressWidget|InitInWorldWidget))
    (:False
      (CaptureComponent|SetCaptureState (Variables|CaptureArea|GetCaptureComponent) "Disabled" (Utilities|Enum|LiteralenumEFortTeam "Spectator")))))

(event EventPostFinishSpawning
  ;[1 1 1] Event PostFinishedSpawning
  (CallFunction|MigrateSameChannelInstances)
  (|CallParentFunction)
  (Networking|CanExecuteCosmeticEvents
    (:False
      (if (Variables|DeviceOptions|GetbCountAsObjectiveKnob)
        (Class|FortMinigameLogicComponent|TryAttachtoMinigame (Variables|Default|GetFortMinigameObjective))))))

(event Custom|DisallowCaptureFromEvent
  (params (Instigator :type Controller%20Object%20Reference))
  ;[1 1 1] Receivers: Allow/Disallow Capture
  (if (Variables|Replicated|GetIsEnabledByPhase)
    (if (Variables|Default|GetIsAllowed)
      (Variables|Default|SetIsAllowed false))))

(event Custom|GiveControlFromEvent
  (params (Instigator :type Controller%20Object%20Reference))
  ;[1 1 1] Receivers : Capture / Reset Team / Neutralize
  (if (Variables|Replicated|GetIsEnabledByPhase)
    (bind _asfort_player_controller (Utilities|Casting|CastToFortPlayerController Instigator))
    (Capture|ForceCapture _asfort_player_controller (Team|GetActorTeam _asfort_player_controller))))

(event Custom|ResetTeamControlFromEvent
  (params (Instigator :type Controller%20Object%20Reference))
  ;[1 1 1] Receivers : Capture / Reset Team / Neutralize
  (if (Variables|Replicated|GetIsEnabledByPhase)
    (CallFunction|SetStartOwnerTeam)
    (CaptureComponent|SetCaptureState (Variables|CaptureArea|GetCaptureComponent) "Disabled" (Variables|Team|GetTeamIndex)) ; We are changing to Disabled state before\nof Set Capture State again, because Capture State\nmay already be "Captured" and it will not be triggered.
    (Capture|SetCaptureStatebyCurrentTeam true)))

(event Custom|AllowCaptureFromEvent
  (params (Instigator :type Controller%20Object%20Reference))
  ;[1 1 1] Receivers: Allow/Disallow Capture
  (if (Variables|Replicated|GetIsEnabledByPhase)
    (if (not (Variables|Default|GetIsAllowed))
      (Variables|Default|SetIsAllowed true))))

(event EventOnTeamsInfoChanged
  ;[1 1 1] Teams Info Inside Area Changed Event
  (if (Getters|IsMinigameinProgress)
    (Capture|TrytoUpdateCaptureState)
    (ItemDelivery|TrytoDeliverItemsfromInsidePlayers)
    (ItemDelivery|TrytoUpdatePeriodicScoring)))

(event Custom|ToggleAllowCaptureFromEvent
  (params (Instigator :type Controller%20Object%20Reference))
  ;[1 1 1] Receivers: Allow/Disallow Capture
  (if (Variables|Replicated|GetIsEnabledByPhase)
    (Variables|Default|SetIsAllowed (not (Variables|Default|GetIsAllowed)))))

(event OnCaptureProgressUpdate(CaptureComponent)
  (params (NewProgress :type Float%20%28single-precision%29))
  ;[1 1 1] Capture Progress
  (Networking|CanExecuteCosmeticEvents
    (:True
      (Utilities|FlowControl|Sequence
        (:then_0
          (MapMarker|MapMarkerUpdateProgress))
        (:then_1
          (Utilities|ShouldInvertProgress
            (:Normal%20Progress
              (Rendering|Material|SetScalarParameterValue (Variables|Default|GetCaptureAreaMID) (Variables|MIDParameters|GetMIDParameterCaptureProgress) NewProgress :!class /Script/Engine.MaterialInstanceDynamic))
            (:Inverse%20Progress
              (Rendering|Material|SetScalarParameterValue (Variables|Default|GetCaptureAreaMID) (Variables|MIDParameters|GetMIDParameterCaptureProgress) (- 1.0 NewProgress) :!class /Script/Engine.MaterialInstanceDynamic)))
          (Audio|Parameter|SetFloatParameter (Variables|Default|GetCapturing%20Audio%20Loop%20A) "Pitch" (Math|Float|MapRangeClamped NewProgress 0.0 1.0 0.85 1.2) :!class /Script/AudioExtensions.AudioParameterControllerInterface))))
    (:False
      (Objective|UpdateObjectiveProgress))))

(event Custom|NeuralizeFromEvent
  (params (Instigator :type Controller%20Object%20Reference))
  ;[1 1 1] Receivers : Capture / Reset Team / Neutralize
  (if (Variables|Replicated|GetIsEnabledByPhase)
    (CaptureComponent|ForceCaptureProgress (Variables|CaptureArea|GetCaptureComponent) 0.0)
    (Variables|Default|SetForceNeutralization true)
    (CaptureComponent|SetCaptureState (Variables|CaptureArea|GetCaptureComponent) "Neutral" (Utilities|Enum|LiteralenumEFortTeam "Spectator"))))

(event EventOnOverlappedPlayerInventoryValidated
  (params (PlayerController :type Fort%20Player%20Controller%20Object%20Reference))
  ;[1 1 1] OnInventoryChanged Events
  (Utilities|FlowControl|Sequence
    (:then_0
      (Utilities|IsValid PlayerController
        (:Is%20Valid
          (ItemDelivery|TrytoDeliverItemsfromPlayer PlayerController))))
    (:then_1
      (if (Capture|IsValidControllerforCapture PlayerController)
        (Capture|TrytoUpdateCaptureState)))
    (:then_2
      (if (PeriodicScoring|IsValidControllerforScoring PlayerController)
        (ItemDelivery|TrytoUpdatePeriodicScoring)))))

(event Custom|DisableFromEvent
  (params (Instigator :type Controller%20Object%20Reference))
  ;[1 1 1] Receivers : Enabled/Disabled
  (if (Variables|Replicated|GetIsEnabledByPhase)
    (Variables|Replicated|SetIsEnabledByPhase false)))

(event OnCaptureStateChanged(CaptureComponent)
  (params (NewCaptureState :type ECaptureState%20Enum :default "Neutral") (PreviousCaptureState :type ECaptureState%20Enum :default "Neutral"))
  ;[1 1 1] OnCaptureStateChanged
  (Objective|UpdateObjectiveProgress)
  (CallFunction|StopCaptureAudio)
  (ProgressWidget|SetupUIforCaptureState NewCaptureState PreviousCaptureState (CaptureComponent|GetCaptureStateOwner (Variables|CaptureArea|GetCaptureComponent)))
  (Capture|CleanUpPreviousCaptureState PreviousCaptureState)
  (Capture|UpdateCaptureState NewCaptureState)
  (Networking|CanExecuteCosmeticEvents
    (:True
      (MapMarker|MapMarkerUpdateProgress))))

(event Custom|EnableFromEvent
  (params (Instigator :type Controller%20Object%20Reference))
  ;[1 1 1] Receivers : Enabled/Disabled
  (if (not (Variables|Replicated|GetIsEnabledByPhase))
    (Variables|Replicated|SetIsEnabledByPhase true)))

(event Custom|ToggleEnabledFromEvent
  (params (Instigator :type Controller%20Object%20Reference))
  ;[1 1 1] Receivers : Enabled/Disabled
  (Variables|Replicated|SetIsEnabledByPhase (not (Variables|Replicated|GetIsEnabledByPhase))))

(event EventOnOverlappedPlayerInventoryChanged
  ;[1 1 1] OnInventoryChanged Events
  (Capture|TrytoUpdateCaptureState)
  (ItemDelivery|TrytoUpdatePeriodicScoring))

(event Custom|OnPawnStartedEmote
  (params (MontageItemDef :type Fort%20Item%20Definition%20Object%20Reference) (PawnEmoting :type Fort%20Pawn%20Object%20Reference))
  ;[1 1 1] OnPlayerEmote Events
  (if (Getters|IsMinigameinProgress)
    (Capture|UpdateCaptureSpeed)
    (Capture|UpdateProgressReverseSpeed)))

(event MinigameStateChangedDelegate(MinigameLogic)
  (params (Minigame :type Fort%20Minigame%20Object%20Reference) (NewMinigameState :type EFortMinigameState%20Enum :default "PreGame"))
  ;[1 1 1] Minigame State Changed
  (EnableKnob|SetEnablebyMinigameState NewMinigameState)
  (CallFunction|SetOwnerbyMinigameState NewMinigameState)
  (Device|SetViewonEditMode false)
  (IsAllowed|SetIsAllowedonPregame NewMinigameState))

(event OnComponentHit(DropArea)
  (params (HitComponent :type Primitive%20Component%20Object%20Reference) (OtherActor :type Actor%20Object%20Reference) (OtherComp :type Primitive%20Component%20Object%20Reference) (NormalImpulse :type Vector :default "0, 0, 0") (Hit :type Hit%20Result%20Structure%20%28by%20ref%29))
  ;[1 1 1] Hit/Overlap Events
  (Default|HandleCaptureAreaOverlap OtherActor))

(event OnComponentEndOverlap(CaptureAreaStaticMesh)
  (params (OverlappedComponent :type Primitive%20Component%20Object%20Reference) (OtherActor :type Actor%20Object%20Reference) (OtherComp :type Primitive%20Component%20Object%20Reference) (OtherBodyIndex :type Integer))
  ;[1 1 1] Hit/Overlap Events
  (Networking|SwitchHasAuthority
    (:Authority
      (if (Getters|IsCaptureAreaInteractable)
        (bind _asfort_player_pawn (Utilities|Casting|CastToFortPlayerPawn OtherActor))
        (bind _asfort_player_controller (Utilities|Casting|CastToFortPlayerController (Pawn|GetController _asfort_player_pawn)))
        (CaptureComponent|UnregisterPlayerInsideArea _asfort_player_controller)
        (Bindings|BindorUnbindonPlayerEmote false _asfort_player_pawn)))))

(event Custom|OnPlayerAddedToMinigame
  (params (UniqueNetId :type Unique%20Net%20Id%20Structure) (bIsLocalPlayer :type Boolean))
  ;[1 1 1] Local player team replication
  (if bIsLocalPlayer
    (Networking|SwitchHasAuthority
      (:Remote
        (CaptureArea|BindorUnbindonPlayerTeamReplicated (Getters|GetLocalPlayerState UniqueNetId) true)))))

(event Custom|OnPawnStoppedEmote
  (params (MontageItemDef :type Fort%20Item%20Definition%20Object%20Reference) (PawnEmoting :type Fort%20Pawn%20Object%20Reference))
  ;[1 1 1] OnPlayerEmote Events
  (Utilities|FlowControl|Delay 0.0) ; Wait for next tick to ensure\nEmoteAbility is not Active anymore
  (if (Getters|IsMinigameinProgress)
    (Capture|UpdateCaptureSpeed)
    (Capture|UpdateProgressReverseSpeed)))

(event OnComponentBeginOverlap(CaptureAreaStaticMesh)
  (params (OverlappedComponent :type Primitive%20Component%20Object%20Reference) (OtherActor :type Actor%20Object%20Reference) (OtherComp :type Primitive%20Component%20Object%20Reference) (OtherBodyIndex :type Integer) (bFromSweep :type Boolean) (SweepResult :type Hit%20Result%20Structure%20%28by%20ref%29))
  ;[1 1 1] Hit/Overlap Events
  (Default|HandleCaptureAreaOverlap OtherActor))

(event Custom|OnPlayerRemovedFromMinigame
  (params (UniqueNetId :type Unique%20Net%20Id%20Structure) (bIsLocalPlayer :type Boolean))
  ;[1 1 1] Local player team replication
  (if bIsLocalPlayer
    (Networking|SwitchHasAuthority
      (:Remote
        (CaptureArea|BindorUnbindonPlayerTeamReplicated (Getters|GetLocalPlayerState UniqueNetId))))))

(event OnItemAdded(PickupItemList)
  (params (ItemAdded :type Actor%20Object%20Reference))
  ;[1 1 1] OnItemAdded
  (CallFunction|FXObjectAddedtoItemList ItemAdded)
  (CallFunction|Resave))

(event EventBP_OnPlayerTeamReplicated
  ;[1 1 1] Local player team replication
  (Networking|CanExecuteCosmeticEvents
    (:True
      (Device|UpdateDeviceColors)
      (CallFunction|SetBeaconParticles)
      (Capture|UpdateCaptureState (CaptureComponent|GetCaptureState (Variables|CaptureArea|GetCaptureComponent))))))

(event MinigameEndedDelegate(MinigameLogic)
  ;[1 1 1] MinigameEndedDelegate
  (MapMarker|MapMarkerSetVisibility false)
  (Networking|CanExecuteCosmeticEvents
    (:True
      (ItemPreviews|StartorRefreshItemPreviews (Variables|/Script/FortniteGame.FortMinigameLogicComponent|GetCurrentMinigame (Variables|Default|GetFortMinigameObjective))))
    (:False
      (CaptureComponent|SetCaptureState (Variables|CaptureArea|GetCaptureComponent) "Disabled" (Utilities|Enum|LiteralenumEFortTeam "Spectator"))
      (Device|ResetDeviceCachedData true)
      (Objective|RemoveObjective))))

(event EventOnCaptureAreaOwnerChanged
  (params (NewOwner :type Fort%20Player%20Controller%20Object%20Reference))
  ;[1 1 1] Capture Area Ownership Changed
  (Networking|CanExecuteCosmeticEvents
    (:True
      (Utilities|FlowControl|Sequence
        (:then_0
          (if (Variables|DeviceOptions|GetDisplayOwnershipVFX)
            (CallFunction|FXonOwnerChanged)
            (else
              (CallFunction|FXPlayerScored))))
        (:then_1
          (if (< (Variables|DeviceOptions|GetAccentColorIndex) 0)
            (Device|UpdateDeviceColors)
            (CallFunction|SetBeaconParticles)))))
    (:False
      (if (Getters|IsMinigameinProgress)
        (Utilities|FlowControl|Sequence
          (:then_0
            (Capture|TrytoUpdateCaptureState)
            (Objective|UpdateObjectiveProgress)
            (ItemDelivery|TrytoDeliverItemsfromInsidePlayers)
            (ItemDelivery|TrytoUpdatePeriodicScoring))
          (:then_1
            (Objective|HandleObjectiveAssignment NewOwner)
            (Utilities|IsValid NewOwner
              (:Is%20Valid
                (Variables|Default|SetCurrentOwner NewOwner)))))))))

(event OnItemListUpdated(PickupItemList)
  ;[1 1 1] OnItemListUpdated
  (Bindings|BindtoContainedItemVariants)
  (ItemPreviews|StartorRefreshItemPreviews (Variables|/Script/FortniteGame.FortMinigameLogicComponent|GetCurrentMinigame (Variables|Default|GetMinigameLogic))))

(event OnAddedtoMinigame(MinigameLogic)
  (params (Minigame :type Fort%20Minigame%20Object%20Reference))
  ;[1 1 1] OnAddedToMinigame
  (ItemPreviews|StartorRefreshItemPreviews Minigame)
  (Device|UpdateDeviceColors)
  (Device|UpdateCollisionforGame Minigame)
  (Device|SetViewonEditMode true)
  (Bindings|BindorUnbindtoPlayerAddandRemove true Minigame)
  (Bindings|BindorUnbindonPlayerChangedTeam true Minigame)
  (Utilities|IsValid Minigame
    (:Is%20Valid
      (EnableKnob|SetEnablebyMinigameState (Variables|/Script/FortniteGame.FortMinigame|GetCurrentState Minigame))
      (CallFunction|BindDropBoxCollisionChanges Minigame)
      (Minigame|AssignOnPlayerJoininProgress Minigame (EventDispatchers|CreateEvent :!event Custom|OnPlayerJoinInProgress)))
    (:Is%20Not%20Valid
      _))
  (MapMarker|MapMarkerSetup)
  (ProgressWidget|InitInWorldWidget))

(event Custom|OnPlayerJoinInProgress
  (params (Pawn :type Pawn%20Object%20Reference))
  ;[1 1 1] PlayerJoinInProgress
  (if (Pawn|IsLocallyControlled Pawn)
    (Networking|CanExecuteCosmeticEvents
      (:True
        (MapMarker|MapMarkerSetup)
        (ProgressWidget|InitInWorldWidget)))))

(event Custom|Bind%20Drop%20Box%20Collision%20Changes
  (params (Minigame :type Fort%20Minigame%20Object%20Reference))
  ;[1 1 1] Disable Collision on drop area during Edit mode, as overlaps are expensive.
  (Minigame|BindEventtoOnPlayModeChangedDelegate Minigame (delegate Custom|OnPlayModeChanged))
  (CallFunction|OnPlayModeChanged Minigame (Minigame|IsinPlayMode Minigame :!class /Script/FortniteGame.FortMinigame)))

(event Custom|OnPlayModeChanged
  (params (Minigame :type Fort%20Minigame%20Object%20Reference) (bIsInPlayMode :type Boolean))
  ;[1 1 1] Disable Collision on drop area during Edit mode, as overlaps are expensive.
  (Collision|SetCollisionEnabled (Variables|Default|GetDropArea) (select bIsInPlayMode "NoCollision" "QueryOnly")))

(event OnRemovedfromMinigame(MinigameLogic)
  (params (Minigame :type Fort%20Minigame%20Object%20Reference))
  ;[1 1 1] OnRemovedFromMinigame
  (Utilities|IsValid Minigame
    (:Is%20Valid
      (Bindings|BindorUnbindtoPlayerAddandRemove false Minigame)
      (Bindings|BindorUnbindonPlayerChangedTeam false Minigame))))

(event MinigameStartedDelegate(MinigameLogic)
  (bind _minigamelogic (Variables|Default|GetMinigameLogic))
  (bind _currentminigame (Variables|/Script/FortniteGame.FortMinigameLogicComponent|GetCurrentMinigame _minigamelogic))
  ;[1 1 1] MinigameStartedDelegate
  (ItemPreviews|StartorRefreshItemPreviews _currentminigame)
  (Device|UpdateCollisionforGame _currentminigame)
  (Capture|SetCaptureStateonStart true)
  (Objective|UpdateObjectiveTrackingState (Variables|/Script/FortniteGame.FortMinigameLogicComponent|GetCurrentMinigame (Variables|Default|GetFortMinigameObjective)))
  (EnableKnob|UpdateOverlappingActors)
  (ItemDelivery|TrytoUpdatePeriodicScoring)
  (MapMarker|MapMarkerInit)
  (MapMarker|MapMarkerSetVisibility true)
  (ProgressWidget|InitInWorldWidget))

(event OnAnyPropertyChanged(ToyOptions)
  (bind _minigamelogic (Variables|Default|GetMinigameLogic))
  (bind _currentminigame (Variables|/Script/FortniteGame.FortMinigameLogicComponent|GetCurrentMinigame _minigamelogic))
  ;[1 1 1] Properties Changed / Loaded
  (Variables|DeviceOptions|SetPlayerNameFilter (Utilities|String|Trim (Utilities|String|TrimTrailing (Variables|DeviceOptions|GetPlayerNameFilter))))
  (Utilities|FlowControl|Sequence
    (:then_0
      (CallFunction|MigrateSameChannelInstances))
    (:then_1
      (Device|UpdateCaptureArea)
      (CallFunction|SetStartOwnerTeam)
      (ItemDelivery|UpdateItemDeliveryScore)
      (Capture|UpdateControlPointSettings)
      (CallFunction|SetBeaconParticles)
      (Device|UpdateDeviceColors)
      (IsAllowed|UpdateAllowedInteractionSetting)
      (ProgressWidget|InitInWorldWidget)
      (Visibility|UpdateVisibilityComponent)
      (Utilities|IsValid _currentminigame
        (:Is%20Valid
          (EnableKnob|SetEnablebyMinigameState (Variables|/Script/FortniteGame.FortMinigame|GetCurrentState _currentminigame)))))
    (:then_2
      (MapMarker|MapMarkerSetup))
    (:then_3
      (Variables|/Script/FortniteGame.FortMinigameObjectiveComponent|SetObjectiveSortOrder (Variables|Default|GetSortOrder) (Variables|Default|GetFortMinigameObjective)))))

(event CaptureComponent|EventOnPlayerEnterZone
  (params (PlayerController :type Fort%20Player%20Controller%20Object%20Reference) (bIsFirst :type Boolean))
  ;[1 1 1] Enter and leave zone messages
  (Utilities|IsValid PlayerController
    (:Is%20Valid
      (Networking|SwitchHasAuthority
        (:Authority
          (GameplayEvent|BroadcastEvent (Variables|UserOptions|Events|GetOnPlayerEnteringZone) 0 (Utilities|Struct|MakeGameplayEventCaptureAreaControlEnterZone PlayerController))
          (UI|CaptureProgressHUDUIControl true PlayerController)
          (if bIsFirst
            (GameplayEvent|BroadcastEvent (Variables|UserOptions|Events|GetOnFirstPlayerEnteringZone) 0 (Utilities|Struct|MakeGameplayEventCaptureAreaControlFirstEnterZone PlayerController))))))))

(event Custom|Play%20Capture%20Audio%20Loop
  (params (Sound :type Sound%20Base%20Object%20Reference))
  (bind _currentloopsound (Variables|Audio|GetCurrentLoopSound))
  (bind _capturing_audio_loop_a (Variables|Default|GetCapturing%20Audio%20Loop%20A))
  (bind _capturing_audio_loop_b (Variables|Default|GetCapturing%20Audio%20Loop%20B))
  ;[1 1 1] Audio for Capturing area
  (if (Variables|DeviceOptions|GetKnob_CaptureSoundEnabled)
    (if (Utilities|NotEqual(Object) Sound (Variables|Audio|GetCurrentLoopSound))
      (Variables|Audio|SetCurrentLoopSound Sound)
      (if (Audio|Components|Audio|IsPlaying (Variables|Default|GetCapturing%20Audio%20Loop%20B))
        (Audio|Components|Audio|SetSound _capturing_audio_loop_a _currentloopsound)
        (Audio|Components|Audio|FadeIn _capturing_audio_loop_a 0.3 1.0 0.0 "Linear" :!class /Script/Engine.AudioComponent)
        (else
          (Audio|Components|Audio|SetSound _capturing_audio_loop_b _currentloopsound)
          (Audio|Components|Audio|FadeIn _capturing_audio_loop_b 0.3 1.0 0.0 "Linear" :!class /Script/Engine.AudioComponent))))))

(event OnPropertiesLoaded(ToyOptions)
  (bind _minigamelogic (Variables|Default|GetMinigameLogic))
  (bind _currentminigame (Variables|/Script/FortniteGame.FortMinigameLogicComponent|GetCurrentMinigame _minigamelogic))
  ;[1 1 1] Properties Changed / Loaded
  (Variables|DeviceOptions|SetPlayerNameFilter (Utilities|String|Trim (Utilities|String|TrimTrailing (Variables|DeviceOptions|GetPlayerNameFilter))))
  (Utilities|FlowControl|Sequence
    (:then_0
      (CallFunction|MigrateSameChannelInstances))
    (:then_1
      (Device|UpdateCaptureArea)
      (CallFunction|SetStartOwnerTeam)
      (ItemDelivery|UpdateItemDeliveryScore)
      (Capture|UpdateControlPointSettings)
      (CallFunction|SetBeaconParticles)
      (Device|UpdateDeviceColors)
      (IsAllowed|UpdateAllowedInteractionSetting)
      (ProgressWidget|InitInWorldWidget)
      (Visibility|UpdateVisibilityComponent)
      (Utilities|IsValid _currentminigame
        (:Is%20Valid
          (EnableKnob|SetEnablebyMinigameState (Variables|/Script/FortniteGame.FortMinigame|GetCurrentState _currentminigame)))))
    (:then_2
      (MapMarker|MapMarkerSetup))
    (:then_3
      (Variables|/Script/FortniteGame.FortMinigameObjectiveComponent|SetObjectiveSortOrder (Variables|Default|GetSortOrder) (Variables|Default|GetFortMinigameObjective)))))

(event CaptureComponent|EventOnPlayerLeavesZone
  (params (PlayerController :type Fort%20Player%20Controller%20Object%20Reference) (bIsLast :type Boolean))
  ;[1 1 1] Enter and leave zone messages
  (Utilities|IsValid PlayerController
    (:Is%20Valid
      (Networking|SwitchHasAuthority
        (:Authority
          (GameplayEvent|BroadcastEvent (Variables|UserOptions|Events|GetOnPlayerExiting%20Zone) 0 (Utilities|Struct|MakeGameplayEventCaptureAreaControlLeaveZone PlayerController))
          (UI|CaptureProgressHUDUIControl false PlayerController)
          (if bIsLast
            (GameplayEvent|BroadcastEvent (Variables|UserOptions|Events|GetOnLastPlayerExiting%20Zone) 0 (Utilities|Struct|MakeGameplayEventCaptureAreaControlLastLeaveZone PlayerController))))))))

(event Custom|Play%20Capture%20State%20Sound
  (params (Sound :type Sound%20Base%20Object%20Reference))
  (bind _captured_audio (Variables|Default|GetCaptured%20Audio))
  ;[1 1 1] Audio for Capturing area
  (if (Variables|DeviceOptions|GetKnob_CaptureSoundEnabled)
    (Audio|Components|Audio|SetSound _captured_audio Sound)
    (Audio|Components|Audio|Play _captured_audio 0.0)))

(event Custom|Stop%20Capture%20Audio
  ;[1 1 1] Audio for Capturing area
  (Variables|Audio|SetCurrentLoopSound 0)
  (if (Audio|Components|Audio|IsPlaying (Variables|Default|GetCapturing%20Audio%20Loop%20A))
    (Audio|Components|Audio|FadeOut (Variables|Default|GetCapturing%20Audio%20Loop%20A) 0.5 0.0 "Linear" :!class /Script/Engine.AudioComponent)
    (elif (Audio|Components|Audio|IsPlaying (Variables|Default|GetCapturing%20Audio%20Loop%20B))
      (Audio|Components|Audio|FadeOut (Variables|Default|GetCapturing%20Audio%20Loop%20B) 0.5 0.0 "Linear" :!class /Script/Engine.AudioComponent))))

(event Custom|SetBeaconParticles
  (bind _beaconparticletype (Variables|DeviceOptions|GetBeaconParticleType))
  (bind _beaconparticletypes (Variables|Beacon|GetBeaconParticleTypes))
  (bind _output (Utilities|Array|Get(aref) _beaconparticletypes _beaconparticletype))
  (bind (_particletype_3_45bbedad49901d6a2ff9fab0205b3f51 _colorparameters_13_112d79bd4c25b6efa3483f992bd13189) (Utilities|Struct|BreakFDestructionObjectParticleType _output))
  (bind _beaconparticle (Variables|Default|GetBeaconParticle))
  (bind _beaconparticle_2 (Variables|Default|GetBeaconParticle))
  ;[1 1 1] Set Beacon Particle
  (Networking|CanExecuteCosmeticEvents
    (:True
      (Utilities|IsValid _beaconparticle_2
        (:Is%20Valid
          (Components|Activation|SetActive _beaconparticle_2 false true))
        (:Is%20Not%20Valid
          _))
      (if (and (or (== (Variables|DeviceOptions|GetHUDElementsIndexKnob) 2) (== (Variables|DeviceOptions|GetHUDElementsIndexKnob) 3)) (Utilities|Array|IsValidIndex _beaconparticletypes _beaconparticletype))
        (bind _object (Utilities|AsyncLoadAsset _particletype_3_45bbedad49901d6a2ff9fab0205b3f51)
          (:Completed
            (bind _asparticle_system (Utilities|Casting|CastToParticleSystem _object))
            (Effects|Components|ParticleSystem|SetTemplate (Variables|Default|GetBeaconParticle) _asparticle_system)
            (Components|Activation|SetActive _beaconparticle (Variables|Replicated|GetIsEnabledByPhase) true)
            (Beacon|SetBeaconColor _beaconparticle _colorparameters_13_112d79bd4c25b6efa3483f992bd13189)
            (Beacon|SetBeaconRelativeLocation)
            (Beacon|SetBeaconScale _beaconparticle)
            (if (== (Variables|DeviceOptions|GetBeaconParticleType) 0)
              (Effects|Components|ParticleSystem|SetEmitterEnable _beaconparticle "Letter" false))))))))

(event ItemsChanged(EditorOnlyPreview)
  (params (HasItems :type Boolean))
  ;[1 1 1] Editor Preview - Show or Hide Bubble based on whether we have an item. Also, Update Relative Scale.
  (UEFNConversionHelper|SetComponentHiddenEditor (Variables|Default|GetPreviewBubble) HasItems)
  (Transformation|SetRelativeScale3D (Variables|Default|GetPreviewMeshScene) (Math|Conversions|ToVector(Integer) (Math|Conversions|ToInteger(Boolean) HasItems))))

(event Custom|LoadItemPreviewsAsync
  ;[0 1 0.1757] Load Item Previews
  (Utilities|Array|Clear (Variables|ItemPreviews|GetMeshPreviews))
  (for _array_element (ItemData|GetContainedItemVariants (Variables|Default|GetPickupItemList))
    (if (ItemPreviews|CreatePickupItemPreview _array_element)
      (Utilities|Array|Add (Variables|ItemPreviews|GetMeshPreviews) (.NewItem (ItemPreviews|CreatePickupItemPreview _array_element)))
      (bind _object (Utilities|AsyncLoadAsset (Utilities|Struct|BreakPickupItemPreview (.NewItem (ItemPreviews|CreatePickupItemPreview _array_element))))
        (:Completed
          (ItemPreviews|AddLoadedPickupItemMesh _object))))))

(event Custom|FX_MulticastPlayerScored
  ;[1 1 1] Player Scored In Area FX
  (CallFunction|FXPlayerScored))

(event Custom|FX_PlayerScored
  ;[1 1 1] Player Scored In Area FX
  (Networking|CanExecuteCosmeticEvents
    (:True
      (if (Variables|Replicated|GetIsEnabledByPhase)
        (Rendering|Material|SetVectorParameterValue (Variables|Default|GetCaptureAreaMID) "TimeStampInfo" (Math|Color|MakeColor (Utilities|Time|GetGameTimeinSeconds) 0.5 (+ (Utilities|Time|GetGameTimeinSeconds) (Variables|Constants|GetInfiniteAnimationDuration))) :!class /Script/Engine.MaterialInstanceDynamic)))))

(event Custom|FX_ObjectAddedToItemQueue
  (params (ItemAdded :type Actor%20Object%20Reference))
  (Networking|CanExecuteCosmeticEvents
    (:True
      (Utilities|IsValid ItemAdded
        (:Is%20Valid
          (Audio|PlaySoundatLocation (Variables|Audio|GetObjectAddedToCaptureAreaSound) (Transformation|GetActorLocation ItemAdded)) ; ObjectAddedToCaptureAreaSound
          (CallFunction|InitializeFX)
          (bind _asfort_pickup (Utilities|Casting|CastToFortPickup ItemAdded))
          (bind _output_get (Variables|Default|SetCurrentPickup _asfort_pickup))
          (Variables|Default|SetPickedUpWorldLocation (Transformation|GetActorLocation _output_get))
          (bind _asb_pickups_parent (Utilities|Casting|CastToB_Pickups_Parent (Utilities|Array|Get(acopy) (Actor|GetAttachedActors _asfort_pickup true false))))
          (Class|BPickupsParent|SetActiveBackgroundParticleSystem _asb_pickups_parent false false)
          (bind _asb_pickups_default (Utilities|Casting|CastToB_Pickups_Default _asb_pickups_parent))
          (Class|BPickupsParent|HideBackgroundandSpotlight _asb_pickups_default)
          (CallFunction|PlaySucktoBackpack)
          (Variables|Default|SetPickupMIDArray (Variables|/Script/FortniteGame.FortPickupsParent|GetMIDs _asb_pickups_parent))
          (Transformation|SetRelativeScale3D (Variables|Default|GetPreviewBubble) "0, 0, 0"))))))

(event Custom|FX_ObjectAddedToItemList
  (params (ItemAdded :type Actor%20Object%20Reference))
  (Networking|CanExecuteCosmeticEvents))

(event Custom|FX_ObjectCaptured
  (params (Pawn :type Fort%20Pawn%20Object%20Reference) (CapturedActor :type Actor%20Object%20Reference))
  (bind _droparea (Variables|Default|GetDropArea))
  ;[0.7672 0.2339 1] Handle the VFX when a player scores
  (Networking|CanExecuteCosmeticEvents
    (:True
      (Utilities|IsValid Pawn
        (:Is%20Valid
          (Utilities|IsValid _droparea
            (:Is%20Valid
              (if (Variables|DeviceOptions|GetKnob_CaptureSoundEnabled)
                (Audio|PlaySoundatLocation (select (Getters|IsFriendlytoLocalPlayer Pawn) (Variables|Audio|GetTeamCapturedSound) (Variables|Audio|GetEnemyCapturedSound)) (Transformation|GetWorldLocation _droparea)) ; play hologram sound
              )
              (switch Utilities|FlowControl|Switch|SwitchonKnob_ECaptureVfx (Variables|Default|GetKnob_VfxType)
                (:NewEnumerator7
                  (GameplayCue|ExecuteGameplayCue%28Burst%29Local Pawn (Variables|Default|GetCaptureAreaCollectedCue) (Variables|Default|GetEmptyCueParams)))
                (:NewEnumerator1
                  (Utilities|IsValid CapturedActor
                    (:Is%20Valid
                      (GameplayCue|ExecuteGameplayCue%28Burst%29Local CapturedActor (Variables|Default|GetCaptureAreaConfettiCue) (Variables|Default|GetEmptyCueParams))))))
              (CallFunction|PlayOverlapTL (Transformation|GetSocketLocation (Variables|/Script/Engine.Character|GetMesh Pawn) "AttachSocket")))))))))

(event Custom|PlaySuckToBackpack
  (bind (_returnvalue_x _returnvalue_y _returnvalue_z) (Math|Vector|BreakVector (+ (Transformation|GetWorldLocation (Variables|Default|GetPreviewMeshScene)) (Variables|Constants|GetBackpackLocationOffset))))
  (bind (_direction _movemeshtocenter _sucktobackpack) (Timeline|Timeline_0 PlayFromStart)
    (:Update
      (bind _currentpickup (Variables|Default|GetCurrentPickup)
        (:then
          (Transformation|SetActorLocation _currentpickup (Math|Vector|Lerp(Vector) (Variables|Default|GetPickedUpWorldLocation) (+ (Transformation|GetWorldLocation (Variables|Default|GetPreviewMeshScene)) (Variables|ItemPreviews|GetHoverLocationOffset)) _movemeshtocenter) false false)
          (for _array_element (Variables|Default|GetPickupMIDArray)
            (Rendering|Material|SetVectorParameterValue _array_element "WorldPostionOfBlackHole" (Math|Color|MakeColor _returnvalue_x _returnvalue_y _returnvalue_z _sucktobackpack) :!class /Script/Engine.MaterialInstanceDynamic) ; Also Sets Animation
          ))
        (:else)))))

(event Custom|FX_ShowRevealPreview
  ;[1 1 1] Show or Hide Reveal Preview
  (if (not (Editor|InEditor))
    (bind (_direction _scale) (Timeline|ScaleMesh PlayFromStart)
      (:Update
        (Transformation|SetRelativeScale3D (Variables|Default|GetPreviewMeshScene) _scale)))))

(event Custom|FX_ShowNextPreview
  (bind _meshpreviews (Variables|ItemPreviews|GetMeshPreviews))
  (bind _returnvalue (Utilities|Array|Length _meshpreviews))
  (bind _returnvalue_2 (<= _returnvalue 1))
  ;[1 1 1] Time to scale mesh between transitions
  (Networking|CanExecuteCosmeticEvents
    (:True
      (if (not _returnvalue_2)
        (CallFunction|FXHideRevealPreview)
        (Utilities|FlowControl|Delay (Components|Timeline|GetTimelineLength (Variables|BP_Creative_CaptureArea|GetScaleMesh))))
      (ItemPreviews|ShowNextItemPreview)
      (if (not _returnvalue_2)
        (Audio|PlaySoundatLocation (Variables|Audio|GetTeamCapturedSound) (Transformation|GetWorldLocation (Variables|Default|GetDropArea))) ; play hologram sound
        (Utilities|FlowControl|Delay 0.2)
        (CallFunction|FXShowRevealPreview)))))

(event Custom|FX_HideRevealPreview
  ;[1 1 1] Show or Hide Reveal Preview
  (Timeline|ScaleMesh ReverseFromEnd))

(event Custom|PlayOverlapTL
  (params (OverlapPoint_X :type Float%20%28double-precision%29) (OverlapPoint_Y :type Float%20%28double-precision%29) (OverlapPoint_Z :type Float%20%28double-precision%29))
  (bind _returnvalue (Transformation|GetActorUpVector))
  (bind _captureareastaticmesh (Variables|Default|GetCaptureAreaStaticMesh))
  (bind (_returnvalue_x _returnvalue_y _returnvalue_z) (Math|Vector|BreakVector (Transformation|GetWorldLocation _captureareastaticmesh)))
  (bind _returnvalue_2 (Math|Vector|vector-vector (Math|Vector|MakeVector OverlapPoint_X OverlapPoint_Y _returnvalue_z) (Math|Vector|MakeVector _returnvalue_x _returnvalue_y _returnvalue_z)))
  (bind _returnvalue_3 (* _returnvalue_2 -1.0))
  (bind _returnvalue_4 (Math|Vector|DotProduct _returnvalue_3 _returnvalue))
  (bind _returnvalue_5 (* _returnvalue _returnvalue_4))
  (bind _returnvalue_6 (- _returnvalue_3 _returnvalue_5))
  (bind _returnvalue_7 (Transformation|GetActorTransform))
  (bind _returnvalue_8 (Math|Transform|InverseTransformDirection _returnvalue_7 _returnvalue_6))
  (bind (_origin _boxextent) (Collision|GetComponentBounds (Variables|Default|GetCaptureAreaStaticMesh)))
  (bind (_boxextent_x _boxextent_y _boxextent_z) (Math|Vector|BreakVector _boxextent))
  ;[1 1 1] Player has interacted visuals
  (if (not (Math|Vector|VectorIsNearlyZero _returnvalue_8))
    (Transformation|SetRelativeRotation (Variables|Default|GetCaptureAreaStaticMesh) (Math|Rotator|MakeRotfromX _returnvalue_8) false false))
  (Rendering|Material|SetScalarParameterValue (Variables|Default|GetCaptureAreaMID) "VBias" (- (/ (- OverlapPoint_Z _returnvalue_z) _boxextent_z) 0.5) :!class /Script/Engine.MaterialInstanceDynamic)
  (bind (_direction _ring _pulse _ringopacity) (Timeline|OverlapTL PlayFromStart)
    (:Update
      (Rendering|Material|SetScalarParameterValue (Variables|Default|GetCaptureAreaMID) "ImpactParam" _ring :!class /Script/Engine.MaterialInstanceDynamic)
      (Rendering|Material|SetScalarParameterValue (Variables|Default|GetCaptureAreaMID) "Pulse" _pulse :!class /Script/Engine.MaterialInstanceDynamic)
      (Rendering|Material|SetScalarParameterValue (Variables|Default|GetCaptureAreaMID) "RingOpacity" _ringopacity :!class /Script/Engine.MaterialInstanceDynamic))))

(event Custom|FX_StartItemPreviewRotation
  ;[0.02416 0.7379 1] Preview Mesh Rotation
  (if (not (Math|Boolean|ORBoolean (TouchPlatform|IsMobileGame) (TouchPlatform|IsNintendoSwitch) (Variables|Default|GetDoNotRenderBubble)))
    (bind (_direction _0_360) (Timeline|RotateMesh Play)
      (:Update
        (Transformation|SetRelativeRotation (Variables|Default|GetPreviewMeshScene) (Math|Rotator|MakeRotator 0.0 0.0 _0_360) false false)))))

(event Custom|FX_StopItemPreviewRotation
  ;[0.02416 0.7379 1] Preview Mesh Rotation
  (Timeline|RotateMesh Stop))

(event Custom|FX_PlayIsAllowedAnimation
  (params (PlayIsAllowedAnimation :type Boolean) (FadeDuration :type Float%20%28double-precision%29))
  ;[1 1 1] Play Locked Animation On/Off With/Without Fade
  (Networking|CanExecuteCosmeticEvents
    (:True
      (if (not PlayIsAllowedAnimation)
        (Rendering|Material|SetVectorParameterValue (Variables|Default|GetCaptureAreaMID) "Blocked State Animation Info" (Math|Color|MakeColor (Utilities|Time|GetTimeSeconds) 1.0 0.0 FadeDuration) :!class /Script/Engine.MaterialInstanceDynamic)
        (else
          (Rendering|Material|SetVectorParameterValue (Variables|Default|GetCaptureAreaMID) "Blocked State Animation Info" (Math|Color|MakeColor 0.0 0.0 (Utilities|Time|GetTimeSeconds) FadeDuration) :!class /Script/Engine.MaterialInstanceDynamic))))))

(event Custom|FX_HandleEnable
  ;[1 0 0.0728] FX for Enable / Disable Events
  (if (Variables|Replicated|GetIsEnabledByPhase)
    ;;[1 1 1] Enable Animation
    (Rendering|Material|SetVectorParameterValue (Variables|Default|GetCaptureAreaMID) "TimeStampInfo" (Math|Color|MakeColor (Utilities|Time|GetGameTimeinSeconds) (Variables|MutableAnimationRelated|GetEnableAnimationLength) (+ (Utilities|Time|GetGameTimeinSeconds) (Variables|Constants|GetInfiniteAnimationDuration)) 0.0) :!class /Script/Engine.MaterialInstanceDynamic)
    (ItemPreviews|SetBubbleScale)
    (CallFunction|FXShowRevealPreview)
    (Rendering|SetVisibility (Variables|Default|GetCaptureAreaLight) true false)
    (Components|Activation|Activate (Variables|Default|GetP_CaptureArea_Blocks) true)
    (CallFunction|SetBeaconParticles)
    (else
      (Rendering|Material|SetVectorParameterValue (Variables|Default|GetCaptureAreaMID) "TimeStampInfo" (Math|Color|MakeColor 0.0 (Variables|MutableAnimationRelated|GetEnableAnimationLength) (Utilities|Time|GetGameTimeinSeconds) 0.5) :!class /Script/Engine.MaterialInstanceDynamic)
      (CallFunction|FXHideRevealPreview)
      (ItemPreviews|SetBubbleScale)
      (Rendering|SetVisibility (Variables|Default|GetCaptureAreaLight) false false)
      (Components|Activation|Deactivate (Variables|Default|GetP_CaptureArea_Blocks))
      (Components|Activation|SetActive (Variables|Default|GetBeaconParticle) false true))))

(event Custom|FX%20OnOwnerChanged
  ;[1 1 1] Fx On Owner Changed
  (Networking|CanExecuteCosmeticEvents
    (:True
      (if (Variables|Replicated|GetIsEnabledByPhase)
        (Rendering|Material|SetVectorParameterValue (Variables|Default|GetCaptureAreaMID) "TimeStampInfo" (Math|Color|MakeColor (Utilities|Time|GetGameTimeinSeconds) 0.5 (+ (Utilities|Time|GetGameTimeinSeconds) (Variables|Constants|GetInfiniteAnimationDuration)) 1.5) :!class /Script/Engine.MaterialInstanceDynamic)
        (Rendering|Material|SetScalarParameterValue (Variables|Default|GetCaptureAreaMID) "Capture Progress" 0.0 :!class /Script/Engine.MaterialInstanceDynamic)))))

(event Custom|ActivateObjectivePulseFromEvent
  (params (Instigator :type Controller%20Object%20Reference))
  (Class|CreativeObjectiveTrailComponent|ToggleObjectiveTrail (Variables|Default|GetCreative_ObjectiveTrail_Component) true self Instigator (Variables|DeviceOptions|GetTeamSetting) (Variables|Default|GetKnob_bShowObjectiveTrailFriendly) (Variables|Default|GetKnob_bShowObjectiveTrailEnemy) (Variables|Default|GetKnob_bShowObjectiveTrailInstigator) (Variables|Default|GetKnob_bPlayAudio) (Variables|Default|GetKnob_fPulseRate) 0))

(event Custom|DeactivateObjectivePulseFromEvent
  (params (Instigator :type Controller%20Object%20Reference))
  (Class|CreativeObjectiveTrailComponent|ToggleObjectiveTrail (Variables|Default|GetCreative_ObjectiveTrail_Component) false self Instigator (Variables|DeviceOptions|GetTeamSetting) (Variables|Default|GetKnob_bShowObjectiveTrailFriendly) (Variables|Default|GetKnob_bShowObjectiveTrailEnemy) (Variables|Default|GetKnob_bShowObjectiveTrailInstigator) (Variables|Default|GetKnob_bPlayAudio) (Variables|Default|GetKnob_fPulseRate) 0))

(event OnItemTemplateChanged(ItemVariantNotifier)
  (params (ItemTemplate :type Fort%20Creative%20Item%20Template%20Object%20Reference))
  (Networking|CanExecuteCosmeticEvents
    (:True
      (ItemPreviews|StartorRefreshItemPreviews (Variables|/Script/FortniteGame.FortMinigameLogicComponent|GetCurrentMinigame (Variables|Default|GetMinigameLogic))))))

(event OnItemTemplateEnabled(ItemVariantNotifier)
  (params (ItemTemplate :type Fort%20Creative%20Item%20Template%20Object%20Reference))
  (Networking|CanExecuteCosmeticEvents
    (:True
      (ItemPreviews|StartorRefreshItemPreviews (Variables|/Script/FortniteGame.FortMinigameLogicComponent|GetCurrentMinigame (Variables|Default|GetMinigameLogic))))))

(event OnItemTemplateDisabled(ItemVariantNotifier)
  (params (ItemTemplate :type Fort%20Creative%20Item%20Template%20Object%20Reference))
  (Networking|CanExecuteCosmeticEvents
    (:True
      (ItemPreviews|StartorRefreshItemPreviews (Variables|/Script/FortniteGame.FortMinigameLogicComponent|GetCurrentMinigame (Variables|Default|GetMinigameLogic))))))

(fn ActivateFromOtherDevice
  (params (Instigator :type Controller%20Object%20Reference) (Requester :type ERequesterType%20Enum :default "NewEnumerator0"))
  (return false))

(fn AddAsObjectiveToMinigame
  (params (Amount :type Integer :default 1) (FortPlayerController :type Fort%20Player%20Controller%20Object%20Reference))
  (bind _controller (LocalVariables|GetController))
  (Utilities|FlowControl|Sequence
    (:then_0
      (Utilities|IsValid FortPlayerController
        (:Is%20Valid
          (LocalVariables|SetController FortPlayerController))
        (:Is%20Not%20Valid
          (bind _fortplayercontroller (CallFunction|TrytoGetValidControllerfromTeam (Variables|Team|GetTeamIndex)))
          (bind _output_get (LocalVariables|SetController _fortplayercontroller))
          (Variables|Default|SetCurrentOwner _output_get))))
    (:then_1
      (Utilities|IsValid _controller
        (:Is%20Valid
          (bind _returnvalue (CreativeDevices|GetMinigameStatsComp _controller))
          (LocalVariables|Setlocal%20MinigameStatsComp _returnvalue)
          (bind _local_minigamestatscomp (LocalVariables|Getlocal%20MinigameStatsComp)
            (:then
              (Class|FortMinigameStatsComponent|SendStatEventwithTags _local_minigamestatscomp (LocalVariables|GetController) "MinigameDynamicEvent" self (GameplayTags|MakeGameplayTagContainerfromTag "(TagName=\"Creative.Item.Destructible\")") (Class|FortMinigameStatsComponent|GetSourceandContextTags _local_minigamestatscomp (LocalVariables|GetController)) (.OutContextTags (Class|FortMinigameStatsComponent|GetSourceandContextTags _local_minigamestatscomp (LocalVariables|GetController))) (Variables|GetAmount)))
            (:else)))))))

(fn AddItemToQueue
  (params (NewItem :type Actor%20Object%20Reference))
  (bind _bitemadded (Class|PickupItemListComponent|AddItemtoQueue (Variables|Default|GetPickupItemList) NewItem))
  (if _bitemadded
    (CallFunction|FXObjectAddedtoItemQueue (Variables|GetNewItem))))

(fn AddLoadedPickupItemMesh
  (params (MeshObject :type Object%20Reference))
  (bind _meshpreviews (Variables|ItemPreviews|GetMeshPreviews))
  (Utilities|IsValid MeshObject
    (:Is%20Valid
      (for _index (range (Utilities|Array|Length (Variables|ItemPreviews|GetMeshPreviews)))
        (if (Utilities|Equal(SoftObjectReference) (Utilities|Struct|BreakPickupItemPreview (Utilities|Array|Get(aref) _meshpreviews _index)) (Utilities|ResolveSoftReference (Variables|GetMeshObject)))
          (Utilities|Struct|SetmembersinPickupItemPreview (Utilities|Array|Get(aref) _meshpreviews _index) (Variables|GetMeshObject)))))))

(fn AllowInteraction
  (Networking|CanExecuteCosmeticEvents
    (:True
      (CallFunction|FXPlayIsAllowedAnimation (Variables|Default|GetIsAllowed) 0.5))
    (:False
      (EnableKnob|UpdateOverlappingActors)
      (Minigame|Objective|SetIsInteractionAllowed (Variables|Default|GetFortMinigameObjective) true))))

(fn BindOrUnbindOnPlayerChangedTeam
  (params (Bind :type Boolean) (CurrentMinigame :type Fort%20Minigame%20Object%20Reference))
  (bind _outputdelegate (EventDispatchers|CreateEvent :!event OnPlayerChangedTeam))
  (Networking|SwitchHasAuthority
    (:Authority
      (Utilities|IsValid CurrentMinigame
        (:Is%20Valid
          (if (Variables|GetBind)
            (Minigame|BindEventtoPlayerChangedTeamDelegate CurrentMinigame _outputdelegate)
            (else
              (Minigame|UnbindEventfromPlayerChangedTeamDelegate CurrentMinigame _outputdelegate))))))))

(fn BindOrUnbindOnPlayerEmote
  (params (Bind :type Boolean) (Pawn :type Fort%20Pawn%20Object%20Reference))
  (bind _outputdelegate (EventDispatchers|CreateEvent :!event OnPawnStartedEmote))
  (bind _outputdelegate_2 (EventDispatchers|CreateEvent :!event OnPawnStoppedEmote))
  (Networking|SwitchHasAuthority
    (:Authority
      (Utilities|IsValid Pawn
        (:Is%20Valid
          (if (Variables|GetBind)
            (EventDispatchers|BindEventtoOnPawnStartedEmote Pawn _outputdelegate)
            (EventDispatchers|BindEventtoOnPawnStoppedEmote Pawn _outputdelegate_2)
            (else
              (EventDispatchers|UnbindEventfromOnPawnStartedEmote Pawn _outputdelegate)
              (EventDispatchers|UnbindEventfromOnPawnStartedEmote Pawn _outputdelegate_2))))))))

(fn BindOrUnbindToPlayerAddAndRemove
  (params (Bind :type Boolean) (CurrentMinigame :type Fort%20Minigame%20Object%20Reference))
  (bind _outputdelegate (EventDispatchers|CreateEvent :!event OnPlayerRemovedFromMinigame))
  (bind _outputdelegate_2 (EventDispatchers|CreateEvent :!event OnPlayerAddedToMinigame))
  (Networking|SwitchHasAuthority
    (:Remote
      (Utilities|IsValid CurrentMinigame
        (:Is%20Valid
          (if (Variables|GetBind)
            (Minigame|BindEventtoPlayerAddedDelegate CurrentMinigame _outputdelegate_2)
            (Minigame|BindEventtoPlayerRemovedDelegate CurrentMinigame _outputdelegate)
            ;[1 1 1] There may be already a player inside
            (for _array_element (Minigame|GetParticipatingPlayers CurrentMinigame)
              (CaptureArea|BindorUnbindonPlayerTeamReplicated _array_element true))
            (else
              (Minigame|UnbindEventfromPlayerAddedDelegate CurrentMinigame _outputdelegate_2)
              (Minigame|UnbindEventfromPlayerRemovedDelegate CurrentMinigame _outputdelegate))))))))

(fn BindToContainedItemVariants
  (ItemTemplate|UnbindfromAllItemVariants (Variables|Default|GetItemVariantNotifier))
  (bind _array_element (Utilities|Array|ForEachLoopwithBreak (ItemData|GetContainedItemVariants (Variables|Default|GetPickupItemList)) :!label _label)
    (:LoopBody
      (ItemTemplate|BindtoItemVariant (Variables|Default|GetItemVariantNotifier) _array_element)
      (goto _label Break))))

(fn CanBeCapturedBy
  (params (TeamToCheck :type Byte))
  (Utilities|FlowControl|Sequence
    (:then_0
      (if (Knobs|Team|IsTeamNone (Variables|DeviceOptions|GetControlledByIndex))
        (return false)))
    (:then_1
      (if (and (Knobs|Team|IsTeamAll (Variables|DeviceOptions|GetControlledByIndex)) (Math|Byte|NotEqual(Byte) (Variables|GetTeamToCheck) (Variables|Team|GetTeamIndex)))
        (return true)))
    (:then_2
      (if (Knobs|Team|IsValidTeamIndex (Variables|DeviceOptions|GetControlledByIndex))
        (return (and (Math|Byte|Equal(Byte) (Variables|GetTeamToCheck) (Teams|IndextoTeam (Knobs|Team|GetTeamValueasIndex (Variables|DeviceOptions|GetControlledByIndex)))) (Math|Byte|NotEqual(Byte) (Variables|GetTeamToCheck) (Variables|Team|GetTeamIndex))))))
    (:then_3
      (return false))))

(fn CanBeUsedForScoringBy
  (params (TeamToCheck :type Byte))
  (Utilities|FlowControl|Sequence
    (:then_0
      (if (Knobs|Team|IsTeamNone (Variables|DeviceOptions|GetCanStartPeriodicScoreIndex))
        (return false)))
    (:then_1
      (if (and (Knobs|Team|IsTeamHostiles (Variables|DeviceOptions|GetCanStartPeriodicScoreIndex)) (Math|Byte|NotEqual(Byte) (Variables|GetTeamToCheck) (Variables|Team|GetTeamIndex)))
        (return true)))
    (:then_2
      (if (Knobs|Team|IsTeamAllies (Variables|DeviceOptions|GetCanStartPeriodicScoreIndex))
        (return (or (Math|Byte|Equal(Byte) (Variables|GetTeamToCheck) (Variables|Team|GetTeamIndex)) (Math|Byte|Equal(Byte) (Variables|Team|GetTeamIndex) (Utilities|Enum|LiteralenumEFortTeam "Spectator"))))))
    (:then_3
      (if (Knobs|Team|IsValidTeamIndex (Variables|DeviceOptions|GetCanStartPeriodicScoreIndex))
        (return (Math|Byte|Equal(Byte) (Teams|IndextoTeam (Knobs|Team|GetTeamValueasIndex (Variables|DeviceOptions|GetCanStartPeriodicScoreIndex))) (Variables|GetTeamToCheck)))))
    (:then_4
      (if (Knobs|Team|IsTeamAll (Variables|DeviceOptions|GetCanStartPeriodicScoreIndex))
        (return true)))
    (:then_5
      (return false))))

(fn CanDeliverItems
  (params (TeamToCheck :type Byte))
  (Utilities|FlowControl|Sequence
    (:then_0
      (if (Knobs|Team|IsTeamAll (Variables|DeviceOptions|GetCanReceiveItemsFrom))
        (return true)))
    (:then_1
      (if (Knobs|Team|IsTeamNone (Variables|DeviceOptions|GetCanReceiveItemsFrom))
        (return false)))
    (:then_2
      (if (and (Knobs|Team|IsTeamHostiles (Variables|DeviceOptions|GetCanReceiveItemsFrom)) (Math|Byte|NotEqual(Byte) (Variables|GetTeamToCheck) (Variables|Team|GetTeamIndex)))
        (return true)))
    (:then_3
      (if (Knobs|Team|IsTeamAllies (Variables|DeviceOptions|GetCanReceiveItemsFrom))
        (return (or (Math|Byte|Equal(Byte) (Variables|GetTeamToCheck) (Variables|Team|GetTeamIndex)) (Math|Byte|Equal(Byte) (Variables|Team|GetTeamIndex) (Utilities|Enum|LiteralenumEFortTeam "Spectator"))))))
    (:then_4
      (if (Knobs|Team|IsValidTeamIndex (Variables|DeviceOptions|GetCanReceiveItemsFrom))
        (return (Math|Byte|Equal(Byte) (Teams|IndextoTeam (Knobs|Team|GetTeamValueasIndex (Variables|DeviceOptions|GetCanReceiveItemsFrom))) (Variables|GetTeamToCheck)))))
    (:then_5
      (return false))))

(fn CanInteract
  (params (InteractingPawn :type Fort%20Pawn%20Object%20Reference) (InteractionBeingAttempted :type EInteractionBeingAttempted%20Enum :default "FirstInteraction") (InteractionType :type TInteractionType%20Enum :default "IT_NoInteraction"))
  (return false))

(fn CaptureProgressHUDUIControl
  (params (TurnOn :type Boolean) (PlayerControllerRef :type Fort%20Player%20Controller%20Object%20Reference))
  (bind _playercontrollerref (Variables|GetPlayerControllerRef))
  (bind _returnvalue (Actor|GetComponentbyClass _playercontrollerref "/Game/Creative/Devices/CaptureArea/CRD_CaptureProgressOnHUDComp.CRD_CaptureProgressOnHUDComp_C"))
  (bind _playercontrollerref_2 (Variables|GetPlayerControllerRef))
  (bind _returnvalue_2 (Actor|GetComponentbyClass _playercontrollerref_2 "/Game/Creative/Devices/CaptureArea/CRD_CaptureProgressOnHUDComp.CRD_CaptureProgressOnHUDComp_C"))
  (Utilities|IsValid (Variables|GetPlayerControllerRef)
    (:Is%20Valid
      (if (Variables|DeviceOptions|GetShowCaptureProgressOnHUD)
        ;[1 1 1] Used to Display Team Count on HUD
        (bind _returnvalue_3 (CallFunction|GetorFetchFortPlayspaceEventRouter))
        (GameplayEventRouter|BroadcastEvent _returnvalue_3 (Utilities|Struct|MakeGERPlayerEnteredorLeftCapArea (Variables|GetPlayerControllerRef)) self)
        (if (Variables|GetTurnOn)
          ;[1 1 1] Add Component if it doesn't already exist
          (Utilities|IsValid _returnvalue_2
            (:Is%20Valid
              (LocalVariables|SetCaptureProgressComponentRef _returnvalue_2))
            (:Is%20Not%20Valid
              (bind _returnvalue_4 (AddComponent|NODE_AddCRD_ScreenCaptureProgressComp-0 :self (Variables|GetPlayerControllerRef)))
              (LocalVariables|SetCaptureProgressComponentRef _returnvalue_4)))
          ;[1 1 1] Assign Capture Area in Component
          (Class|CRDCaptureProgressonHUDComp|AddCaptureArea (LocalVariables|GetCaptureProgressComponentRef) self)
          ;[1 1 1] Listen for team switch while in area -
          ; Fixes a bug where switching team while in zone didn't get reflected in widget's allies count display
          (bind _asfort_player_state_athena (Utilities|Casting|CastToFortPlayerStateAthena (Variables|/Script/Engine.Controller|GetPlayerState (Variables|GetPlayerControllerRef))))
          (Player|BindEventtoOnTeamIndexChangedDelegate _asfort_player_state_athena (EventDispatchers|CreateEvent :!event OnPlayerWithinZoneSwitchedTeams))
          (else
            ;[1 1 1] Remove Reference from Component (it cleans up itself if the capture area list is empty)
            (Utilities|IsValid _returnvalue
              (:Is%20Valid
                (Class|CRDCaptureProgressonHUDComp|RemoveCaptureArea _returnvalue self)
                (bind _asfort_player_state_athena_2 (Utilities|Casting|CastToFortPlayerStateAthena (Variables|/Script/Engine.Controller|GetPlayerState (Variables|GetPlayerControllerRef))))
                (Player|UnbindEventfromOnTeamIndexChangedDelegate _asfort_player_state_athena_2 (EventDispatchers|CreateEvent :!event OnPlayerWithinZoneSwitchedTeams))))))))))

(fn CheckActorVelocityVectortoFilter
  (params (Captured%20Actor :type Actor%20Object%20Reference))
  (bind _returnvalue (Transformation|GetVelocity Captured%20Actor))
  (bind _self self)
  (bind _returnvalue_2 (Transformation|GetActorUpVector _self))
  (bind _returnvalue_3 (Math|Vector|Normalize _returnvalue_2))
  (bind _returnvalue_4 (Math|Vector|Normalize _returnvalue))
  (bind _returnvalue_5 (Math|Vector|DotProduct _returnvalue_4 _returnvalue_3))
  ;[1 0.8808 0.185] Grab a normalized vector of the object's velocity, then get a dot product of it with the local up vector to see if the object is moving either up or down relative to the capture area
  (Utilities|IsValid Captured%20Actor
    (:Is%20Valid
      (switch Utilities|FlowControl|Switch|SwitchonKnob_EVectorFilterDirection (Variables|Default|GetKnob_VectorFilter)
        (:NewEnumerator7
          (return true))
        (:NewEnumerator1
          (return (> _returnvalue_5 0)))
        (:NewEnumerator0
          (return (< _returnvalue_5 0)))))
    (:Is%20Not%20Valid
      (return false))))

(fn CheckForMeshMaterialExceptions
  (params (Static%20Mesh :type Static%20Mesh%20Object%20Reference) (DefaultMaterialInterface :type Material%20Interface%20Object%20Reference))
  (bind (_value _returnvalue) (Utilities|Map|Find (Variables|Default|GetSpecifyMaterialForTheseStaticMeshes) (Utilities|ResolveSoftReference Static%20Mesh)))
  (return (select _returnvalue _value DefaultMaterialInterface)))

(fn CheckMeshForSpecialExceptions
  (params (StaticMesh :type Static%20Mesh%20Object%20Reference) (Default%20Transform :type Transform))
  (bind _staticmesh (Variables|Mesh|GetStaticMesh))
  (bind _output (Utilities|ResolveSoftReference _staticmesh))
  (bind _specifytransformforthesestaticmeshes (Variables|Default|GetSpecifyTransformForTheseStaticMeshes))
  (bind (_value _returnvalue) (Utilities|Map|Find _specifytransformforthesestaticmeshes _output))
  ;[1 1 1] Avoid rendering the "bubble" for specific objects
  (Variables|Default|SetDoNotRenderBubble (Utilities|Array|ContainsItem (Variables|Default|GetDoNotRenderBubbleForTheseStaticMeshes) (Utilities|ResolveSoftReference (Variables|Mesh|GetStaticMesh))))
  ;[1 1 1] Allow user-specified transforms for specific objects
  (if _returnvalue
    (Transformation|SetRelativeTransform (Variables|Default|GetPreviewStaticMesh) _value false true)
    (else
      (Transformation|SetRelativeTransform (Variables|Default|GetPreviewStaticMesh) (Variables|GetDefault%20Transform) false true))))

(fn CleanUpPreviousCaptureState
  (params (PreviousCaptureState :type ECaptureState%20Enum :default "Neutral"))
  (Networking|CanExecuteCosmeticEvents
    (:True
      (if (and (Utilities|Enum|Equal(Enum) PreviousCaptureState "Neutralizing") (Utilities|Enum|Equal(Enum) (CaptureComponent|GetCaptureState (Variables|CaptureArea|GetCaptureComponent)) "Neutral"))
        (Rendering|Material|SetScalarParameterValue (Variables|Default|GetCaptureAreaMID) (Variables|MIDParameters|GetMIDParameterTeamOwner) 0.0 :!class /Script/Engine.MaterialInstanceDynamic)
        (elif (and (Utilities|Enum|Equal(Enum) (Variables|GetPreviousCaptureState) "Deneutralizing") (Utilities|Enum|Equal(Enum) (CaptureComponent|GetCaptureState (Variables|CaptureArea|GetCaptureComponent)) "Captured"))
          (Rendering|Material|SetScalarParameterValue (Variables|Default|GetCaptureAreaMID) (Variables|MIDParameters|GetMIDParameterTeamOwner) (Getters|GetFloatfromTeam (Variables|Team|GetTeamIndex)) :!class /Script/Engine.MaterialInstanceDynamic)
          (elif (and (Utilities|Enum|Equal(Enum) (Variables|GetPreviousCaptureState) "Decapturing") (Utilities|Enum|Equal(Enum) (CaptureComponent|GetCaptureState (Variables|CaptureArea|GetCaptureComponent)) "Captured"))
            (Rendering|Material|SetScalarParameterValue (Variables|Default|GetCaptureAreaMID) (Variables|MIDParameters|GetMIDParameterInteractingTeam) (Getters|GetFloatfromTeam (Variables|Team|GetTeamIndex)) :!class /Script/Engine.MaterialInstanceDynamic)))))))

(fn CreatePickupItemPreview
  (params (ItemHandle :type Item%20Variant%20Handle%20Structure))
  (bind _itemhandle (Variables|GetItemHandle))
  (bind _rarity (Creative|ItemData|GetRarity _itemhandle))
  (bind _returnvalue (Math|Conversions|ToInteger(Byte) _rarity))
  (bind _raritysourcecolors (Variables|ItemPreviews|GetRaritySourceColors))
  (Utilities|FlowControl|Sequence
    (:then_0
      (bind _output_get (LocalVariables|SetItemDefinition (Creative|ItemData|GetItemDefinition ItemHandle "/Script/FortniteGame.FortWorldItemDefinition")))
      (Utilities|IsValid _output_get
        (:Is%20Valid
          (bind _meshtransform (Class|FortItemFunctionLibrary|GetPickupMeshTransform _output_get)
            (:IsIdentity
              _)
            (:IsOverridden
              (LocalVariables|SetTransform _meshtransform)))
          (bind _output_get_2 (LocalVariables|SetMeshObjectPath (ItemPreviews|GetMeshObjectPathforItem _output_get)))
          (if (Utilities|IsValidSoftObjectReference _output_get_2)
            (Utilities|FlowControl|Sequence
              (:then_0
                (if (Utilities|Array|IsValidIndex _raritysourcecolors _returnvalue)
                  (LocalVariables|SetRarityColor (Utilities|Array|Get(aref) _raritysourcecolors _returnvalue))
                  (else
                    (LocalVariables|SetRarityColor (Utilities|Array|Get(aref) (Variables|ItemPreviews|GetRaritySourceColors))))))
              (:then_1
                (LocalVariables|SetbSuccess true)
                (Utilities|Struct|SetmembersinPickupItemPreview (LocalVariables|GetOutputItem) (LocalVariables|GetMeshObjectPath) (LocalVariables|GetRarityColor) (LocalVariables|GetTransform))))))))
    (:then_1
      (return (LocalVariables|GetbSuccess) (LocalVariables|GetOutputItem)))))

(fn DisallowInteraction
  (Networking|CanExecuteCosmeticEvents
    (:True
      (CallFunction|FXPlayIsAllowedAnimation (Variables|Default|GetIsAllowed) 0.5)
      (Rendering|Material|SetScalarParameterValue (Variables|Default|GetCaptureAreaMID) (Variables|MIDParameters|GetMIDParameterCaptureEffectDirection) 0.0 :!class /Script/Engine.MaterialInstanceDynamic)
      (Rendering|Material|SetScalarParameterValue (Variables|Default|GetCaptureAreaMID) (Variables|MIDParameters|GetMIDParameterCaptureEffectOpacity) 0.0 :!class /Script/Engine.MaterialInstanceDynamic)
      (Device|UpdateDeviceColors)
      (CallFunction|PlayCaptureStateSound (Variables|Audio|GetNeutralizedSound)))
    (:False
      (Device|ResetDeviceCachedData false)
      (Minigame|Objective|SetIsInteractionAllowed (Variables|Default|GetFortMinigameObjective) false)
      (if (Getters|IsMinigameinProgress)
        (Capture|SetCaptureStatebyCurrentTeam true)))))

(fn DoesPlayerHaveCaptureItem
  (params (FortPlayerController :type Fort%20Player%20Controller%20Object%20Reference))
  (bind _array_element (Utilities|Array|ForEachLoopwithBreak (ItemData|GetContainedItemVariants (Variables|Default|GetPickupItemList)) :!label _label)
    (:LoopBody
      (bind (_array_element_item _array_element_itemvariantguid) (Utilities|Struct|BreakItemVariantHandle _array_element))
      (if (> (Inventory|GetItemQuantityOnPlayer (Variables|GetFortPlayerController) _array_element_item _array_element_itemvariantguid) 0)
        (LocalVariables|SetFoundMatchingItem true)
        (goto _label Break)))
    (:Completed
      (return (LocalVariables|GetFoundMatchingItem)))))

(fn EnableCaptureComponent
  (if (Utilities|Enum|Equal(Enum) (Utilities|Enum|BytetoEnumEFortTeam (Variables|Team|GetTeamIndex)) "Spectator")
    (LocalVariables|SetNextCaptureState "Neutral")
    (CaptureComponent|ForceCaptureProgress (Variables|CaptureArea|GetCaptureComponent) 0.0)
    (else
      (LocalVariables|SetNextCaptureState "Captured")
      (CaptureComponent|ForceCaptureProgress (Variables|CaptureArea|GetCaptureComponent) 1.0)))
  (CaptureComponent|SetCaptureState (Variables|CaptureArea|GetCaptureComponent) (LocalVariables|GetNextCaptureState) (Variables|Team|GetTeamIndex)))

(fn ForceCapture
  (params (TriggerController :type Fort%20Player%20Controller%20Object%20Reference) (TriggerTeam :type Byte))
  (Utilities|FlowControl|Sequence
    (:then_0
      (Utilities|IsValid TriggerController
        (:Is%20Valid
          (Variables|Default|SetForcedCaptureStateController TriggerController))
        (:Is%20Not%20Valid
          (bind _fortplayercontroller (CallFunction|TrytoGetValidControllerfromTeam (Variables|GetTriggerTeam)))
          (Variables|Default|SetForcedCaptureStateController _fortplayercontroller))))
    (:then_1
      (Variables|Default|SetWasCaptureForced true)
      (if (Utilities|Enum|Equal(Enum) (CaptureComponent|GetCaptureState (Variables|CaptureArea|GetCaptureComponent)) "Captured")
        (CaptureComponent|SetCaptureState (Variables|CaptureArea|GetCaptureComponent) "Disabled" (Variables|GetTriggerTeam)))
      (CaptureComponent|SetCaptureState (Variables|CaptureArea|GetCaptureComponent) "Captured" (Variables|GetTriggerTeam)))
    (:then_2
      (Variables|Default|SetWasCaptureForced false)
      (Variables|Default|SetForcedCaptureStateController 0))))

(fn GetDeviceColor
  (return (Utilities|Array|Get(aref) (Class|GetClassDefaults "/Game/Creative/Sets/MilitaryBase/Props/CreativeTeamColors.CreativeTeamColors_C") (Math|Integer|SelectInt (Math|Float|Truncate (Getters|GetAccentColorParam)) (- (Utilities|Array|Length (Class|GetClassDefaults "/Game/Creative/Sets/MilitaryBase/Props/CreativeTeamColors.CreativeTeamColors_C")) 1) (Utilities|Array|IsValidIndex (Class|GetClassDefaults "/Game/Creative/Sets/MilitaryBase/Props/CreativeTeamColors.CreativeTeamColors_C") (Math|Float|Truncate (Getters|GetAccentColorParam)))))))

(fn GetInstigatingController
  (return 0))

(fn GetPawnToAwardScoreToFromInstigator
  (params (Instigator :type Controller%20Object%20Reference))
  ;[0.3431 0.7003 1] Returns the pawn and controller of the player that should receive points. Returns nothing if the player is not on the correct team, and there is no player to pass the score to on the other team
  (Utilities|IsValid Instigator
    (:Is%20Valid
      (if (ItemDelivery|CanDeliverItems (Team|GetActorTeam Instigator))
        (bind _asfort_player_controller (Utilities|Casting|CastToFortPlayerController Instigator))
        (return (Game|Player|GetPlayerPawn _asfort_player_controller) _asfort_player_controller)
        (else
          ;;[1 0.9338 0.2831] If the score was made by a player on the wrong team, find a different pawn on the correct team and give them the score
          (bind _currentminigame (Variables|/Script/FortniteGame.FortMinigameLogicComponent|GetCurrentMinigame (Variables|Default|GetMinigameLogic))
            (:then
              (for _array_element (Minigame|GetParticipatingPlayers _currentminigame)
                (Utilities|IsValid (PlayerState|GetCurrentPawn _array_element)
                  (:Is%20Valid
                    (bind _asfort_player_controller_2 (Utilities|Casting|CastToFortPlayerController (Pawn|GetController (PlayerState|GetCurrentPawn _array_element))))
                    (if (ItemDelivery|CanDeliverItems (Team|GetActorTeam _asfort_player_controller_2))
                      (return (PlayerState|GetCurrentPawn _array_element) _asfort_player_controller_2)))))
              (return 0 0))
            (:else)))))))

(fn GetAccentColorParam
  (Utilities|FlowControl|Sequence
    (:then_0
      (if (>= (Variables|DeviceOptions|GetAccentColorIndex) 0)
        (LocalVariables|SetParamValue (Math|Conversions|ToFloat(Integer) (Variables|DeviceOptions|GetAccentColorIndex)))
        (elif (== (Variables|DeviceOptions|GetAccentColorIndex) -1)
          (LocalVariables|SetParamValue (Math|Conversions|ToFloat(Integer) (Getters|GetTeamColorIndex (Variables|Team|GetTeamIndex))))
          (else
            (LocalVariables|SetParamValue (Math|Conversions|ToFloat(Integer) (Getters|GetTeamRelationshipColorIndex)))))))
    (:then_1
      (return (LocalVariables|GetParamValue)))))

(fn GetBeaconColor
  (Utilities|FlowControl|Sequence
    (:then_0
      (if (!= (Variables|DeviceOptions|GetBeaconColorIndexKnob) -1)
        (LocalVariables|SetLocalIndex (Variables|DeviceOptions|GetBeaconColorIndexKnob))
        (else
          (LocalVariables|SetLocalIndex (Utilities|Select :Index (Team|GetTeamAffiliationtoLocalPlayer (Math|Conversions|ToByte(Integer) 0) true true) :Friendly (Variables|Beacon|GetFriendlyTeamBeaconColorIndex) :Neutral (Variables|Beacon|GetFriendlyTeamBeaconColorIndex) :Hostile (Variables|Beacon|GetEnemyTeamBeaconColorIndex))))))
    (:then_1
      (if (Utilities|Array|IsValidIndex (Variables|Beacon|GetBeaconColors) (LocalVariables|GetLocalIndex))
        (return (Utilities|Array|Get(acopy) (Variables|Beacon|GetBeaconColors) (LocalVariables|GetLocalIndex)) true)
        (else
          (return "0, 0, 0" false))))))

(fn GetCaptureIconData
  (return (Creative|UI|IconPicker|GetIconLibraryTexture (Variables|DeviceOptions|GetIndicatorPresetIndexKnob)) (Variables|DeviceOptions|GetHealthBarTextSize) (Variables|DeviceOptions|GetHealthBarDistance) (Math|Conversions|ToBoolean(Integer) (Variables|DeviceOptions|GetDisplayDistanceText)) (Variables|DeviceOptions|GetbShowOwningTeamName) (Variables|DeviceOptions|GetbHealthBarRequireLOS) (Math|Conversions|ToBoolean(Integer) (Variables|DeviceOptions|GetShowOffscreenArrow)) (and (or (== (Variables|DeviceOptions|GetHUDElementsIndexKnob) 1) (== (Variables|DeviceOptions|GetHUDElementsIndexKnob) 3)) (Variables|Replicated|GetIsEnabledByPhase)) (Utilities|Array|MakeArray (Variables|DeviceOptions|GetHealthBarOverrideTextFriendly) (Variables|DeviceOptions|GetHealthBarOverrideTextNeutral) (Variables|DeviceOptions|GetHealthBarOverrideTextEnemy))))

(fn GetCollisionStaticMesh
  (return "0, 0, 0" (Variables|/Script/Engine.StaticMeshComponent|GetStaticMesh (Variables|Default|GetBaseMesh))))

(fn GetControllerOfTeamInArea
  (params (TeamID :type Byte))
  (for _array_element (Variables|CaptureArea|GetCaptureAreaTeamsInfo)
    (bind (_array_element_players _array_element_insideteamindex) (Utilities|Struct|BreakCaptureAreaTeamInfo _array_element))
    (if (Math|Byte|Equal(Byte) _array_element_insideteamindex (Variables|GetTeamID))
      (for _array_element_2 _array_element_players
        (Utilities|IsValid _array_element_2
          (:Is%20Valid
            (return _array_element_2))))))
  (return 0))

(fn GetDataFromBlueprint
  (return (Variables|DeviceOptions|GetIndicatorPresetIndexKnob)))

(fn GetFirstCapturablePlayerControllerInArea
  (for _array_element (Variables|CaptureArea|GetCaptureAreaTeamsInfo)
    (bind _array_element_players (Utilities|Struct|BreakCaptureAreaTeamInfo _array_element))
    (for _array_element_2 _array_element_players
      (Utilities|IsValid _array_element_2
        (:Is%20Valid
          (if (Capture|IsValidControllerforCapture _array_element_2)
            (return _array_element_2))))))
  (return 0))

(fn GetFirstCapturableTeamIndexInArea
  (for _array_element (Variables|CaptureArea|GetCaptureAreaTeamsInfo)
    (bind (_array_element_players _array_element_insideteamindex) (Utilities|Struct|BreakCaptureAreaTeamInfo _array_element))
    (for _array_element_2 _array_element_players
      (Utilities|IsValid _array_element_2
        (:Is%20Valid
          (if (Capture|IsValidControllerforCapture _array_element_2)
            (return _array_element_insideteamindex))))))
  (return (Utilities|Enum|LiteralenumEFortTeam "Spectator")))

(fn GetFirstValidTeamIndexForScoringInArea
  (for _array_element (Variables|CaptureArea|GetCaptureAreaTeamsInfo)
    (bind (_array_element_players _array_element_insideteamindex) (Utilities|Struct|BreakCaptureAreaTeamInfo _array_element))
    (for _array_element_2 _array_element_players
      (Utilities|IsValid _array_element_2
        (:Is%20Valid
          (if (PeriodicScoring|IsValidControllerforScoring _array_element_2)
            (return _array_element_insideteamindex)))))
    (return (Utilities|Enum|LiteralenumEFortTeam "Spectator")))
  (return (Utilities|Enum|LiteralenumEFortTeam "Spectator")))

(fn GetFloatFromTeam
  (params (TeamToFloat :type Byte))
  (return (Math|Conversions|ToFloat(Byte) (Math|Byte|Byte-Byte (Variables|GetTeamToFloat) (Utilities|Enum|LiteralenumEFortTeam "HumanPvP_Team1")))))

(fn GetLocalPlayerState
  (params (NetId :type Unique%20Net%20Id%20Structure))
  (bind _array_element (Utilities|Array|ForEachLoopwithBreak (Fortnite|GetLocalFortPlayerControllers true) :!label _label)
    (:LoopBody
      (Utilities|IsValid _array_element
        (:Is%20Valid
          (bind _asfort_player_state (Utilities|Casting|CastToFortPlayerState (Variables|/Script/Engine.Controller|GetPlayerState _array_element)))
          (if (Online|UniqueNetIdRepl|Equal(UniqueIdRepl) (PlayerState|GetUniqueID _asfort_player_state) (Variables|GetNetId))
            (LocalVariables|SetFoundPlayerState _asfort_player_state)
            (goto _label Break)))))
    (:Completed
      (return (LocalVariables|GetFoundPlayerState)))))

(fn GetMeshComponents
  (return (Utilities|Array|MakeArray (Variables|Default|GetBaseMesh))))

(fn GetMeshObjectPathForItem
  (params (ItemDefinition :type Fort%20World%20Item%20Definition%20Object%20Reference))
  (Utilities|FlowControl|Sequence
    (:then_0
      (Utilities|IsValid ItemDefinition
        (:Is%20Valid
          (bind _itemdata (Class|ItemDefinitionBase|GetItemData ItemDefinition)
            (:Found
              (if (Utilities|IsValidSoftObjectReference (Utilities|Struct|BreakFortnitePickupItemData _itemdata))
                (LocalVariables|SetPickupMesh (Utilities|Struct|BreakFortnitePickupItemData _itemdata))
                (elif (Utilities|IsValidSoftObjectReference (.PickupSkeletalMesh (Utilities|Struct|BreakFortnitePickupItemData _itemdata)))
                  (LocalVariables|SetPickupMesh (.PickupSkeletalMesh (Utilities|Struct|BreakFortnitePickupItemData _itemdata))))))))))
    (:then_1
      (return (LocalVariables|GetPickupMesh)))))

(fn GetOrFetchFortPlayspaceEventRouter
  (if (not (Utilities|IsValid (Variables|Default|GetFortPlayspaceEventRouter) :!class /Script/Engine.KismetSystemLibrary))
    (bind _returnvalue (GameplayEventRouter|GetEventRouter "/Script/FortniteGame.FortPlayspace" self))
    (Variables|Default|SetFortPlayspaceEventRouter _returnvalue))
  (return (Variables|Default|GetFortPlayspaceEventRouter)))

(fn GetOverrideMeshMaterial
  (params (MeshComp :type Static%20Mesh%20Component%20Object%20Reference) (MatIdx :type Integer))
  (Utilities|FlowControl|Sequence
    (:then_0
      (if (Utilities|Equal(Object) (Variables|GetMeshComp) (Variables|Default|GetCaptureAreaStaticMesh))
        (bind _captureareamid (Variables|Default|GetCaptureAreaMID)
          (:then
            (return _captureareamid))
          (:else))))
    (:then_1
      (if (Utilities|Equal(Object) (Variables|GetMeshComp) (Variables|Default|GetBaseMesh))
        (bind _captureareabasemid (Variables|Default|GetCaptureAreaBaseMID)
          (:then
            (return _captureareabasemid))
          (:else))))
    (:then_2
      (if (Utilities|Equal(Object) (Variables|GetMeshComp) (Variables|Default|GetPreviewBubble))
        (return (Variables|ItemPreviews|GetPreviewBubbleMaterial))))))

(fn GetSpawnedPickups
  (return 0))

(fn GetTeamColorIndex
  (params (TeamIndexRequested :type Byte))
  (bind _minigamelogic (Variables|Default|GetMinigameLogic))
  (bind _currentminigame (Variables|/Script/FortniteGame.FortMinigameLogicComponent|GetCurrentMinigame _minigamelogic))
  (bind _returnvalue (Minigame|GetMinigameTeamComponent _currentminigame))
  (Utilities|IsValid _currentminigame
    (:Is%20Valid
      (Utilities|IsValid _returnvalue
        (:Is%20Valid
          (return (Minigame|GameRules|GetTeamColorIndex _returnvalue TeamIndexRequested)))
        (:Is%20Not%20Valid
          _)))
    (:Is%20Not%20Valid
      _))
  (return (Math|Float|Truncate (Getters|GetFloatfromTeam TeamIndexRequested))))

(fn GetTeamRelationshipColorIndex
  (bind _returnvalue (Fortnite|GetFirstLocalFortPlayerController true))
  (Utilities|IsValid _returnvalue
    (:Is%20Valid
      (switch Utilities|FlowControl|Switch|SwitchonEFortTeamAffiliation (Team|GetTeamAffiliation _returnvalue self)
        (:Friendly
          (return 1))
        (:Neutral
          (return 0))
        (:Hostile
          (return 2))))))

(fn HandleCaptureAreaOverlap
  (params (OverlappingActor :type Actor%20Object%20Reference))
  (bind _pickupitemlist (Variables|Default|GetPickupItemList))
  (bind _itemstobeaddedqueue (Variables|/Game/Creative/Devices/CaptureArea/PickupItemListComponent.PickupItemListComponent_C|GetItemsToBeAddedQueue _pickupitemlist))
  (bind _output (Utilities|Array|Get(acopy) _itemstobeaddedqueue))
  (bind _pawntoaward (LocalVariables|GetPawnToAward))
  (bind _controllertoaward (LocalVariables|GetControllerToAward))
  (bind _instigator (Variables|Actor|GetInstigator OverlappingActor))
  (Networking|SwitchHasAuthority
    (:Authority
      (Utilities|IsValid OverlappingActor
        (:Is%20Valid
          (if (Getters|IsMinigameinProgress)
            (if (Getters|IsCaptureAreaInteractable)
              (bind _asfort_player_pawn (Utilities|Casting|CastToFortPlayerPawn OverlappingActor)
                (:then
                  (bind _asfort_player_controller (Utilities|Casting|CastToFortPlayerController (Pawn|GetController _asfort_player_pawn)))
                  (CaptureComponent|RegisterPlayerInsideArea _asfort_player_controller)
                  (Bindings|BindorUnbindonPlayerEmote true _asfort_player_pawn))
                (:CastFailed
                  (bind _asfort_pickup_athena (Utilities|Casting|CastToFortPickupAthena OverlappingActor)
                    (:then
                      (Utilities|IsPickupValid _asfort_pickup_athena (Variables|Default|GetInvalidPickupGameplayTagContainer)
                        (:IsValid
                          (Utilities|IsValid (Pickup|GetPawnWhoDroppedPickup _asfort_pickup_athena)
                            (:Is%20Valid
                              (if (ItemDelivery|CanDeliverItems (Team|GetActorTeam (Pickup|GetPawnWhoDroppedPickup _asfort_pickup_athena)))
                                (LocalVariables|SetCapturingPawn (Pickup|GetPawnWhoDroppedPickup _asfort_pickup_athena))
                                (bind _asfort_player_controller_2 (Utilities|Casting|CastToFortPlayerController (Pawn|GetController (Pickup|GetPawnWhoDroppedPickup _asfort_pickup_athena))))
                                (bind (_addedscore _returnvalue) (CaptureArea|CaptureItemsfromPickup (ItemData|GetContainedItemVariants (Variables|Default|GetPickupItemList)) _asfort_pickup_athena))
                                (bind _output_get (LocalVariables|SetTotalItemsCaptured _returnvalue))
                                (if (> _output_get 0)
                                  (GameplayEvent|BroadcastEvent (Variables|UserOptions|Events|GetOnItemIsDelivered) 0 (Utilities|Struct|MakeGameplayEventItemDelivered _asfort_player_controller_2)) ; This transmits regardless of whether the item is consumed or\na score is given. As long as the item and player are valid.
                                  (if (Variables|DeviceOptions|GetConsumeItemWhenDropped)
                                    (PeriodicScoring|SendScoreStatEvents _asfort_player_controller_2 (+ _addedscore (* _returnvalue (Variables|DeviceOptions|GetItemDeliveryScore))))
                                    (CallFunction|FXObjectCaptured (LocalVariables|GetCapturingPawn) _asfort_pickup_athena)
                                    (Actor|DestroyActor (Variables|GetOverlappingActor))
                                    (GameplayEvent|BroadcastEvent (Variables|UserOptions|Events|GetOnItemIsConsumed) 0 (Utilities|Struct|MakeGameplayEventItemConsumed _asfort_player_controller_2))))))))))
                    (:CastFailed
                      (if (CallFunction|CheckActorVelocityVectorToFilter OverlappingActor)
                        ;[0.2758 1 0.5132] Capture Phyisics Object or Object With the Capture Item Tag
                        (bind _returnvalue_2 (GameplayTags|HasMatchingGameplayTag(Message) OverlappingActor (Variables|Default|GetCaptureItemTag)))
                        (if _returnvalue_2
                          (Utilities|IsValid _instigator
                            (:Is%20Valid
                              (bind (_pawn _fort_player_controller) (Getters|GetPawntoAwardScoretofromInstigator (Pawn|GetController _instigator)))
                              (LocalVariables|SetPawnToAward _pawn)
                              (LocalVariables|SetControllerToAward _fort_player_controller)))
                          (else
                            (bind _asfort_physics_pawn_object (Utilities|Casting|CastToFortPhysicsPawnObject OverlappingActor))
                            (bind (_pawn_2 _fort_player_controller_2) (Getters|GetPawntoAwardScoretofromInstigator (Variables|/Script/Engine.Pawn|GetLastHitBy _asfort_physics_pawn_object)))
                            (LocalVariables|SetPawnToAward _pawn_2)
                            (LocalVariables|SetControllerToAward _fort_player_controller_2)))
                        (if (and (Utilities|IsValid _pawntoaward :!class /Script/Engine.KismetSystemLibrary) (Utilities|IsValid _controllertoaward :!class /Script/Engine.KismetSystemLibrary))
                          (PeriodicScoring|SendScoreStatEvents _controllertoaward (Variables|DeviceOptions|GetItemDeliveryScore))
                          (GameplayEvent|BroadcastEvent (Variables|UserOptions|Events|GetOnItemIsDelivered) 0 (Utilities|Struct|MakeGameplayEventItemDelivered _controllertoaward))
                          (CallFunction|FXObjectCaptured _pawntoaward OverlappingActor)
                          (if (Variables|DeviceOptions|GetConsumeItemOnScoring)
                            (Actor|DestroyActor OverlappingActor)
                            (GameplayEvent|BroadcastEvent (Variables|UserOptions|Events|GetOnItemIsConsumed) 0 (Utilities|Struct|MakeGameplayEventItemConsumed _controllertoaward))))))))))
            (else
              (bind _asfort_capturable_item_interface (Utilities|Casting|CastToFortCapturableItemInterface OverlappingActor))
              (if (CapturableItemInterface|CanbeCaptured _asfort_capturable_item_interface (Variables|Default|GetInvalidPickupGameplayTagContainer) self)
                (if (Class|PickupItemListComponent|IsIteminQueue (Variables|Default|GetPickupItemList) OverlappingActor)
                  _
                  (elif (Utilities|Array|IsNotEmpty _itemstobeaddedqueue)
                    (Utilities|IsValid _output
                      (:Is%20Valid
                        (Actor|DestroyActor _output))
                      (:Is%20Not%20Valid
                        _))))
                (Class|PickupItemListComponent|ClearItemList _pickupitemlist)
                (ItemPreviews|AddItemtoQueue OverlappingActor)))))))))

(fn HandleObjectiveAssignment
  (params (NewOwner :type Fort%20Player%20Controller%20Object%20Reference))
  (if (Variables|DeviceOptions|GetbCountAsObjectiveKnob)
    (Utilities|FlowControl|Sequence
      (:then_0
        ;[1 1 1] If previously had an owner, we remove this objective from him
        (bind _currentowner (Variables|Default|GetCurrentOwner)
          (:then
            (Objective|RemoveasObjectivefromMinigame 1 _currentowner))
          (:else)))
      (:then_1
        (Objective|AddasObjectivetoMinigame 1 NewOwner)))))

(fn IgnoreWorldGridScaleOverride
  (return false))

(fn InitinWorldWidget
  (bind _captureiconwidgetcomponent (Variables|InWorldCaptureIcon|GetCaptureIconWidgetComponent))
  (Variables|/Script/FortniteGame.FortWidgetComponent|SetbClampProjectedLocalPositionToViewport (Math|Conversions|ToBoolean(Integer) (Variables|DeviceOptions|GetClampToScreen)) _captureiconwidgetcomponent)
  (Init|InitfromObject(Message) (UserInterface|GetWidget _captureiconwidgetcomponent) self))

(fn InitializeFX)

(fn IsCaptureAreaInteractable
  (return (and (Variables|Replicated|GetIsEnabledByPhase) (and (Variables|Default|GetIsAllowed) (Utilities|Enum|NotEqual(Enum) (CaptureComponent|GetCaptureState (Variables|CaptureArea|GetCaptureComponent)) "Disabled")))))

(fn IsEnabled
  (return false))

(fn IsFriendlyToLocalPlayer
  (params (Pawn :type Fort%20Pawn%20Object%20Reference))
  (Utilities|FlowControl|Sequence
    (:then_0
      (bind _pawn (Variables|GetPawn)
        (:then
          (LocalVariables|SetAffiliation (Team|GetTeamAffiliationtoLocalPlayer (Team|GetActorTeam _pawn))))
        (:else)))
    (:then_1
      (return (Utilities|Select :Index (LocalVariables|GetAffiliation) :Friendly true :Neutral false :Hostile false)))))

(fn IsHoldingItemForPawn
  (params (Item :type Item%20Variant%20Handle%20Structure) (Pawn :type Pawn%20Object%20Reference))
  (return false))

(fn IsLocalTeam
  (params (InTeamByte :type Byte))
  (bind _returnvalue (Fortnite|GetFirstLocalFortPlayerController true))
  (Utilities|IsValid _returnvalue
    (:Is%20Valid
      (return (== InTeamByte (Team|GetActorTeam _returnvalue))))
    (:Is%20Not%20Valid
      (return false))))

(fn IsMinigameInProgress
  (bind _minigamelogic (Variables|Default|GetMinigameLogic))
  (bind _currentminigame (Variables|/Script/FortniteGame.FortMinigameLogicComponent|GetCurrentMinigame _minigamelogic))
  (Utilities|FlowControl|Sequence
    (:then_0
      (Utilities|IsValid _currentminigame
        (:Is%20Valid
          (LocalVariables|SetbSuccess (Minigame|HasStarted _currentminigame :!class /Script/FortniteGame.FortMinigame)))
        (:Is%20Not%20Valid
          (return false))))
    (:then_1
      (return (LocalVariables|GetbSuccess)))))

(fn IsPickupInsideCaptureArea
  (params (PickUpToCheck :type Fort%20Pickup%20Athena%20Object%20Reference))
  (bind (_returnvalue_x _returnvalue_y) (Math|Vector|BreakVector (Transformation|GetWorldLocation (Variables|Default|GetCaptureAreaStaticMesh))))
  (bind _ascapsule_collision (Utilities|Casting|CastToCapsuleComponent (Variables|Transformation|GetRootComponent PickUpToCheck))
    (:then
      (bind (_returnvalue_x_2 _returnvalue_y_2) (Math|Vector|BreakVector (Transformation|GetWorldLocation _ascapsule_collision)))
      (return (> (* 512.0 (Variables|DeviceOptions|GetCaptureAreaSize)) (- (Math|Vector|Distance(Vector) (Math|Vector|MakeVector _returnvalue_x _returnvalue_y) (Math|Vector|MakeVector _returnvalue_x_2 _returnvalue_y_2)) (Variables|/Script/Engine.CapsuleComponent|GetCapsuleRadius _ascapsule_collision)))))
    (:CastFailed
      (return false))))

(fn IsTargetableByLinearMover
  (return true))

(fn IsValidControllerForCapture
  (params (PlayerController :type Fort%20Player%20Controller%20Object%20Reference))
  (return (and (Capture|CanbeCapturedBy (Team|GetActorTeam PlayerController)) (Capture|IsValidItemFilterforCapture PlayerController))))

(fn IsValidControllerForScoring
  (params (PlayerController :type Fort%20Player%20Controller%20Object%20Reference))
  (return (and (PeriodicScoring|CanbeUsedforScoringBy (Team|GetActorTeam PlayerController)) (Math|Boolean|ORBoolean (and (or (Utilities|Enum|Equal(Enum) (Variables|Default|GetbItemsFilterPlayerCapture) "ForPeriodicScoring") (Utilities|Enum|Equal(Enum) (Variables|Default|GetbItemsFilterPlayerCapture) "Both")) (Getters|DoesPlayerHaveCaptureItem PlayerController)) (Utilities|Enum|Equal(Enum) (Variables|Default|GetbItemsFilterPlayerCapture) "None") (Utilities|Enum|Equal(Enum) (Variables|Default|GetbItemsFilterPlayerCapture) "ToTakeControl")))))

(fn IsValidItemFilterForCapture
  (params (PlayerControllerReference :type Fort%20Player%20Controller%20Object%20Reference))
  (return (Math|Boolean|ORBoolean (and (or (Utilities|Enum|Equal(Enum) (Variables|Default|GetbItemsFilterPlayerCapture) "ToTakeControl") (Utilities|Enum|Equal(Enum) (Variables|Default|GetbItemsFilterPlayerCapture) "Both")) (Getters|DoesPlayerHaveCaptureItem PlayerControllerReference)) (Utilities|Enum|Equal(Enum) (Variables|Default|GetbItemsFilterPlayerCapture) "None") (Utilities|Enum|Equal(Enum) (Variables|Default|GetbItemsFilterPlayerCapture) "ForPeriodicScoring"))))

(fn MapMarkerInit
  (MapMarker|MapMarkerOwningTeamChanged (Variables|/Script/FortniteGame.CaptureComponent|GetOwningTeam (Variables|CaptureArea|GetCaptureComponent)) (Variables|/Script/FortniteGame.CaptureComponent|GetOwningTeam (Variables|CaptureArea|GetCaptureComponent)))
  (MapMarker|MapMarkerUpdateProgress))

(fn MapMarkerOwningTeamChanged
  (params (NewTeam :type Byte) (OldTeam :type Byte))
  (bind _output_get (LocalVariables|SetNewCoreGamerColors (Class|PDAUICreativeCoreGamer|GetOwnershipTeamColor (Variables|Default|GetPrimary%20Data%20Asset) (Minigame|GetMinigame self :!class /Script/FortniteGame.MinigameBlueprintLibrary) (Fortnite|GetFirstLocalFortPlayerController true) NewTeam)))
  (Utilities|FlowControl|Sequence
    (:then_1
      (Class|CreativeMiniMapComponent|ApplyVectorOverridestoMaterial (Variables|Default|GetCreativeMapMarkerComponent) "Inner Box Color 1" (.Dark_23_D4DB58604621CF24D0ED08939E74FF74 (Utilities|Struct|BreakSUICreativeCoreGamerTeamColors _output_get))))
    (:then_2
      (Class|CreativeMiniMapComponent|ApplyVectorOverridestoMaterial (Variables|Default|GetCreativeMapMarkerComponent) "Inner Box Color 2" (.Darker_24_DB140E20492081760F7E2281F4D17B74 (Utilities|Struct|BreakSUICreativeCoreGamerTeamColors _output_get))))
    (:then_3
      (Class|CreativeMiniMapComponent|ApplyScalarOverridestoMaterial (Variables|Default|GetCreativeMapMarkerComponent) "Rotation Angle" (Math|Float|SelectFloat 0.0 0.125 (MapMarker|IsLocalTeam (Variables|GetNewTeam)))))
    (:then_4
      (Class|CreativeMiniMapComponent|ApplyScalarOverridestoMaterial (Variables|Default|GetCreativeMapMarkerComponent) "Start Point" (Math|Float|SelectFloat 1.0 0.875 (MapMarker|IsLocalTeam (Variables|GetNewTeam)))))))

(fn MapMarkerSetup
  (bind _outputdelegate (EventDispatchers|CreateEvent :!event Map%20Marker%20Owning%20Team%20Changed))
  (bind _outputdelegate_2 (EventDispatchers|CreateEvent :!event OnMapMarkerIconDataReplicated))
  (bind _capturecomponent (Variables|CaptureArea|GetCaptureComponent))
  (bind _creativemapmarkercomponent (Variables|Default|GetCreativeMapMarkerComponent))
  (Networking|SwitchHasAuthority
    (:Authority
      (Class|CreativeMapMarkerComponent|SetupMapMarker :self (Variables|Default|GetCreativeMapMarkerComponent) :Icon%20Data%20Icon%20Texture (Creative|UI|IconPicker|GetIconLibraryTextureWithEmptyTexture (Variables|DeviceOptions|GetIndicatorPresetIndexKnob)) :Icon%20Data%20Icon%20Color "(R=0.352941,G=0.352941,B=0.352941,A=1.000000)" :VisibilityOnMap "NewEnumerator1" :UseTeamFilter false :IsMarkerVisible false :IndicatorTextColor "(R=1.000000,G=1.000000,B=1.000000,A=1.000000)" :IconTextureParamName "Icon Texture" :Icon%20Scale 1.0)
      (MapMarker|MapMarkerSetVisibility true))
    (:Remote
      (MapMarker|MapMarkerInit)
      (if (Variables|DeviceOptions|GetShowMapMarker)
        (EventDispatchers|UnbindEventfromOnOwningTeamChanged _capturecomponent _outputdelegate)
        (EventDispatchers|BindEventtoOnOwningTeamChanged _capturecomponent _outputdelegate)
        (EventDispatchers|UnbindEventfromOnIconDataReplicatedEvent _creativemapmarkercomponent _outputdelegate_2)
        (EventDispatchers|BindEventtoOnIconDataReplicatedEvent _creativemapmarkercomponent _outputdelegate_2)
        (else
          (EventDispatchers|UnbindEventfromOnOwningTeamChanged _capturecomponent _outputdelegate)
          (EventDispatchers|UnbindEventfromOnIconDataReplicatedEvent _creativemapmarkercomponent _outputdelegate_2))))))

(fn MapMarkerUpdateProgress
  (bind _newprogresscolors (LocalVariables|GetNewProgressColors))
  (bind (_lighter_20_2cec1cf44a4c9b981667c4a540644467 _light_21_721ebfad498e3c43dbfe548837331bf8 _medium_22_3780b0024b6caf8e6cb9918162a449f7) (Utilities|Struct|BreakSUICreativeCoreGamerTeamColors _newprogresscolors))
  (Class|CreativeMiniMapComponent|ApplyScalarOverridestoMaterial (Variables|Default|GetCreativeMapMarkerComponent) "Percentage" (Variables|/Script/FortniteGame.CaptureComponent|GetCaptureProgress (Variables|CaptureArea|GetCaptureComponent)))
  (Utilities|FlowControl|Sequence
    (:then_0
      (LocalVariables|SetNewProgressColors (Class|PDAUICreativeCoreGamer|GetProgressTeamColor (Variables|Default|GetPrimary%20Data%20Asset) (Minigame|GetMinigame self :!class /Script/FortniteGame.MinigameBlueprintLibrary) (Fortnite|GetFirstLocalFortPlayerController true) (Variables|/Script/FortniteGame.CaptureComponent|GetOwningTeam (Variables|CaptureArea|GetCaptureComponent)) (Variables|/Script/FortniteGame.CaptureComponent|GetStateOwner (Variables|CaptureArea|GetCaptureComponent)) (CaptureComponent|GetCaptureState (Variables|CaptureArea|GetCaptureComponent)))))
    (:then_1
      (Class|CreativeMiniMapComponent|ApplyVectorOverridestoMaterial (Variables|Default|GetCreativeMapMarkerComponent) "Progress Color 1" _lighter_20_2cec1cf44a4c9b981667c4a540644467)
      (Class|CreativeMiniMapComponent|ApplyVectorOverridestoMaterial (Variables|Default|GetCreativeMapMarkerComponent) "Progress Color 2" _medium_22_3780b0024b6caf8e6cb9918162a449f7))))

(fn MapMarkerSetVisibility
  (params (ShouldBeVisible :type Boolean))
  (MiniMapData|SetMiniMapIndicatorIsVisible (Variables|Default|GetCreativeMapMarkerComponent) (and ShouldBeVisible (Variables|DeviceOptions|GetShowMapMarker))))

(fn MigrateSameChannelInstances
  (bind _bwasmigrated (UEFNConversionHelper|MigratetoReceiverwhenSameChannel (Variables|Default|GetEnableReceivingFrom) (Variables|Default|GetDisableReceivingFrom) (Variables|Default|GetToggleEnabledReceivingFrom)))
  (bind _output_get (LocalVariables|SetbMigrated _bwasmigrated))
  (bind _bwasmigrated_2 (UEFNConversionHelper|MigratetoReceiverwhenSameChannel (Variables|Default|GetUnlockReceivingFrom) (Variables|Default|GetLockReceivingFrom) (Variables|Default|GetToggleCaptureAllowedReceivingFrom)))
  (return (or _bwasmigrated_2 _output_get)))

(fn OnCaptureContested
  (Networking|CanExecuteCosmeticEvents
    (:True
      (CallFunction|PlayCaptureStateSound (Variables|Audio|GetNeutralizedSound)))
    (:False
      (GameplayEvent|BroadcastEvent (Variables|UserOptions|Events|GetOnAreaIsContested) 0 (Utilities|Struct|MakeGameplayEventCaptureAreaContested (Getters|GetControllerOfTeaminArea (Variables|/Script/FortniteGame.CaptureComponent|GetStateOwner (Variables|CaptureArea|GetCaptureComponent))))))))

(fn OnCaptured
  (bind _foundcontroller (LocalVariables|GetFoundController))
  (bind _returnvalue (Team|GetActorTeam _foundcontroller))
  (Networking|CanExecuteCosmeticEvents
    (:True
      (Rendering|Material|SetScalarParameterValue (Variables|Default|GetCaptureAreaMID) (Variables|MIDParameters|GetMIDParameterCaptureEffectOpacity) 0.0 :!class /Script/Engine.MaterialInstanceDynamic)
      (Rendering|Material|SetScalarParameterValue (Variables|Default|GetCaptureAreaMID) (Variables|MIDParameters|GetMIDParameterCaptureEffectDirection) 0.0 :!class /Script/Engine.MaterialInstanceDynamic)
      (Rendering|Material|SetScalarParameterValue (Variables|Default|GetCaptureAreaMID) (Variables|MIDParameters|GetMIDParameterCaptureProgress) 0.0 :!class /Script/Engine.MaterialInstanceDynamic)
      (Device|UpdateDeviceColors)
      (CallFunction|PlayCaptureStateSound (Variables|Audio|GetCapturedSound)))
    (:False
      (Utilities|FlowControl|Sequence
        (:then_0
          (if (Variables|Default|GetWasCaptureForced)
            ;[1 1 1] When Capture was forced we have a valid ForcedCaptureStateController
            (bind _forcedcapturestatecontroller (Variables|Default|GetForcedCaptureStateController)
              (:then
                (LocalVariables|SetFoundController _forcedcapturestatecontroller)
                (Variables|Default|SetWasCaptureForced false)
                (Variables|Default|SetForcedCaptureStateController 0))
              (:else))
            (else
              ;[1 1 1] When Capture was not forced we get a valid controller on point
              (LocalVariables|SetFoundController (Capture|GetFirstCapturablePlayerControllerinArea)))))
        (:then_1
          (Utilities|IsValid (LocalVariables|GetFoundController)
            (:Is%20Valid
              (if (Math|Byte|NotEqual(Byte) (Variables|Team|GetTeamIndex) _returnvalue)
                (GameplayEvent|BroadcastEvent (Variables|UserOptions|Events|GetOnControlChange) 0 (Utilities|Struct|MakeGameplayEventCaptureAreaControlChanged (LocalVariables|GetFoundController)))
                (PeriodicScoring|SendScoreStatEvents (LocalVariables|GetFoundController) (Variables|DeviceOptions|GetScoreOnControl))
                (CallFunction|SetTeam _returnvalue)
                (CallFunction|OnCaptureAreaOwnerChanged (LocalVariables|GetFoundController))
                (Networking|FlushNetDormancy)))))
        (:then_2
          (Capture|TrytoUpdateCaptureState))))))

(fn OnCapturingStarted
  (Networking|CanExecuteCosmeticEvents
    (:True
      (Rendering|Material|SetScalarParameterValue (Variables|Default|GetCaptureAreaMID) (Variables|MIDParameters|GetMIDParameterCaptureEffectOpacity) 1.0 :!class /Script/Engine.MaterialInstanceDynamic)
      (Rendering|Material|SetScalarParameterValue (Variables|Default|GetCaptureAreaMID) (Variables|MIDParameters|GetMIDParameterCaptureEffectDirection) 0.0 :!class /Script/Engine.MaterialInstanceDynamic)
      (CallFunction|UpdateInteractingColor)
      (CallFunction|PlayCaptureAudioLoop (Variables|Audio|GetCapturingLoopSound)))
    (:False
      (if (or (Utilities|Enum|Equal(Enum) (Variables|/Script/FortniteGame.CaptureComponent|GetPreviousCaptureState (Variables|CaptureArea|GetCaptureComponent)) "Neutral") (Utilities|Enum|Equal(Enum) (Variables|/Script/FortniteGame.CaptureComponent|GetPreviousCaptureState (Variables|CaptureArea|GetCaptureComponent)) "Captured"))
        (GameplayEvent|BroadcastEvent (Variables|UserOptions|Events|GetOnControlChangeStarts) 0 (Utilities|Struct|MakeGameplayEventCaptureAreaControlChangeStarted (Capture|GetFirstCapturablePlayerControllerinArea)))))))

(fn OnDecapturing
  (Networking|CanExecuteCosmeticEvents
    (:True
      (Rendering|Material|SetScalarParameterValue (Variables|Default|GetCaptureAreaMID) (Variables|MIDParameters|GetMIDParameterCaptureEffectOpacity) 1.0 :!class /Script/Engine.MaterialInstanceDynamic)
      (Rendering|Material|SetScalarParameterValue (Variables|Default|GetCaptureAreaMID) (Variables|MIDParameters|GetMIDParameterCaptureEffectDirection) 0.0 :!class /Script/Engine.MaterialInstanceDynamic)
      (CallFunction|PlayCaptureAudioLoop (Variables|Audio|GetNeutralizingLoopSound)))))

(fn OnDeneutralizing
  (Networking|CanExecuteCosmeticEvents
    (:True
      (Rendering|Material|SetScalarParameterValue (Variables|Default|GetCaptureAreaMID) (Variables|MIDParameters|GetMIDParameterInteractingTeam) 0.0 :!class /Script/Engine.MaterialInstanceDynamic)
      (Rendering|Material|SetScalarParameterValue (Variables|Default|GetCaptureAreaMID) (Variables|MIDParameters|GetMIDParameterCaptureEffectOpacity) 1.0 :!class /Script/Engine.MaterialInstanceDynamic)
      (Rendering|Material|SetScalarParameterValue (Variables|Default|GetCaptureAreaMID) (Variables|MIDParameters|GetMIDParameterCaptureEffectDirection) 1.0 :!class /Script/Engine.MaterialInstanceDynamic)
      (Rendering|Material|SetScalarParameterValue (Variables|Default|GetCaptureAreaMID) (Variables|MIDParameters|GetMIDParameterTeamOwner) (Getters|GetAccentColorParam) :!class /Script/Engine.MaterialInstanceDynamic)
      (CallFunction|PlayCaptureAudioLoop (Variables|Audio|GetCapturingLoopSound)))))

(fn OnMapMarkerIconDataReplicated
  (MapMarker|MapMarkerInit))

(fn OnNeutral
  (bind _capturecomponent (Variables|CaptureArea|GetCaptureComponent))
  (Networking|CanExecuteCosmeticEvents
    (:True
      (Rendering|Material|SetScalarParameterValue (Variables|Default|GetCaptureAreaMID) (Variables|MIDParameters|GetMIDParameterCaptureEffectOpacity) 0.0 :!class /Script/Engine.MaterialInstanceDynamic)
      (Rendering|Material|SetScalarParameterValue (Variables|Default|GetCaptureAreaMID) (Variables|MIDParameters|GetMIDParameterCaptureEffectDirection) 0.0 :!class /Script/Engine.MaterialInstanceDynamic)
      (CallFunction|PlayCaptureStateSound (Variables|Audio|GetNeutralizedSound)))
    (:False
      (if (not (Variables|Default|GetForceNeutralization))
        (if (and (== (Variables|DeviceOptions|GetNeutralizeTIme) -1) (Math|Byte|NotEqual(Byte) (Variables|Team|GetTeamIndex) (Utilities|Enum|LiteralenumEFortTeam "Spectator")))
          (Capture|TrytoUpdateCaptureState)
          (if (Utilities|Enum|NotEqual(Enum) (CaptureComponent|GetCaptureState (Variables|CaptureArea|GetCaptureComponent)) "Capturing")
            (CaptureComponent|ForceCaptureProgress _capturecomponent 1.0)
            (CaptureComponent|SetCaptureState _capturecomponent "Captured" (Variables|Team|GetTeamIndex)))
          (else
            _))
        (else
          (Variables|Default|SetForceNeutralization false)))
      (Utilities|FlowControl|Sequence
        (:then_0
          (if (Utilities|Enum|NotEqual(Enum) (Utilities|Enum|BytetoEnumEFortTeam (Variables|Team|GetTeamIndex)) "Spectator")
            (CallFunction|SetTeam (Utilities|Enum|LiteralenumEFortTeam "Spectator"))
            (CallFunction|OnCaptureAreaOwnerChanged 0)
            (Networking|FlushNetDormancy)))
        (:then_1
          (Capture|TrytoUpdateCaptureState)
          (ItemDelivery|TrytoUpdatePeriodicScoring))))))

(fn OnNeutralizingStarted
  (Networking|CanExecuteCosmeticEvents
    (:True
      (Rendering|Material|SetScalarParameterValue (Variables|Default|GetCaptureAreaMID) (Variables|MIDParameters|GetMIDParameterInteractingTeam) 0.0 :!class /Script/Engine.MaterialInstanceDynamic)
      (Rendering|Material|SetScalarParameterValue (Variables|Default|GetCaptureAreaMID) (Variables|MIDParameters|GetMIDParameterCaptureEffectOpacity) 1.0 :!class /Script/Engine.MaterialInstanceDynamic)
      (Rendering|Material|SetScalarParameterValue (Variables|Default|GetCaptureAreaMID) (Variables|MIDParameters|GetMIDParameterCaptureEffectDirection) 1.0 :!class /Script/Engine.MaterialInstanceDynamic)
      (CallFunction|PlayCaptureAudioLoop (Variables|Audio|GetNeutralizingLoopSound)))))

(fn OnPeriodicScoringTimerFinished
  (if (Utilities|Enum|NotEqual(Enum) (Variables|DeviceOptions|GetPlayerCapture) "OwningTeam")
    (for _array_element (Variables|CaptureArea|GetCaptureAreaTeamsInfo)
      (bind (_array_element_players _array_element_insideteamindex) (Utilities|Struct|BreakCaptureAreaTeamInfo _array_element))
      (if (PeriodicScoring|CanbeUsedforScoringBy _array_element_insideteamindex)
        (for _array_element_2 _array_element_players
          (if (PeriodicScoring|IsValidControllerforScoring _array_element_2)
            (if (or (Utilities|Enum|Equal(Enum) (Variables|DeviceOptions|GetPlayerCapture) "OnePlayerPerTeam") (Utilities|Enum|Equal(Enum) (Variables|DeviceOptions|GetPlayerCapture) "OwningTeam"))
              (if (not (Utilities|Set|ContainsItem (LocalVariables|GetTeamsCaptured) (Utilities|Enum|BytetoEnumEFortTeam (Team|GetActorTeam _array_element_2))))
                (Utilities|Set|Add (LocalVariables|GetTeamsCaptured) (Utilities|Enum|BytetoEnumEFortTeam (Team|GetActorTeam _array_element_2)))))
            (PeriodicScoring|SendScoreStatEvents _array_element_2 (Variables|DeviceOptions|GetScorePerCapture))
            (LocalVariables|SetScored true)))))
    (if (LocalVariables|GetScored)
      (CallFunction|FXMulticastPlayerScored))
    (else
      (bind _fortplayercontroller (CallFunction|TrytoGetValidControllerfromTeam (Variables|Team|GetTeamIndex)))
      (Utilities|IsValid _fortplayercontroller
        (:Is%20Valid
          (PeriodicScoring|SendScoreStatEvents _fortplayercontroller (Variables|DeviceOptions|GetScorePerCapture))
          (CallFunction|FXMulticastPlayerScored))))))

(fn OnPlayerWithinZoneSwitchedTeams
  (bind _returnvalue (CallFunction|GetorFetchFortPlayspaceEventRouter))
  (GameplayEventRouter|BroadcastEvent _returnvalue (Utilities|Struct|MakeGERPlayerSwitchedTeamWhileinZone false) self))

(fn OnRep_IsAllowed
  (if (Variables|Default|GetIsAllowed)
    (IsAllowed|AllowInteraction)
    (else
      (IsAllowed|DisallowInteraction))))

(fn OnRep_IsEnabledByPhase
  (Networking|CanExecuteCosmeticEvents
    (:True
      (CallFunction|FXHandleEnable)
      (if (not (Variables|Replicated|GetIsEnabledByPhase))
        (CallFunction|StopCaptureAudio)))
    (:False
      (if (Variables|Replicated|GetIsEnabledByPhase)
        ;[1 1 1] Enable Capture, Update Overlapping Actors
        ; Try to: Start Tracking Objective and Add Objective Value
        (Utilities|FlowControl|Sequence
          (:then_0
            (Capture|EnableCaptureComponent))
          (:then_1
            (ItemDelivery|TrytoUpdatePeriodicScoring))
          (:then_2
            (if (Getters|IsCaptureAreaInteractable)
              (EnableKnob|UpdateOverlappingActors)))
          (:then_3
            (Objective|UpdateObjectiveTrackingState (Variables|/Script/FortniteGame.FortMinigameLogicComponent|GetCurrentMinigame (Variables|Default|GetFortMinigameObjective)))))
        (else
          (Device|ResetDeviceCachedData true)
          (Objective|RemoveObjective)
          (CaptureComponent|SetCaptureState (Variables|CaptureArea|GetCaptureComponent) "Disabled" (Variables|Team|GetTeamIndex)))))))

(fn RefreshPreviewRotation
  (bind _rotatemesh (Variables|BP_Creative_CaptureArea|GetRotateMesh))
  (bind _returnvalue (Components|Timeline|IsPlaying _rotatemesh))
  (if (Variables|Default|GetDoNotRenderBubble)
    (if _returnvalue
      (CallFunction|FXStopItemPreviewRotation))
    (elif (not _returnvalue)
      (CallFunction|FXStartItemPreviewRotation))))

(fn RemoveEmoteRelatedBindingfromPlayers
  (for _array_element (Variables|CaptureArea|GetCaptureAreaTeamsInfo)
    (bind _array_element_players (Utilities|Struct|BreakCaptureAreaTeamInfo _array_element))
    (for _array_element_2 _array_element_players
      (Utilities|IsValid _array_element_2
        (:Is%20Valid
          (Utilities|IsValid (Game|Player|GetPlayerPawn _array_element_2)
            (:Is%20Valid
              (Bindings|BindorUnbindonPlayerEmote false (Game|Player|GetPlayerPawn _array_element_2)))))))))

(fn RemoveAsObjectiveFromMinigame
  (params (Amount :type Integer :default 1) (FortPlayerController :type Fort%20Player%20Controller%20Object%20Reference))
  (LocalVariables|Setlocal%20FortPlayerController FortPlayerController)
  (bind _local_fortplayercontroller (LocalVariables|Getlocal%20FortPlayerController)
    (:then
      (bind _returnvalue (CreativeDevices|GetMinigameStatsComp _local_fortplayercontroller))
      (LocalVariables|Setlocal%20MinigameStatsComp _returnvalue)
      (bind _local_minigamestatscomp (LocalVariables|Getlocal%20MinigameStatsComp)
        (:then
          (Class|FortMinigameStatsComponent|SendStatEventwithTags _local_minigamestatscomp (LocalVariables|Getlocal%20FortPlayerController) "MinigameDynamicEvent" self (GameplayTags|MakeGameplayTagContainerfromTag "(TagName=\"Creative.Item.Destructible\")") (Class|FortMinigameStatsComponent|GetSourceandContextTags _local_minigamestatscomp (LocalVariables|Getlocal%20FortPlayerController)) (.OutContextTags (Class|FortMinigameStatsComponent|GetSourceandContextTags _local_minigamestatscomp (LocalVariables|Getlocal%20FortPlayerController))) (* Amount -1)))
        (:else)))
    (:else)))

(fn RemoveObjective
  (Utilities|FlowControl|Sequence
    (:then_0
      (if (Variables|DeviceOptions|GetbCountAsObjectiveKnob)
        (bind _currentowner (Variables|Default|GetCurrentOwner)
          (:then
            (Objective|RemoveasObjectivefromMinigame 1 _currentowner)
            (Variables|Default|SetCurrentOwner 0))
          (:else))))
    (:then_1
      (if (and (Variables|DeviceOptions|GetbTrackProgressKnob) (Variables|DeviceOptions|GetbCountAsObjectiveKnob))
        (Minigame|Objective|StopTracking (Variables|Default|GetFortMinigameObjective) true 1.0)))))

(fn ResetDeviceCachedData
  (params (StopPeriodicScoringTimer :type Boolean :default true))
  (Networking|CanExecuteCosmeticEvents
    (:False
      (CaptureArea|UnbindAllControllerstoInventoryChangedEvent)
      (CallFunction|RemoveEmoteRelatedBindingFromPlayers)
      (CaptureArea|UnregisterAllRegisteredPlayerInsideArea)
      (Utilities|Array|Clear (Variables|CaptureArea|GetCaptureAreaTeamsInfo))
      (if (Variables|GetStopPeriodicScoringTimer)
        (PeriodicScoring|StopPeriodicScoringTimer)))))

(fn ResetMeshExceptions
  (Variables|Default|SetDoNotRenderBubble false))

(fn SendScoreStatEvents
  (params (PlayerController :type Fort%20Player%20Controller%20Object%20Reference) (Score :type Integer))
  (bind _playercontroller (Variables|GetPlayerController))
  (Utilities|IsValid PlayerController
    (:Is%20Valid
      (if (> Score 0)
        (LocalVariables|SetTargetTags (GameplayTags|GetOwnedGameplayTags self :!class /Script/GameplayTags.GameplayTagAssetInterface))
        (Utilities|FlowControl|Sequence
          (:then_0
            (if (GameplayTags|IsGameplayTagValid (Variables|Constants|GetCaptureEventTag))
              (GameplayTags|AddGameplayTag (LocalVariables|GetTargetTags) (Variables|Constants|GetCaptureEventTag) :!class /Script/GameplayTags.BlueprintGameplayTagLibrary)))
          (:then_1
            (bind _returnvalue (CreativeDevices|GetMinigameStatsComp _playercontroller "Athena"))
            (LocalVariables|Setlocal%20MinigameStatsComp _returnvalue)
            (bind _local_minigamestatscomp (LocalVariables|Getlocal%20MinigameStatsComp)
              (:then
                (Class|FortMinigameStatsComponent|SendStatEventwithTags _local_minigamestatscomp _playercontroller "MinigameDynamicEvent" self (LocalVariables|GetTargetTags) (Class|FortMinigameStatsComponent|GetSourceandContextTags _local_minigamestatscomp _playercontroller) (.OutContextTags (Class|FortMinigameStatsComponent|GetSourceandContextTags _local_minigamestatscomp _playercontroller)) (Variables|GetScore)))
              (:else)))
          (:then_2
            (GameplayEvent|BroadcastEvent (Variables|UserOptions|Events|GetOnAreaIsScored) 0 (Utilities|Struct|MakeGameplayEventCaptureAreaScore (Variables|GetPlayerController))))
          (:then_3
            (bind (_addedscore _returnvalue_2) (CaptureArea|CaptureItemsfromPlayer (Variables|DeviceOptions|GetConsumeItemOnScoring) (ItemData|GetContainedItemVariants (Variables|Default|GetPickupItemList)) (Variables|GetPlayerController) 1))
            (if (> _returnvalue_2 0)
              (GameplayEvent|BroadcastEvent (Variables|UserOptions|Events|GetOnItemIsConsumed) 0 (Utilities|Struct|MakeGameplayEventItemConsumed (Variables|GetPlayerController))))))))))

(fn SetTriggerCollision
  (params (On :type Boolean))
  (return false))

(fn SetBeaconColor
  (params (BeaconParticleObject :type Cascade%20Particle%20System%20Component%20Object%20Reference%20%28by%20ref%29) (BeaconParticlesColor :type Array%20of%20FDestruction%20Object%20Particle%20Color%20Parameter%20Structures))
  (Utilities|IsValid BeaconParticleObject
    (:Is%20Valid
      (bind (_output _bisvalid) (Beacon|GetBeaconColor))
      (LocalVariables|SetNewBeaconColor _output)
      (if _bisvalid
        (for _array_element BeaconParticlesColor
          (Effects|Components|ParticleSystem|SetVectorParameter (Variables|GetBeaconParticleObject) (Utilities|Struct|BreakFDestructionObjectParticleColorParameter _array_element) (* (LocalVariables|GetNewBeaconColor) (.ColorParameterMultiplier_7_AE2CF29447670F4CE0F86BA17A6BB6D6 (Utilities|Struct|BreakFDestructionObjectParticleColorParameter _array_element)))))))))

(fn SetBeaconRelativeLocation
  (Transformation|SetRelativeLocation (Variables|Default|GetBeaconParticle) (Math|Transform|InverseTransformLocation (Transformation|GetWorldTransform (Variables|Default|GetBaseMesh)) (+ (Transformation|GetSocketLocation (Variables|Default|GetBaseMesh) (Variables|Beacon|GetFxBeaconSocketName)) (* (Transformation|GetUpVector (Variables|Default|GetBaseMesh)) 400))) false false))

(fn SetBeaconScale
  (params (BeaconParticleObject :type Cascade%20Particle%20System%20Component%20Object%20Reference%20%28by%20ref%29))
  (bind _beaconscale (Variables|DeviceOptions|GetBeaconScale))
  (Utilities|IsValid BeaconParticleObject
    (:Is%20Valid
      (if (> _beaconscale 0.0)
        (Transformation|SetWorldScale3D (Variables|GetBeaconParticleObject) (Math|Conversions|ToVector(Float) _beaconscale))
        (else
          (Transformation|SetRelativeScale3D (Variables|GetBeaconParticleObject) "1.000000,1.000000,1.000000"))))))

(fn SetBubbleScale
  (bind _previewbubble (Variables|Default|GetPreviewBubble))
  (if (or (not (Variables|Replicated|GetIsEnabledByPhase)) (Variables|Default|GetDoNotRenderBubble))
    (Development|SetHiddeninGame _previewbubble true false)
    (else
      (Development|SetHiddeninGame _previewbubble false false)
      (Transformation|SetRelativeScale3D _previewbubble (Math|Conversions|ToVector(Float) (+ (/ (Variables|ItemPreviews|GetComponentBounds) 64.0) (Variables|Constants|GetBubbleScaleBias)))))))

(fn SetCaptureStateByCurrentTeam
  (params (ForceProgress :type Boolean))
  (Utilities|FlowControl|Sequence
    (:then_0
      (if (Variables|GetForceProgress)
        (CaptureComponent|ForceCaptureProgress (Variables|CaptureArea|GetCaptureComponent) (Math|Float|SelectFloat 0.0 1.0 (Utilities|Enum|Equal(Enum) (Utilities|Enum|BytetoEnumEFortTeam (Variables|Team|GetTeamIndex)) "Spectator")))))
    (:then_1
      (if (Utilities|Enum|Equal(Enum) (Utilities|Enum|BytetoEnumEFortTeam (Variables|Team|GetTeamIndex)) "Spectator")
        (CaptureComponent|SetCaptureState (Variables|CaptureArea|GetCaptureComponent) "Neutral" (Variables|Team|GetTeamIndex))
        (else
          (Capture|ForceCapture 0 (Variables|Team|GetTeamIndex)))))))

(fn SetCaptureStateOnStart
  (params (ForceProgress :type Boolean))
  (Networking|CanExecuteCosmeticEvents
    (:False
      (Capture|SetCaptureStatebyCurrentTeam (Variables|GetForceProgress)))))

(fn SetEnableByMinigameState
  (params (MinigameState :type EFortMinigameState%20Enum :default "PreGame"))
  (bind _self self)
  (bind _enabledindex (Variables|DeviceOptions|GetEnabledIndex))
  (bind _minigamestate (Variables|GetMinigameState))
  (bind _enabled (Class|CreativeCommonDeviceFunctionLibrary|IsEnabledPhaseMatch _minigamestate _enabledindex _self))
  (Networking|SwitchHasAuthority
    (:Authority
      (if (Math|Boolean|NotEqual(Boolean) (Variables|Replicated|GetIsEnabledByPhase) _enabled)
        (Variables|Replicated|SetIsEnabledByPhase _enabled)))))

(fn SetFlagTeamColors
  (params (Static%20Mesh :type Static%20Mesh%20Object%20Reference))
  (if (Utilities|Equal(SoftObjectReference) (Utilities|ResolveSoftReference Static%20Mesh) (LocalVariables|GetFlagStaticMesh))
    (bind _returnvalue (Rendering|Material|CreateDynamicMaterialInstance (Variables|Default|GetPreviewStaticMesh) 0 0 "None" :!class /Script/Engine.PrimitiveComponent))
    (Rendering|Material|SetScalarParameterValue _returnvalue "TeamColor" (Getters|GetAccentColorParam) :!class /Script/Engine.MaterialInstanceDynamic)))

(fn SetIsAllowedOnPregame
  (params (FortMinigameState :type EFortMinigameState%20Enum :default "PreGame"))
  (if (Utilities|Enum|Equal(Enum) FortMinigameState "PreGame")
    (IsAllowed|UpdateAllowedInteractionSetting)))

(fn SetOwnerByMinigameState
  (params (NewMinigameState :type EFortMinigameState%20Enum :default "PreGame"))
  (if (Utilities|Enum|Equal(Enum) NewMinigameState "PreGame")
    (Networking|CanExecuteCosmeticEvents
      (:False
        (CallFunction|SetStartOwnerTeam)
        (CaptureComponent|ForceCaptureProgress (Variables|CaptureArea|GetCaptureComponent) 1.0)))))

(fn SetStartOwnerTeam
  (if (or (and (Math|Byte|NotEqual(Byte) (Utilities|Enum|LiteralenumEFortTeam "Spectator") (Variables|Team|GetTeamIndex)) (Knobs|Team|IsTeamAll (Variables|DeviceOptions|GetTeamSetting))) (and (Knobs|Team|IsValidTeamIndex (Variables|DeviceOptions|GetTeamSetting)) (Math|Byte|NotEqual(Byte) (Teams|IndextoTeam (Knobs|Team|GetTeamValueasIndex (Variables|DeviceOptions|GetTeamSetting))) (Variables|Team|GetTeamIndex))))
    (Networking|CanExecuteCosmeticEvents
      (:False
        (if (Knobs|Team|IsTeamAll (Variables|DeviceOptions|GetTeamSetting))
          (CallFunction|SetTeam (Utilities|Enum|LiteralenumEFortTeam "Spectator"))
          (else
            (CallFunction|SetTeam (Teams|IndextoTeam (Knobs|Team|GetTeamValueasIndex (Variables|DeviceOptions|GetTeamSetting))))))
        (CallFunction|OnCaptureAreaOwnerChanged 0)
        (Networking|FlushNetDormancy)))))

(fn SetupUIForCaptureState
  (params (NewState :type ECaptureState%20Enum :default "Neutral") (PreviousState :type ECaptureState%20Enum :default "Neutral") (StateOwnerTeam :type Byte) (ForceUpdate :type Boolean))
  (Networking|CanExecuteCosmeticEvents
    (:False
      (Minigame|Objective|SetTrackedState (Variables|Default|GetFortMinigameObjective) (Variables|GetNewState) (Variables|GetStateOwnerTeam)))))

(fn SetupVisibilityComponent
  (bind _visibleingamecomponent (Variables|Default|GetVisibleInGameComponent))
  ;[0.3406 0.3241 1] Add components we may want hidden during gameplay to the visibilty component.
  (Visibility|SetVisibilityComponents _visibleingamecomponent (Utilities|Array|MakeArray (Variables|Default|GetCaptureAreaStaticMesh) (Variables|Default|GetP_CaptureArea_Blocks) (Variables|Default|GetCaptureAreaLight)) true)
  (Visibility|SetVisibilityComponentsWithCollision _visibleingamecomponent (Utilities|Map|MakeMap (Variables|Default|GetBaseMesh) "QueryAndPhysics") true)
  (Visibility|UpdateVisibilityComponent))

(fn SetViewOnEditMode
  (params (ForceViewMode :type Boolean))
  (Networking|CanExecuteCosmeticEvents
    (:True
      (if (or (not (Class|CreativeCommonDeviceFunctionLibrary|IsinPlayMode self)) (Variables|GetForceViewMode))
        (Rendering|Material|SetScalarParameterValue (Variables|Default|GetCaptureAreaMID) (Variables|MIDParameters|GetMIDParameterCaptureEffectOpacity) 0.0 :!class /Script/Engine.MaterialInstanceDynamic)
        (Rendering|Material|SetScalarParameterValue (Variables|Default|GetCaptureAreaMID) (Variables|MIDParameters|GetMIDParameterCaptureProgress) 1.0 :!class /Script/Engine.MaterialInstanceDynamic)
        (Rendering|Material|SetScalarParameterValue (Variables|Default|GetCaptureAreaMID) (Variables|MIDParameters|GetMIDParameterCaptureEffectDirection) 0.0 :!class /Script/Engine.MaterialInstanceDynamic)
        (Device|UpdateDeviceColors)))))

(fn ShouldInvisiblePropsBeAllowed
  (return false))

(fn ShouldMaterialInstanceParametersBeCopied
  (params (MaterialObject :type Material%20Object%20Reference))
  (return (and (or (Utilities|Enum|Equal(Enum) (Variables|/Script/Engine.Material|GetBlendMode MaterialObject) "BLEND_Opaque") (Utilities|Enum|Equal(Enum) (Variables|/Script/Engine.Material|GetBlendMode MaterialObject) "BLEND_Masked")) (not (Utilities|Array|ContainsItem (Variables|ItemPreviews|GetMaterialsToExclude) (Utilities|ResolveSoftReference MaterialObject))))))

(fn ShowNextItemPreview
  (bind _previewmeshindex (Variables|ItemPreviews|GetPreviewMeshIndex))
  (bind _meshpreviews (Variables|ItemPreviews|GetMeshPreviews))
  (bind _output (Utilities|Array|Get(acopy) _meshpreviews _previewmeshindex))
  (bind (_meshassetpath_8_6b40c41c47213e2d7d041a876713c4f5 _meshobject_5_319aaacd4f43372817f7429c492d1b93 _raritycolor_2_e689f1664314461dba24e7b7f58d6edc _pickuptransform_15_7d98aff949c4025f30a8f4b107b5ccc8) (Utilities|Struct|BreakPickupItemPreview _output))
  (bind _previewstaticmesh (Variables|Default|GetPreviewStaticMesh))
  (bind (_origin _boxextent _sphereradius) (Collision|GetComponentBounds _previewstaticmesh))
  (bind (_origin_2 _boxextent_2 _sphereradius_2) (Collision|GetComponentBounds (Variables|Default|GetPreviewSkeletalMesh)))
  (bind _previewskeletalmesh (Variables|Default|GetPreviewSkeletalMesh))
  (bind _returnvalue (Transformation|GetAllSocketNames _previewskeletalmesh))
  (bind _currentindex (LocalVariables|GetCurrentIndex))
  (bind _previewmeshmid (Variables|ItemPreviews|GetPreviewMeshMID))
  (bind _output_2 (Utilities|Array|Get(acopy) _previewmeshmid _currentindex))
  (bind _previewskeletalmesh_2 (Variables|Default|GetPreviewSkeletalMesh))
  (bind _previewmeshmid_2 (Variables|ItemPreviews|GetPreviewMeshMID))
  (bind _currentstaticmeshobject (Variables|Default|GetCurrentStaticMeshObject))
  (bind _currentstaticmeshobject_2 (Variables|Default|GetCurrentStaticMeshObject))
  (if (Utilities|IsArrayNonEmpty _meshpreviews)
    (Utilities|FlowControl|Sequence
      (:then_0
        ;[1 1 1] Update the Preview Mesh Index
        (Math|Integer|IncrementInt (Variables|ItemPreviews|GetPreviewMeshIndex)
          (:%20%20
            (if (not (Utilities|Array|IsValidIndex (Variables|ItemPreviews|GetMeshPreviews) (Variables|ItemPreviews|GetPreviewMeshIndex)))
              (Variables|ItemPreviews|SetPreviewMeshIndex 0)))))
      (:then_1
        (LocalVariables|SetRarityColor _raritycolor_2_e689f1664314461dba24e7b7f58d6edc)
        (Utilities|IsValid _meshobject_5_319aaacd4f43372817f7429c492d1b93
          (:Is%20Valid
            (LocalVariables|SetTransform _pickuptransform_15_7d98aff949c4025f30a8f4b107b5ccc8)
            (bind _asstatic_mesh (Utilities|Casting|CastToStaticMesh _meshobject_5_319aaacd4f43372817f7429c492d1b93)
              (:then
                ;[0.1882 0.5469 1 0.493] Static Mesh Route
                (Variables|Default|SetCurrentStaticMeshObject _asstatic_mesh)
                (Rendering|SetVisibility (Variables|Default|GetPreviewSkeletalMesh) false false)
                (Rendering|SetVisibility (Variables|Default|GetPreviewStaticMesh) true false)
                (Utilities|Array|Clear (LocalVariables|GetStaticMeshMaterials))
                ;;[1 1 1] Doesn't seem to be a native "get materials" array for static mesh objects, we look for up to 5 mateirals and make array
                (bind _index (Utilities|FlowControl|ForLoopwithBreak 0 4 :!label _label)
                  (:LoopBody
                    (Utilities|IsValid (StaticMesh|GetMaterial _currentstaticmeshobject_2 _index)
                      (:Is%20Valid
                        (Utilities|Array|Add (LocalVariables|GetStaticMeshMaterials) (StaticMesh|GetMaterial _currentstaticmeshobject_2 _index)))
                      (:Is%20Not%20Valid
                        (goto _label Break))))
                  (:Completed
                    (Components|StaticMesh|SetStaticMesh _previewstaticmesh (Variables|Default|GetCurrentStaticMeshObject))
                    (Variables|ItemPreviews|SetComponentBounds _sphereradius)
                    ;;[1 1 1] Copy Materials
                    (for (_array_element _array_index) (LocalVariables|GetStaticMeshMaterials)
                      (LocalVariables|SetCurrentMaterialInterface _array_element)
                      (bind _output_get (LocalVariables|SetCurrentIndex _array_index))
                      (if (< _output_get (Utilities|Array|Length (Variables|ItemPreviews|GetPreviewMeshMID)))
                        (bind _returnvalue_2 (Rendering|Material|GetBaseMaterial (LocalVariables|GetCurrentMaterialInterface)))
                        (if (Getters|ShouldMaterialInstanceParametersbeCopied _returnvalue_2)
                          (Rendering|Material|SetMaterial (Variables|Default|GetPreviewStaticMesh) (LocalVariables|GetCurrentIndex) _output_2)
                          (Rendering|Material|CopyMaterialInstanceParameters _output_2 (LocalVariables|GetCurrentMaterialInterface) true)
                          (Rendering|Material|SetVectorParameterValue _output_2 "Color" (LocalVariables|GetRarityColor) :!class /Script/Engine.MaterialInstanceDynamic)
                          (else
                            (Rendering|Material|SetMaterial (Variables|Default|GetPreviewStaticMesh) (LocalVariables|GetCurrentIndex) (CallFunction|CheckforMeshMaterialExceptions (Variables|Default|GetCurrentStaticMeshObject) (LocalVariables|GetCurrentMaterialInterface)))))))
                    (CallFunction|CheckMeshforSpecialExceptions _currentstaticmeshobject (LocalVariables|GetTransform))
                    (ItemPreviews|SetBubbleScale)
                    (CallFunction|SetFlagTeamColors _currentstaticmeshobject)
                    (CallFunction|RefreshPreviewRotation))))
              (:CastFailed
                (bind _asskeletal_mesh (Utilities|Casting|CastToSkeletalMesh _meshobject_5_319aaacd4f43372817f7429c492d1b93))
                ;[1 0.2891 0.7518 0.264] Skeletal Mesh Path
                (Transformation|DetachFromComponent _previewskeletalmesh_2 "KeepRelative" "KeepRelative" "KeepRelative" true)
                (Transformation|SetRelativeTransform _previewskeletalmesh_2 (LocalVariables|GetTransform) false true)
                (Rendering|SetVisibility (Variables|Default|GetPreviewStaticMesh) false false)
                (Rendering|SetVisibility _previewskeletalmesh_2 true false)
                (Components|SkinnedMesh|SetSkinnedAssetandUpdate _previewskeletalmesh_2 _asskeletal_mesh true)
                (Variables|ItemPreviews|SetComponentBounds (* _sphereradius_2 (.x (Transformation|GetActorScale3D))))
                ;;[1 1 1] Copy Materials
                (for (_array_element_2 _array_index_2) (Variables|/Script/Engine.SkeletalMesh|GetMaterials _asskeletal_mesh)
                  (if (< _array_index_2 (Utilities|Array|Length (Variables|ItemPreviews|GetPreviewMeshMID)))
                    (Rendering|Material|SetMaterial (Variables|Default|GetPreviewSkeletalMesh) _array_index_2 (Utilities|Array|Get(acopy) _previewmeshmid_2 _array_index_2))
                    (Rendering|Material|CopyMaterialInstanceParameters (Utilities|Array|Get(acopy) _previewmeshmid_2 _array_index_2) (Utilities|Struct|BreakSkeletalMaterial _array_element_2) true)
                    (Rendering|Material|SetVectorParameterValue (Utilities|Array|Get(acopy) _previewmeshmid_2 _array_index_2) "Color" (LocalVariables|GetRarityColor) :!class /Script/Engine.MaterialInstanceDynamic)))
                ;;[1 1 1] Many weapons have offset pivots. Get center mass throws NaN errors, so this is my workaround to getting the offet to re-center skel meshes
                (LocalVariables|SetSkelMeshCenter "0, 0, 0")
                (for _array_element_3 _returnvalue
                  (LocalVariables|SetSkelMeshCenter (+ (Transformation|GetSocketLocation (Variables|Default|GetPreviewSkeletalMesh) _array_element_3) (LocalVariables|GetSkelMeshCenter))))
                (Transformation|SetRelativeLocation (Variables|Default|GetPreviewSkeletalMesh) (* (/ (LocalVariables|GetSkelMeshCenter) (Utilities|Array|Length _returnvalue)) -1.0) false false)
                (Transformation|AttachComponentToComponent (Variables|Default|GetPreviewSkeletalMesh) (Variables|Default|GetPreviewMeshScene) "None" "KeepRelative" "KeepRelative" "KeepRelative" true)
                (CallFunction|ResetMeshExceptions)
                (ItemPreviews|SetBubbleScale)
                (CallFunction|RefreshPreviewRotation)))))))))

(fn StartOrRefreshItemPreviews
  (params (Minigame :type Fort%20Minigame%20Object%20Reference))
  (Networking|CanExecuteCosmeticEvents
    (:True
      (Utilities|IsValid Minigame
        (:Is%20Valid
          (ItemPreviews|StopItemPreviews)
          (Utilities|IsValid Minigame
            (:Is%20Valid
              (if (or (not (Minigame|IsinProgress Minigame)) (and (Minigame|IsinProgress Minigame) (Variables|DeviceOptions|GetbItemVisibleInGame)))
                (CallFunction|LoadItemPreviewsAsync)
                (bind _returnvalue (Utilities|Time|SetTimerbyEvent (EventDispatchers|CreateEvent :!event FX_ShowNextPreview) (LocalVariables|GetPreviewMeshCycleTimer) true))
                (Variables|ItemPreviews|SetPreviewMeshTimer _returnvalue)
                (Rendering|SetVisibility (Variables|Default|GetPreviewSkeletalMesh) true false)))))))))

(fn StartPeriodicScoringTimer
  (params (ScoringTeam :type Byte))
  (Networking|CanExecuteCosmeticEvents
    (:False
      (if (and (Variables|Replicated|GetIsEnabledByPhase) (not (Utilities|Time|IsTimerActivebyHandle (Variables|PeriodicScoring|GetPeriodicScoringTimerHandle))))
        (bind _returnvalue (Utilities|Time|SetTimerbyEvent (EventDispatchers|CreateEvent :!event OnPeriodicScoringTimerFinished) (Math|Conversions|ToFloat(Integer) (Variables|DeviceOptions|GetPlayerCaptureTime)) true))
        (Variables|PeriodicScoring|SetPeriodicScoringTimerHandle _returnvalue)
        (Variables|PeriodicScoring|SetTeamScoringPeriodically (Variables|GetScoringTeam))))))

(fn StopItemPreviews
  (Utilities|Time|ClearandInvalidateTimerbyHandle (Variables|ItemPreviews|GetPreviewMeshTimer))
  (Components|SkinnedMesh|SetSkinnedAssetandUpdate (Variables|Default|GetPreviewSkeletalMesh) 0 true)
  (Components|StaticMesh|SetStaticMesh (Variables|Default|GetPreviewStaticMesh) 0)
  (Rendering|SetVisibility (Variables|Default|GetPreviewSkeletalMesh) false false)
  (CallFunction|FXStopItemPreviewRotation)
  (Transformation|SetRelativeScale3D (Variables|Default|GetPreviewBubble) "0, 0, 0"))

(fn StopPeriodicScoringTimer
  (Networking|CanExecuteCosmeticEvents
    (:False
      (Utilities|Time|ClearandInvalidateTimerbyHandle (Variables|PeriodicScoring|GetPeriodicScoringTimerHandle))
      (Variables|PeriodicScoring|SetTeamScoringPeriodically 0))))

(fn TryToCapture
  (params (CapturingTeamIndex :type Byte))
  (CaptureComponent|SetCaptureState (Variables|CaptureArea|GetCaptureComponent) "Capturing" (Variables|GetCapturingTeamIndex))
  (if (== (Variables|DeviceOptions|GetControlTime) 0)
    (CaptureComponent|SetCaptureState (Variables|CaptureArea|GetCaptureComponent) "Captured" (Variables|GetCapturingTeamIndex))))

(fn TryToContest
  (params (TeamToContest :type Byte))
  (bind _array_element (Utilities|Array|ForEachLoopwithBreak (Variables|CaptureArea|GetCaptureAreaTeamsInfo) :!label _label)
    (:LoopBody
      (bind (_array_element_players _array_element_insideteamindex) (Utilities|Struct|BreakCaptureAreaTeamInfo _array_element))
      (if (Math|Byte|NotEqual(Byte) _array_element_insideteamindex (Variables|GetTeamToContest))
        (CaptureComponent|SetCaptureState (Variables|CaptureArea|GetCaptureComponent) "Contested" _array_element_insideteamindex)
        (goto _label Break)))))

(fn TryToDecapture
  (if (== (Math|Float|Truncate (Variables|DeviceOptions|GetDecaySpeed)) -1)
    (CaptureComponent|SetCaptureState (Variables|CaptureArea|GetCaptureComponent) "Neutral" 0)
    (else
      (CaptureComponent|SetCaptureState (Variables|CaptureArea|GetCaptureComponent) "Decapturing" (Variables|/Script/FortniteGame.CaptureComponent|GetCapturingStateOwner (Variables|CaptureArea|GetCaptureComponent)))
      (Capture|UpdateProgressReverseSpeed))))

(fn TryToDeliverItemsFromInsidePlayers
  (for _array_element (Variables|CaptureArea|GetCaptureAreaTeamsInfo)
    (bind _array_element_players (Utilities|Struct|BreakCaptureAreaTeamInfo _array_element))
    (for _array_element_2 _array_element_players
      (Utilities|IsValid _array_element_2
        (:Is%20Valid
          (ItemDelivery|TrytoDeliverItemsfromPlayer _array_element_2))))))

(fn TryToDeliverItemsFromPlayer
  (params (PlayerController :type Fort%20Player%20Controller%20Object%20Reference))
  (bind _playercontroller (Variables|GetPlayerController))
  (bind _playercontroller_2 (Variables|GetPlayerController))
  (if (Math|Boolean|ANDBoolean (PeriodicScoring|IsValidControllerforScoring PlayerController) (or (Utilities|Enum|Equal(Enum) (Variables|DeviceOptions|GetPlayerCapture) "OnePlayerPerTeam") (Utilities|Enum|Equal(Enum) (Variables|DeviceOptions|GetPlayerCapture) "EachPlayer")) (or (Utilities|Enum|Equal(Enum) (Variables|Default|GetbItemsFilterPlayerCapture) "ForPeriodicScoring") (Utilities|Enum|Equal(Enum) (Variables|Default|GetbItemsFilterPlayerCapture) "Both")))
    _
    (elif (Getters|IsCaptureAreaInteractable)
      (Utilities|IsValid (Variables|GetPlayerController)
        (:Is%20Valid
          (if (ItemDelivery|CanDeliverItems (Team|GetActorTeam (Variables|GetPlayerController)))
            (bind (_addedscore _returnvalue) (CaptureArea|CaptureItemsfromPlayer (Variables|DeviceOptions|GetConsumeItemOnScoring) (ItemData|GetContainedItemVariants (Variables|Default|GetPickupItemList)) _playercontroller -1)) ; Amount to Remove is -1 because we \nwant to remove all
            (LocalVariables|SetTotalItemsCaptured _returnvalue)
            (PeriodicScoring|SendScoreStatEvents _playercontroller (+ _addedscore (* _returnvalue (Variables|DeviceOptions|GetItemDeliveryScore))))
            (if (> (LocalVariables|GetTotalItemsCaptured) 0)
              (CallFunction|FXObjectCaptured (Game|Player|GetPlayerPawn _playercontroller) (Game|Player|GetPlayerPawn _playercontroller))
              (GameplayEvent|BroadcastEvent (Variables|UserOptions|Events|GetOnItemIsDelivered) 0 (Utilities|Struct|MakeGameplayEventItemDelivered _playercontroller_2))
              (GameplayEvent|BroadcastEvent (Variables|UserOptions|Events|GetOnItemIsConsumed) 0 (Utilities|Struct|MakeGameplayEventItemConsumed _playercontroller_2)))))))))

(fn TryToDeneutralize
  (if (== (Math|Float|Truncate (Variables|DeviceOptions|GetDecaySpeed)) -1)
    (CaptureComponent|SetCaptureState (Variables|CaptureArea|GetCaptureComponent) "Captured" (Variables|Team|GetTeamIndex))
    (else
      (CaptureComponent|SetCaptureState (Variables|CaptureArea|GetCaptureComponent) "Deneutralizing" (Variables|Team|GetTeamIndex))
      (Capture|UpdateProgressReverseSpeed))))

(fn TryToGetValidControllerFromTeam
  (params (Team :type Byte))
  (bind _currentminigame (Variables|/Script/FortniteGame.FortMinigameLogicComponent|GetCurrentMinigame (Variables|Default|GetMinigameLogic))
    (:then
      (bind _array_element (Utilities|Array|ForEachLoopwithBreak (Minigame|GetParticipatingPlayers _currentminigame))
        (:LoopBody
          (if (Math|Byte|Equal(Byte) Team (Team|GetActorTeam _array_element))
            (bind _asfort_player_controller (Utilities|Casting|CastToFortPlayerController (Actor|GetOwner _array_element)))
            (return _asfort_player_controller)))
        (:Completed
          (return 0))))
    (:else)))

(fn TryToNeutralize
  (params (TeamIndexTrigger :type Byte))
  (bind _capturecomponent (Variables|CaptureArea|GetCaptureComponent))
  (Math|Integer|CompareInt (Variables|DeviceOptions|GetNeutralizeTIme)
    (:>
      (CaptureComponent|SetCaptureState (Variables|CaptureArea|GetCaptureComponent) "Neutralizing" (Variables|GetTeamIndexTrigger)))
    (:==
      (CaptureComponent|ForceCaptureProgress _capturecomponent 0.0)
      (CaptureComponent|SetCaptureState _capturecomponent "Neutral" (Variables|GetTeamIndexTrigger)))
    (:<
      (if (or (Utilities|Enum|Equal(Enum) (CaptureComponent|GetCaptureState (Variables|CaptureArea|GetCaptureComponent)) "Captured") (Utilities|Enum|Equal(Enum) (Variables|/Script/FortniteGame.CaptureComponent|GetPreviousCaptureState (Variables|CaptureArea|GetCaptureComponent)) "Captured"))
        (CaptureComponent|ForceCaptureProgress _capturecomponent 0.0))
      (Capture|TrytoCapture (Variables|GetTeamIndexTrigger)))) ; -1 = No Neutralization (So Start Capturing)\n0 = Instant (So Neutral)\n> 0 = Timed Neutralization
)

(fn TryToUpdateCaptureState
  (Networking|CanExecuteCosmeticEvents
    (:False
      (if (Math|Boolean|ANDBoolean (Getters|IsCaptureAreaInteractable) (Getters|IsMinigameinProgress) (not (Knobs|Team|IsTeamNone (Variables|DeviceOptions|GetControlledByIndex))))
        (Math|Integer|CompareInt (Utilities|Array|Length (Variables|CaptureArea|GetCaptureAreaTeamsInfo)) 1
          (:>
            ;[1 1 1] Area can be contested
            (switch Utilities|FlowControl|Switch|SwitchonECaptureState (CaptureComponent|GetCaptureState (Variables|CaptureArea|GetCaptureComponent))
              (:Neutral
                _)
              (:Capturing
                _)
              (:Decapturing
                _)
              (:Neutralizing
                (Capture|TrytoContest (Variables|/Script/FortniteGame.CaptureComponent|GetNeutralizingStateOwner (Variables|CaptureArea|GetCaptureComponent)) :!label _label))
              (:Deneutralizing
                (goto _label))
              (:Captured
                (Capture|TrytoContest (Variables|Team|GetTeamIndex))))
            (Capture|TrytoContest (Variables|/Script/FortniteGame.CaptureComponent|GetCapturingStateOwner (Variables|CaptureArea|GetCaptureComponent))))
          (:==
            ;[1 1 1] Validate team inside is a Capturable Team
            (bind _output_get (LocalVariables|SetFirstTeamIndexFound (Capture|GetFirstCapturableTeamIndexinArea)))
            (if (Utilities|Enum|NotEqual(Enum) (Utilities|Enum|BytetoEnumEFortTeam _output_get) "Spectator")
              (if (Math|Byte|NotEqual(Byte) (LocalVariables|GetFirstTeamIndexFound) (Variables|Team|GetTeamIndex))
                ;[1 0 0.01492] Area is Overlapped by a team that does not actually control it
                (Capture|UpdateCaptureSpeed)
                (switch Utilities|FlowControl|Switch|SwitchonECaptureState (CaptureComponent|GetCaptureState (Variables|CaptureArea|GetCaptureComponent))
                  (:Neutral
                    ;;[1 1 1] Area is Neutral
                    (Capture|TrytoCapture (LocalVariables|GetFirstTeamIndexFound)))
                  (:Contested
                    _)
                  (:Decapturing
                    _)
                  (:Neutralizing
                    ;;[1 1 1] Area is Controlled
                    (Capture|TrytoNeutralize (LocalVariables|GetFirstTeamIndexFound) :!label _label_2))
                  (:Deneutralizing
                    (goto _label_2))
                  (:Captured
                    (goto _label_2)))
                ;;[1 1 1] Area has to be captured, decaptured, or neutralized
                (if (Utilities|Enum|Equal(Enum) (Utilities|Enum|BytetoEnumEFortTeam (Variables|Team|GetTeamIndex)) "Spectator")
                  (if (Math|Byte|Equal(Byte) (LocalVariables|GetFirstTeamIndexFound) (Variables|/Script/FortniteGame.CaptureComponent|GetCapturingStateOwner (Variables|CaptureArea|GetCaptureComponent)))
                    (Capture|TrytoCapture (Variables|/Script/FortniteGame.CaptureComponent|GetCapturingStateOwner (Variables|CaptureArea|GetCaptureComponent)))
                    (else
                      (Capture|TrytoDecapture)))
                  (else
                    (Capture|TrytoNeutralize (LocalVariables|GetFirstTeamIndexFound))))
                (else
                  _))))
          (:<
            _))
        ;[1 1 1] No Controller in Area able to Control the point but can reverse progress
        (switch Utilities|FlowControl|Switch|SwitchonECaptureState (CaptureComponent|GetCaptureState (Variables|CaptureArea|GetCaptureComponent))
          (:Capturing
            (Capture|TrytoDecapture))
          (:Contested
            (switch Utilities|FlowControl|Switch|SwitchonECaptureState (Variables|/Script/FortniteGame.CaptureComponent|GetPreviousCaptureState (Variables|CaptureArea|GetCaptureComponent))
              (:Neutral
                (CaptureComponent|SetCaptureState (Variables|CaptureArea|GetCaptureComponent) "Neutral" (Utilities|Enum|LiteralenumEFortTeam "Spectator")))
              (:Capturing
                _)
              (:Decapturing
                _)
              (:Neutralizing
                (Capture|TrytoDeneutralize :!label _label_3))
              (:Deneutralizing
                (goto _label_3))
              (:Captured
                (CaptureComponent|SetCaptureState (Variables|CaptureArea|GetCaptureComponent) "Captured" (Variables|Team|GetTeamIndex))))
            (Capture|TrytoDecapture))
          (:Decapturing
            _)
          (:Neutralizing
            (Capture|TrytoDeneutralize))
          (:Deneutralizing
            _))
        (Capture|UpdateProgressReverseSpeed)))))

(fn TryToUpdatePeriodicScoring
  (Networking|CanExecuteCosmeticEvents
    (:False
      (if (and (Getters|IsMinigameinProgress) (Utilities|Enum|NotEqual(Enum) (Variables|DeviceOptions|GetPlayerCapture) "Off"))
        (if (Utilities|Enum|NotEqual(Enum) (Variables|DeviceOptions|GetPlayerCapture) "OwningTeam")
          (if (Getters|IsCaptureAreaInteractable)
            (Math|Integer|CompareInt (Utilities|Array|Length (Variables|CaptureArea|GetCaptureAreaTeamsInfo)) 1
              (:>
                ;[1 1 1] Multiple Teams in Area
                (if (Variables|DeviceOptions|GetEnemiesContestScoring)
                  (Utilities|Time|PauseTimerbyHandle (Variables|PeriodicScoring|GetPeriodicScoringTimerHandle))))
              (:==
                (bind _output_get (LocalVariables|SetFirstTeamIndexFound (PeriodicScoring|GetFirstValidTeamIndexforScoringinArea)))
                (if (Utilities|Enum|NotEqual(Enum) (Utilities|Enum|BytetoEnumEFortTeam _output_get) "Spectator")
                  ;[1 1 1] Valid Controller for Scoring In Area
                  (if (Math|Byte|Equal(Byte) (LocalVariables|GetFirstTeamIndexFound) (Variables|PeriodicScoring|GetTeamScoringPeriodically))
                    (Utilities|Time|UnpauseTimerbyHandle (Variables|PeriodicScoring|GetPeriodicScoringTimerHandle))
                    (else
                      (PeriodicScoring|StartPeriodicScoringTimer (LocalVariables|GetFirstTeamIndexFound))))
                  (else
                    _)))
              (:<
                _))
            ;[1 1 1] No valid Controller in Area
            (if (Utilities|Time|IsTimerActivebyHandle (Variables|PeriodicScoring|GetPeriodicScoringTimerHandle))
              (PeriodicScoring|StopPeriodicScoringTimer)))
          (else
            (PeriodicScoring|UpdateOwningTeamScoringTimer)))))))

(fn UpdateAudioLoopAttenuation
  (Utilities|FlowControl|Sequence
    (:then_0
      (Utilities|IsValid (Variables|Default|GetCapture%20Area%20Ambient%20Audio)
        (:Is%20Valid
          (Audio|Components|Audio|Stop (Variables|Default|GetCapture%20Area%20Ambient%20Audio)))
        (:Is%20Not%20Valid
          _))
      (bind _structout (Utilities|Struct|SetmembersinSoundAttenuationSettings (Variables|/Script/Engine.AudioComponent|GetAttenuationOverrides (Variables|Default|GetCapture%20Area%20Ambient%20Audio)) true "Sphere" (Math|Vector|MakeVector (* (Variables|DeviceOptions|GetCaptureAreaSize) 512.0)) 1.2e+03))
      (Audio|Components|Audio|AdjustAttenuation (Variables|Default|GetCapture%20Area%20Ambient%20Audio) _structout)
      (Audio|Components|Audio|Play (Variables|Default|GetCapture%20Area%20Ambient%20Audio) 0.0))
    (:then_1
      (bind _structout_2 (Utilities|Struct|SetmembersinSoundAttenuationSettings (Variables|/Script/Engine.AudioComponent|GetAttenuationOverrides (Variables|Default|GetCapturing%20Audio%20Loop%20A)) false "Sphere" (Math|Vector|MakeVector (* (Variables|DeviceOptions|GetCaptureAreaSize) 512.0)) 1e+03))
      (Audio|Components|Audio|AdjustAttenuation (Variables|Default|GetCapturing%20Audio%20Loop%20A) _structout_2)
      (bind _structout_3 (Utilities|Struct|SetmembersinSoundAttenuationSettings (Variables|/Script/Engine.AudioComponent|GetAttenuationOverrides (Variables|Default|GetCapturing%20Audio%20Loop%20B)) false "Sphere" (Math|Vector|MakeVector (* (Variables|DeviceOptions|GetCaptureAreaSize) 512.0)) 1e+03))
      (Audio|Components|Audio|AdjustAttenuation (Variables|Default|GetCapturing%20Audio%20Loop%20B) _structout_3)
      (bind _structout_4 (Utilities|Struct|SetmembersinSoundAttenuationSettings (Variables|/Script/Engine.AudioComponent|GetAttenuationOverrides (Variables|Default|GetCaptured%20Audio)) false "Sphere" (Math|Vector|MakeVector (* (Variables|DeviceOptions|GetCaptureAreaSize) 512.0)) 1e+03))
      (Audio|Components|Audio|AdjustAttenuation (Variables|Default|GetCaptured%20Audio) _structout_4))))

(fn UpdateAllowedInteractionSetting
  (Networking|CanExecuteCosmeticEvents
    (:False
      (Variables|Default|SetIsAllowed (Variables|DeviceOptions|GetAllowedOnGameStart)))))

(fn UpdateCaptureArea
  (bind (_min _max) (Components|StaticMesh|GetLocalBounds (Variables|Default|GetCaptureAreaStaticMesh)))
  (bind _captureareasize (Variables|DeviceOptions|GetCaptureAreaSize))
  (bind _p_capturearea_blocks (Variables|Default|GetP_CaptureArea_Blocks))
  (Transformation|SetRelativeScale3D (Variables|Default|GetCaptureAreaStaticMesh) (Math|Vector|MakeVector (* (Variables|DeviceOptions|GetCaptureAreaSize) (Variables|Constants|GetCaptureAreaScaleFactor)) (* (Variables|DeviceOptions|GetCaptureAreaSize) (Variables|Constants|GetCaptureAreaScaleFactor)) (* (Variables|DeviceOptions|GetCaptureAreaHeight) (Variables|Constants|GetCaptureHeightScaleFactor))))
  (Transformation|SetRelativeScale3D (Variables|Default|GetDropArea) (Math|Vector|MakeVector (* (Variables|DeviceOptions|GetCaptureAreaSize) 2.0) (* (Variables|DeviceOptions|GetCaptureAreaSize) 2.0) (.z (Variables|/Script/Engine.SceneComponent|GetRelativeScale3D (Variables|Default|GetDropArea)))))
  (Rendering|Lighting|SetAttenuationRadius (Variables|Default|GetCaptureAreaLight) (Utilities|ClampMaxFloat (+ (* (Variables|DeviceOptions|GetCaptureAreaSize) (Variables|Constants|GetLightSizeScale)) (Variables|Constants|GetLightSizeBias)) (Variables|Constants|GetMaxLightSize)))
  (Transformation|SetWorldScale3D _p_capturearea_blocks (Math|Vector|MakeVector _captureareasize _captureareasize (Variables|DeviceOptions|GetCaptureAreaHeight)))
  (Effects|Components|ParticleSystem|SetFloatParameter _p_capturearea_blocks "BlocksSize" (* (/ 1.0 _captureareasize) 5.0))
  (Effects|Components|ParticleSystem|SetFloatParameter (Variables|Default|GetP_CaptureArea_Blocks) "SpawnRateScale" (/ (Math|Float|float*float (Math|Trig|GetPI) (Math|Float|Square (* (Variables|DeviceOptions|GetCaptureAreaSize) 256.0)) (.x (* _max (Variables|/Script/Engine.SceneComponent|GetRelativeScale3D (Variables|Default|GetCaptureAreaStaticMesh))))) 82354968.0))
  (CallFunction|UpdateAudioLoopAttenuation)
  (EnableKnob|UpdateEnableAnimationLength))

(fn UpdateCaptureSpeed
  (bind _array_element (Utilities|Array|ForEachLoopwithBreak (Variables|CaptureArea|GetCaptureAreaTeamsInfo) :!label _label)
    (:LoopBody
      (bind (_array_element_players _array_element_insideteamindex) (Utilities|Struct|BreakCaptureAreaTeamInfo _array_element))
      (if (Capture|CanbeCapturedBy _array_element_insideteamindex)
        (for _array_element_2 _array_element_players
          (if (Capture|IsValidControllerforCapture _array_element_2)
            (Math|Integer|IncrementInt (LocalVariables|GetNumberOfValidCaptureControllers)
              (:%20%20
                (if (Emote|IsPlayingEmote _array_element_2)
                  (Math|Integer|IncrementInt (LocalVariables|GetNumberOfValidEmotingControllers)))))))
        (goto _label Break)))
    (:Completed
      (CaptureComponent|UpdateCaptureSpeed (Variables|CaptureArea|GetCaptureComponent) (LocalVariables|GetNumberOfValidCaptureControllers) (LocalVariables|GetNumberOfValidEmotingControllers)))))

(fn UpdateCaptureState
  (params (CaptureState :type ECaptureState%20Enum :default "Neutral"))
  (switch Utilities|FlowControl|Switch|SwitchonECaptureState CaptureState
    (:Neutral
      (Capture|OnNeutral))
    (:Capturing
      (Capture|OnCapturingStarted))
    (:Contested
      (Capture|OnCaptureContested))
    (:Decapturing
      (Capture|OnDecapturing))
    (:Neutralizing
      (Capture|OnNeutralizingStarted))
    (:Deneutralizing
      (Capture|OnDeneutralizing))
    (:Captured
      (Capture|OnCaptured))))

(fn UpdateCollisionForGame
  (params (Minigame :type Fort%20Minigame%20Object%20Reference))
  (Networking|SwitchHasAuthority
    (:Authority
      (Utilities|IsValid Minigame
        (:Is%20Valid
          (if (Minigame|IsinProgress (Variables|GetMinigame))
            (Collision|SetCollisionEnabled (Variables|Default|GetCaptureAreaStaticMesh) "QueryAndPhysics")
            (else
              (Collision|SetCollisionEnabled (Variables|Default|GetCaptureAreaStaticMesh) "NoCollision"))))))))

(fn UpdateControlPointSettings
  (CaptureComponent|SetNeutralizingTime (Variables|CaptureArea|GetCaptureComponent) (Variables|DeviceOptions|GetNeutralizeTIme))
  (CaptureComponent|SetCaptureTime (Variables|CaptureArea|GetCaptureComponent) (Math|Conversions|ToFloat(Integer) (Variables|DeviceOptions|GetControlTime)))
  (Networking|CanExecuteCosmeticEvents
    (:True
      (Rendering|Material|SetScalarParameterValue (Variables|Default|GetCaptureAreaMID) (Variables|MIDParameters|GetMIDParameterSimpleCapture) (Math|Conversions|ToFloat(Boolean) (not (Variables|DeviceOptions|GetDisplayProgressVFX))) :!class /Script/Engine.MaterialInstanceDynamic))
    (:False
      (CaptureComponent|SetProgressDecaySpeed (Variables|CaptureArea|GetCaptureComponent) (Variables|DeviceOptions|GetDecaySpeed))
      (CaptureComponent|SetCaptureSpeedBoost (Variables|CaptureArea|GetCaptureComponent) (Utilities|ClampMinFloat (Variables|DeviceOptions|GetControlSpeedBoost) 1.0))
      (CaptureComponent|SetCaptureSpeedBoostEmoting (Variables|CaptureArea|GetCaptureComponent) (Variables|DeviceOptions|GetControlSpeedEmotingBoostKnob)))))

(fn UpdateDeviceColors
  (bind _teamlightcolors (Class|GetClassDefaults "/Game/Creative/Sets/MilitaryBase/Props/CreativeTeamColors.CreativeTeamColors_C"))
  (bind _returnvalue (Utilities|Array|Length _teamlightcolors))
  (bind _returnvalue_2 (Math|Float|Truncate (Getters|GetAccentColorParam)))
  (bind _returnvalue_3 (Utilities|Array|IsValidIndex _teamlightcolors _returnvalue_2))
  (bind _returnvalue_4 (- _returnvalue 1))
  (bind _accentcolorparam (Getters|GetAccentColorParam))
  (bind _returnvalue_5 (Math|Integer|SelectInt _returnvalue_2 _returnvalue_4 _returnvalue_3))
  (bind _output (Utilities|Array|Get(aref) _teamlightcolors _returnvalue_5))
  (Networking|CanExecuteCosmeticEvents
    (:True
      (Rendering|Material|SetScalarParameterValue (Variables|Default|GetCaptureAreaBaseMID) (Variables|MIDParameters|GetMIDParameterTeamOwner) _accentcolorparam :!class /Script/Engine.MaterialInstanceDynamic)
      (Effects|Components|ParticleSystem|SetColorParameter (Variables|Default|GetP_CaptureArea_Blocks) "Block Color" _output)
      (Rendering|Components|Light|SetLightColor (Variables|Default|GetCaptureAreaLight) _output false)
      (Rendering|Material|SetScalarParameterValue (Variables|Default|GetCaptureAreaMID) (Variables|MIDParameters|GetMIDParameterTeamOwner) _accentcolorparam :!class /Script/Engine.MaterialInstanceDynamic)
      ;[1 1 1] Only in these cases we are not interacting actually with the device,
      ; So it's safe to update InteractingTeam Color
      (if (Math|Boolean|ORBoolean (Utilities|Enum|Equal(Enum) (CaptureComponent|GetCaptureState (Variables|CaptureArea|GetCaptureComponent)) "Neutral") (Utilities|Enum|Equal(Enum) (CaptureComponent|GetCaptureState (Variables|CaptureArea|GetCaptureComponent)) "Captured") (Utilities|Enum|Equal(Enum) (CaptureComponent|GetCaptureState (Variables|CaptureArea|GetCaptureComponent)) "Disabled"))
        (Rendering|Material|SetScalarParameterValue (Variables|Default|GetCaptureAreaMID) (Variables|MIDParameters|GetMIDParameterInteractingTeam) _accentcolorparam :!class /Script/Engine.MaterialInstanceDynamic)))))

(fn UpdateEnableAnimationLength
  (bind _captureareastaticmesh (Variables|Default|GetCaptureAreaStaticMesh))
  (bind _returnvalue_x (Math|Vector|BreakVector (Transformation|GetWorldScale _captureareastaticmesh)))
  (bind _returnvalue (Math|Float|MapRangeClamped _returnvalue_x 512.0 8192.0 0.0 1.0))
  (Variables|MutableAnimationRelated|SetEnableAnimationLength (Math|Float|Lerp (Variables|Constants|GetMinEnableAnimationLength) (Variables|Constants|GetMaxEnableAnimationLength) _returnvalue))
  (Variables|MutableAnimationRelated|SetEnableAnimationScaleModifier _returnvalue))

(fn UpdateInteractingColor
  (if (== (Variables|DeviceOptions|GetAccentColorIndex) -2)
    (Rendering|Material|SetScalarParameterValue (Variables|Default|GetCaptureAreaMID) (Variables|MIDParameters|GetMIDParameterInteractingTeam) (Math|Float|SelectFloat 1.0 2.0 (Math|Byte|Equal(Byte) (Variables|/Script/FortniteGame.CaptureComponent|GetCapturingStateOwner (Variables|CaptureArea|GetCaptureComponent)) (Team|GetActorTeam (Fortnite|GetFirstLocalFortPlayerController true)))) :!class /Script/Engine.MaterialInstanceDynamic)
    (else
      (Rendering|Material|SetScalarParameterValue (Variables|Default|GetCaptureAreaMID) (Variables|MIDParameters|GetMIDParameterInteractingTeam) (Math|Conversions|ToFloat(Integer) (Getters|GetTeamColorIndex (Variables|/Script/FortniteGame.CaptureComponent|GetCapturingStateOwner (Variables|CaptureArea|GetCaptureComponent)))) :!class /Script/Engine.MaterialInstanceDynamic))))

(fn UpdateItemDeliveryScore
  (if (== (Variables|DeviceOptions|GetItemDeliveryScore) -1)
    (Variables|DeviceOptions|SetItemDeliveryScore (Variables|DeviceOptions|GetScorePerCapture))))

(fn UpdateObjectiveProgress
  (Networking|CanExecuteCosmeticEvents
    (:False
      (if (and (Variables|DeviceOptions|GetbCountAsObjectiveKnob) (Variables|DeviceOptions|GetbTrackProgressKnob))
        (Minigame|Objective|SetTrackedProgress (Variables|Default|GetFortMinigameObjective) (Variables|/Script/FortniteGame.CaptureComponent|GetCaptureProgress (Variables|CaptureArea|GetCaptureComponent)))))))

(fn UpdateObjectiveTrackingState
  (params (Minigame :type Fort%20Minigame%20Object%20Reference))
  (Networking|CanExecuteCosmeticEvents
    (:False
      (Utilities|IsValid Minigame
        (:Is%20Valid
          (if (and (Variables|DeviceOptions|GetbCountAsObjectiveKnob) (Variables|Replicated|GetIsEnabledByPhase))
            (Utilities|FlowControl|Sequence
              (:then_0
                (if (Variables|DeviceOptions|GetbTrackProgressKnob)
                  (Minigame|HasStarted (Variables|GetMinigame) :!class /Script/FortniteGame.FortMinigame
                    (:Yes
                      (if (not (Minigame|IsPostGame (Variables|GetMinigame)))
                        (Minigame|Objective|StartTracking (Variables|Default|GetFortMinigameObjective) (- (Utilities|Array|FindItem (Variables|Constants|GetTextureArrayForIndicator) (Utilities|Struct|BreakCreativeIconOption (Variables|DeviceOptions|GetIndicatorPresetIndexKnob))) 1) (Variables|/Script/FortniteGame.CaptureComponent|GetCaptureProgress (Variables|CaptureArea|GetCaptureComponent)) "CaptureObjective" (CaptureComponent|GetCaptureState (Variables|CaptureArea|GetCaptureComponent)) (CaptureComponent|GetCaptureStateOwner (Variables|CaptureArea|GetCaptureComponent)) (Variables|Default|GetIsAllowed))))
                    (:No
                      _)))
                (Minigame|Objective|StopTracking (Variables|Default|GetFortMinigameObjective) true 1.0))
              (:then_1
                (Variables|Default|GetCurrentOwner
                  (:then)
                  (:else
                    (if (Utilities|Enum|NotEqual(Enum) (Utilities|Enum|BytetoEnumEFortTeam (Variables|Team|GetTeamIndex)) "Spectator")
                      (Objective|AddasObjectivetoMinigame 1)))) ; Current Owner is going to be \nset on AddAsObjectiveToMinigame
              ))))))))

(fn UpdateOverlappingActors
  (bind (_origin _boxextent) (Collision|GetComponentBounds (Variables|Default|GetCaptureAreaStaticMesh)))
  (bind (_boxextent_x _boxextent_y _boxextent_z) (Math|Vector|BreakVector _boxextent))
  (Utilities|FlowControl|Sequence
    (:then_0
      (for _array_element (Collision|GetOverlappingActors (Variables|Default|GetCaptureAreaStaticMesh) "/Script/FortniteGame.FortPlayerPawn" :!class /Script/Engine.PrimitiveComponent)
        (Utilities|IsValid _array_element
          (:Is%20Valid
            (Default|HandleCaptureAreaOverlap _array_element)))))
    (:then_1
      (for _array_element_2 (Collision|GetOverlappingActors (Variables|Default|GetCaptureAreaStaticMesh) "/Script/FortniteGame.FortPhysicsPawnObject" :!class /Script/Engine.PrimitiveComponent)
        (Utilities|IsValid _array_element_2
          (:Is%20Valid
            (Default|HandleCaptureAreaOverlap _array_element_2)))))
    (:then_2
      (bind _outhits (Collision|MultiBoxTraceForObjects :Start (Math|Vector|vector-vector _origin (Math|Vector|MakeVector 0.0 0.0 (/ _boxextent_z 2.0))) :End (Math|Vector|vector+vector _origin (Math|Vector|MakeVector 0.0 0.0 (/ _boxextent_z 2.0))) :HalfSize (Math|Vector|MakeVector _boxextent_x _boxextent_y (/ _boxextent_z 2.0)) :Orientation (Transformation|GetWorldRotation (Variables|Default|GetDropArea)) :ObjectTypes (Utilities|Array|MakeArray "ObjectTypeQuery4") :bTraceComplex false :DrawDebugType "None" :bIgnoreSelf true :TraceColor "(R=1.000000,G=0.000000,B=0.000000,A=1.000000)" :TraceHitColor "(R=0.000000,G=1.000000,B=0.000000,A=1.000000)" :DrawTime 1e+01))
      (for _array_element_3 _outhits
        (bind (_array_element_bblockinghit _array_element_binitialoverlap _array_element_time _array_element_distance _array_element_location _array_element_impactpoint _array_element_normal _array_element_impactnormal _array_element_physmat _array_element_hitactor) (Collision|BreakHitResult _array_element_3))
        (bind _asfort_pickup_athena (Utilities|Casting|CastToFortPickupAthena _array_element_hitactor))
        (if (Getters|IsPickupInsideCaptureArea _asfort_pickup_athena)
          (Default|HandleCaptureAreaOverlap _array_element_hitactor))))))

(fn UpdateOwningTeamScoringTimer
  (if (and (Variables|Replicated|GetIsEnabledByPhase) (Utilities|Enum|Equal(Enum) (Variables|DeviceOptions|GetPlayerCapture) "OwningTeam"))
    (if (Math|Byte|NotEqual(Byte) (Variables|Team|GetTeamIndex) (Utilities|Enum|LiteralenumEFortTeam "Spectator"))
      (PeriodicScoring|StartPeriodicScoringTimer (Variables|Team|GetTeamIndex))
      (else
        (PeriodicScoring|StopPeriodicScoringTimer)))))

(fn UpdateProgressReverseSpeed
  (if (Variables|DeviceOptions|GetOwnerCanRevertProgress)
    (bind _array_element (Utilities|Array|ForEachLoopwithBreak (Variables|CaptureArea|GetCaptureAreaTeamsInfo) :!label _label)
      (:LoopBody
        (bind (_array_element_players _array_element_insideteamindex) (Utilities|Struct|BreakCaptureAreaTeamInfo _array_element))
        (if (or (Capture|CanbeCapturedBy _array_element_insideteamindex) (Math|Byte|Equal(Byte) _array_element_insideteamindex (Variables|Team|GetTeamIndex)))
          (for _array_element_2 _array_element_players
            (if (Capture|IsValidItemFilterforCapture _array_element_2)
              (Math|Integer|IncrementInt (LocalVariables|GetNumberOfValidCaptureControllers)
                (:%20%20
                  (if (Emote|IsPlayingEmote _array_element_2)
                    (Math|Integer|IncrementInt (LocalVariables|GetNumberOfValidEmotingControllers)))))))
          (goto _label Break)))
      (:Completed
        _)))
  (CaptureComponent|UpdateProgressReverseSpeed (Variables|CaptureArea|GetCaptureComponent) (LocalVariables|GetNumberOfValidCaptureControllers) (LocalVariables|GetNumberOfValidEmotingControllers)))

(fn UpdateVisibilityComponent
  ;[0.2339 1 0.8497] Update the visibility component with the current Visibile In Game option value
  (Visibility|SetVisiblityDuringPhase (Variables|Default|GetVisibleInGameComponent) (select (Variables|DeviceOptions|GetbIsVisible) "NewEnumerator3" "NewEnumerator1")))

(fn ConstructionScript
  (bind _self self)
  (bind _returnvalue (Transformation|GetActorLocation _self))
  (bind _x (.x _returnvalue))
  (bind _y (.y _returnvalue))
  (bind _z (.z _returnvalue))
  (bind _returnvalue_2 (Math|Float|float+float _x _y _z))
  (bind _returnvalue_3 (/ _returnvalue_2 3.0))
  (bind _returnvalue_4 (Math|Float|Truncate _returnvalue_3))
  (bind _returnvalue_5 (Utilities|String|ToString(Integer) _returnvalue_4))
  (bind _returnvalue_6 (Utilities|GetDisplayName _self))
  (bind _returnvalue_7 (Utilities|String|Append _returnvalue_6 _returnvalue_5))
  (Networking|CanExecuteCosmeticEvents
    (:True
      (Utilities|FlowControl|Sequence
        (:then_0
          (Variables|ItemPreviews|SetPreviewBubbleMaterial (Rendering|Material|GetMaterial (Variables|Default|GetPreviewBubble) 0))
          (for _ (range 1 (+ (Variables|Constants|GetNumOfClientMIDsToCreate) 1))
            (bind _returnvalue_8 (Rendering|Material|CreateDynamicMaterialInstance "/Game/Packages/Fortnite_UniqueMaterials_1/BackPackShrinkingEffect.BackPackShrinkingEffect" "None" "Transient" :!class /Script/Engine.KismetMaterialLibrary))
            (Utilities|Array|Add (Variables|ItemPreviews|GetPreviewMeshMID) _returnvalue_8)))
        (:then_1
          (bind _returnvalue_9 (Rendering|Material|CreateDynamicMaterialInstance (Variables|Default|GetBaseMesh) 0 0 (Utilities|String|StringToName (Utilities|String|Append _returnvalue_7 "_B")) :!class /Script/Engine.PrimitiveComponent))
          (Variables|Default|SetCaptureAreaBaseMID _returnvalue_9))
        (:then_2
          (bind _returnvalue_10 (Rendering|Material|CreateDynamicMaterialInstance (Variables|Default|GetCaptureAreaStaticMesh) 0 0 (Utilities|String|StringToName (Utilities|String|Append _returnvalue_7 "_C")) :!class /Script/Engine.PrimitiveComponent))
          (Variables|Default|SetCaptureAreaMID _returnvalue_10))
        (:then_3
          (Variables|Default|SetDefaultPreviewStaticMeshRelativeTransform (Transformation|GetRelativeTransform (Variables|Default|GetPreviewStaticMesh))))))))
