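; Player controller component that shows a capture area's progress HUD through a Dynamic UI scene.
; It keeps a list of overlapping capture areas, tracks the most recently added, and destroys itself when a removal empties the list.
; OnRep of the active area rebinds capture progress and state events, relayed to the UI via dispatchers on cosmetic-capable instances.
; Gameplay event router listeners update the replicated in-zone ally count, forcing net updates.
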(event Custom|ReceiveCaptureProgressUpdate
  (params (NewProgress :type Float%20%28single-precision%29))
  ;[1 1 1] Updates from Capture Component
  ; Events are bound to from OnRep_ActiveCapAreaReference
  ; (We use both events because state can change w/o progress change, or vise versa - ex) Fully Captured to Contested)
  (Networking|CanExecuteCosmeticEvents
    (:True
      (Default|CallOnUpdateProgress (Variables|/Script/FortniteGame.CaptureComponent|GetCaptureProgress (Variables|References|GetActive%20Capture%20Component)) (CaptureComponent|GetCaptureState (Variables|References|GetActive%20Capture%20Component))))))

(event Custom|OnForceCaptureStateChange
  ;[1 1 1] Updates from Capture Component
  ; Events are bound to from OnRep_ActiveCapAreaReference
  ; (We use both events because state can change w/o progress change, or vise versa - ex) Fully Captured to Contested)
  (Networking|CanExecuteCosmeticEvents
    (:True
      (Default|CallOnUpdateProgress (Variables|/Script/FortniteGame.CaptureComponent|GetCaptureProgress (Variables|References|GetActive%20Capture%20Component)) (CaptureComponent|GetCaptureState (Variables|References|GetActive%20Capture%20Component)) true))))

(event Custom|OnCaptureStateChange
  (params (NewCaptureState :type ECaptureState%20Enum :default "Neutral") (PreviousCaptureState :type ECaptureState%20Enum :default "Neutral"))
  ;[1 1 1] Updates from Capture Component
  ; Events are bound to from OnRep_ActiveCapAreaReference
  ; (We use both events because state can change w/o progress change, or vise versa - ex) Fully Captured to Contested)
  (if (Utilities|Enum|NotEqual(Enum) NewCaptureState PreviousCaptureState)
    (Networking|CanExecuteCosmeticEvents
      (:True
        (Default|CallOnUpdateProgress (Variables|/Script/FortniteGame.CaptureComponent|GetCaptureProgress (Variables|References|GetActive%20Capture%20Component)) (CaptureComponent|GetCaptureState (Variables|References|GetActive%20Capture%20Component)) true)))))

(event EventBeginPlay
  (bind _returnvalue (Components|GetOwner))
  (bind (_asfort_player_controller _bsuccess) (Utilities|Casting|CastToFortPlayerController _returnvalue))
  ;[1 1 1] Begin Play
  (Networking|CanExecuteCosmeticEvents
    (:True
      (if _bsuccess
        (Variables|References|SetOwningFortPlayerController _asfort_player_controller)
        (DynamicUI|AddScene (WorldSubsystems|GetDynamicUIManager) (Variables|Default|GetDynamicUIScene) (Variables|References|GetOwningFortPlayerController))))))

(event Custom|ListenToEnterOrExitZone
  (bind _activecapareareference (Variables|References|GetActiveCapAreaReference))
  ;[1 1 1] Listen to Player Enter or Left
  (bind _returnvalue (GameplayEventRouter|GetEventRouter "/Script/FortniteGame.FortPlayspace" _activecapareareference))
  (Utilities|FlowControl|Sequence
    (:then_0
      (GameplayEventRouter|StartListeningToEvent _returnvalue "/Game/Creative/Devices/CaptureArea/GER_PlayerSwitchedTeamWhileInZone.GER_PlayerSwitchedTeamWhileInZone" _activecapareareference
        (:OnEventReceived
          (Utilities|FlowControl|Delay 0.0) ; Delay a Frame - \nWithout doing this, it'll run before the \nCapture Area Component's Teams Info Array is updated
        )) ; Listen for whenever a player within zone switches teams...
    )
    (:then_1
      (bind (_asynctaskproxy _payload) (GameplayEventRouter|StartListeningToEvent _returnvalue "/Game/Creative/Devices/CaptureArea/GER_PlayerEnteredOrLeftCapArea.GER_PlayerEnteredOrLeftCapArea" _activecapareareference)
        (:then
          _)
        (:OnEventReceived
          (bind _payload_fortplayercontroller_2_53c21d0347e40070853eeb9c6a9ee9df (Utilities|Struct|BreakGERPlayerEnteredOrLeftCapArea _payload))
          (CallFunction|UpdateAlliesCount (Team|GetActorTeam _payload_fortplayercontroller_2_53c21d0347e40070853eeb9c6a9ee9df)))) ; Listen for whenever a new player enters or exits zone...
    ))
  (CallFunction|UpdateAlliesCount (Team|GetActorTeam (Components|GetOwner))))

(event Custom|AddCaptureArea
  (params (CapAreaReference :type Building%20Prop%20Capture%20Area%20Object%20Reference))
  ;[1 1 1] Add Capture Area to List
  ; (We use a list because cap points could overlap, meaning multiple may be active at once)
  (Utilities|Array|AddUnique (Variables|References|GetCap%20Area%20Reference) CapAreaReference)
  ;[1 1 1] Assign to Capture Area
  (CallFunction|AssigntoCaptureArea))

(event Custom|RemoveCaptureArea
  (params (CapAreaReference :type Building%20Prop%20Capture%20Area%20Object%20Reference))
  ;[1 1 1] Remove Capture Area
  (CallFunction|RemoveCapturePointfromList CapAreaReference)
  ;[1 1 1] Assign to Capture Area
  (CallFunction|AssigntoCaptureArea)
  ;[1 1 1] All Capture Points Removed?
  ; Remove the Component
  (if (Utilities|Array|IsEmpty (Variables|References|GetCap%20Area%20Reference))
    (Components|DestroyComponent)))

(event EventEndPlay
  (params (EndPlayReason :type EEndPlayReason%20Enum :default "Destroyed"))
  ;[1 1 1] Cleanup
  (CallFunction|RemoveWidget))

(event Custom|UpdateAlliesCount
  (params (TeamOfPlayer :type Byte))
  (bind _returnvalue (Components|GetOwner))
  (bind _returnvalue_2 (Team|GetActorTeam _returnvalue))
  ;[1 1 1] Allies in Zone
  (if (== TeamOfPlayer _returnvalue_2)
    (bind _array_element (Utilities|Array|ForEachLoopwithBreak (Variables|/Script/FortniteGame.BuildingProp_CaptureArea|GetCaptureAreaTeamsInfo (Variables|References|GetActiveCapAreaReference)) :!label _label)
      (:LoopBody
        (if (== (.InsideTeamIndex (Utilities|Struct|BreakCaptureAreaTeamInfo _array_element)) _returnvalue_2)
          (Variables|Default|SetAlliesInCapZone (Utilities|Array|Length (Utilities|Struct|BreakCaptureAreaTeamInfo _array_element)))
          (Networking|ForceNetUpdate (Components|GetOwner))
          (goto _label Break))))))

(fn AssignToCaptureArea
  (bind _cap_area_reference (Variables|References|GetCap%20Area%20Reference))
  (bind _returnvalue (Utilities|Array|LastIndex _cap_area_reference))
  (bind _output (Utilities|Array|Get(acopy) _cap_area_reference _returnvalue))
  ;[1 1 1] Is the list Empty?
  (if (Utilities|Array|IsEmpty (Variables|References|GetCap%20Area%20Reference))
    ;[1 1 1] Clear Active Point
    (Variables|References|SetActiveCapAreaReference 0)
    (else
      ;[1 1 1] If not... Display meter for the most recently added point
      (Utilities|IsValid _output
        (:Is%20Valid
          (Variables|References|SetActiveCapAreaReference _output)
          (CallFunction|ListentoEnterorExitZone)))))
  (Networking|ForceNetUpdate (Components|GetOwner)))

(fn ForceUpdateData
  (Variables|References|GetActiveCapAreaReference
    (:then
      ;[1 1 1] Update Data
      (bind _iconoption (Class|BPICaptureAreaCaptureProgressHUD|GetDatafromBlueprint(Message) (Variables|References|GetActiveCapAreaReference)))
      (Default|CallOnCaptureZoneChanged (Creative|UI|IconPicker|GetIconLibraryTexture _iconoption))
      ;[1 1 1] Force Update
      (CallFunction|OnForceCaptureStateChange)
      ;[1 1 1] Update Teammate Count
      (Default|CallOnUpdatePlayerCount (Variables|Default|GetAlliesInCapZone)))
    (:else)))

(fn OnRep_ActiveCapAreaReference
  (Networking|CanExecuteCosmeticEvents
    (:True
      (Utilities|IsValid (Variables|References|GetActive%20Capture%20Component)
        (:Is%20Valid
          ;[1 1 1] Clear Binding of Last Active Cap Zone
          (EventDispatchers|UnbindEventfromOnCaptureProgressUpdate (Variables|References|GetActive%20Capture%20Component) (EventDispatchers|CreateEvent :!event ReceiveCaptureProgressUpdate))
          (EventDispatchers|UnbindEventfromOnCaptureStateChanged (Variables|References|GetActive%20Capture%20Component) (EventDispatchers|CreateEvent :!event OnCaptureStateChange)))
        (:Is%20Not%20Valid
          _))
      ;[1 1 1] Did we get a new valid capture point to track?
      (bind _activecapareareference (Variables|References|GetActiveCapAreaReference)
        (:then
          ;[1 1 1] Bind to Capture Updates
          (bind _capturecomponent (Variables|/Script/FortniteGame.BuildingProp_CaptureArea|GetCaptureComponent _activecapareareference)
            (:then
              (bind _output_get (Variables|References|SetActive%20Capture%20Component _capturecomponent))
              (EventDispatchers|BindEventtoOnCaptureProgressUpdate _output_get (EventDispatchers|CreateEvent :!event ReceiveCaptureProgressUpdate))
              (EventDispatchers|BindEventtoOnCaptureStateChanged _output_get (EventDispatchers|CreateEvent :!event OnCaptureStateChange))
              ;[1 1 1] Force Update Data
              (CallFunction|ForceUpdateData))
            (:else)))
        (:else
          ;[1 1 1] Remove the Widget
          (CallFunction|RemoveWidget))))))

(fn OnRep_AlliesInCapZone
  (Networking|CanExecuteCosmeticEvents
    (:True
      (Default|CallOnUpdatePlayerCount (Variables|Default|GetAlliesInCapZone)))))

(fn RemoveCapturePointFromList
  (params (CapAreaReference :type Building%20Prop%20Capture%20Area%20Object%20Reference%20%28by%20ref%29))
  (Utilities|Array|RemoveItem (Variables|References|GetCap%20Area%20Reference) CapAreaReference))

(fn RemoveWidget
  (bind _dynamicuiscene (Variables|Default|GetDynamicUIScene)
    (:then
      (bind _owningfortplayercontroller (Variables|References|GetOwningFortPlayerController)
        (:then
          (DynamicUI|RemoveScene (WorldSubsystems|GetDynamicUIManager) _dynamicuiscene _owningfortplayercontroller))
        (:else)))
    (:else)))
