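# Translucent UI material for the Playground LTM capture-objective HUD progress indicator, filling a masked container by PercentFill.
# A scrolling, repeating pulse with reversible direction brightens the fill, while ShowDamage reddens the band between CurHealth and fill level.
# Texture R, G and B channels shape interior, border and blocked icon, which overlays everything through BlockedState.
# An optional mask cuts a wedge from the G channel using two rotated, offset half-planes.
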Material /Game/Athena/HUD/LTM/Playground/Materials/M_UI_HUD_CaptureObjective_Progress
graph {
 properties {blendMode=BLEND_Translucent, materialDomain=MD_UI, pixelDepthOffsetMode=PDOM_Legacy, usedShadingModels=Default Lit}
 nodes [
  n1 = ScalarParameter() {defaultValue=1, parameterName=Capture Pulse Strength}
  n2 = ScalarParameter() {desc="0 = UP // 1 = Down", parameterName=InvertCapturePulseDirection, sliderMax=1}
  n3 = Multiply(A=@n2) {constB=-2}
  n4 = Multiply(A=Time(), B=Multiply(A=ScalarParameter() {defaultValue=0.5, parameterName=CapturePulseSpeed}, B=Add(A=@n3)))
  n5 = StaticComponentMaskParameter(None=TextureCoordinate()) {defaultG=true, parameterName=Orientation}
  n6 = Frac(None=Multiply(A=Add(A=@n4, B=@n5), B=ScalarParameter() {defaultValue=5, parameterName=CapturePulsePeriod}))
  n7 = LinearInterpolate(A=OneMinus(None=@n6), B=@n6, Alpha=@n2)
  n8 = Add(A=@n5, B=Constant() {r=-1})
  n9 = LinearInterpolate(Alpha=Clamp(None=ScalarParameter() {defaultValue=0.5, parameterName=PercentFill})) {constA=0.095, constB=0.913}
  n10 = Clamp(None=Multiply(A=Add(A=@n8, B=@n9)) {constB=128})
  n11 = LinearInterpolate(Alpha=Clamp(None=ScalarParameter() {defaultValue=1, parameterName=CurHealth})) {constA=0.095, constB=0.913}
  n12 = Clamp(None=Multiply(A=Add(A=@n8, B=@n11) {constA=-1}) {constB=128})
  n13 = DotProduct(A=@n12, B=OneMinus(None=@n10))
  n14 = Multiply(
   A=Multiply(A=Power(Base=@n7) {constExponent=3}, B=LinearInterpolate(A=@n10, B=@n13, Alpha=@n2))
   B=ScalarParameter() {parameterName=CapturePulseOpacity, sliderMax=1}
  )
  n15 = VectorParameter() {parameterName=BackingColor, defaultValue=<rgba 0.004 0.0232 0.2159 1>}
  n16 = ScalarParameter() {parameterName=ShowDamage}
  n17 = LinearInterpolate(
   A=Multiply(A=VectorParameter() {parameterName=FillColor, defaultValue=<rgba 0 0.2159 1 1>}.RGB, B=LinearInterpolate(Alpha=@n16) {constA=1, constB=1.4})
   B=Constant3Vector() {constant=<rgba 1 0 0 0>}
   Alpha=Multiply(A=@n16, B=@n13)
  )
  n18 = LinearInterpolate(A=@n10, B=@n12, Alpha=@n16)
  n19 = Clamp(None=LinearInterpolate(A=Constant() {r=0.8}, B=Constant() {r=1}, Alpha=@n18))
  n20 = Multiply(A=Add(A=Multiply(A=@n14, B=@n1)), B=LinearInterpolate(B=LinearInterpolate(A=@n15.RGB, B=@n17, Alpha=@n18), Alpha=@n19))
  n21 = Multiply(A=@n14) {constB=0.2}
  n22 = TextureSampleParameter2D(UVs=TextureCoordinate()) {parameterName=Icon_Param, texture=&/T_UI_CreativeCaptureObjectiveProgressContainer_Mask}
  n23 = DotProduct(A=OneMinus(None=@n22.R), B=@n22.G)
  n24 = VectorParameter() {parameterName=BorderColor, defaultValue=<rgba 0.0513 0.7379 1 1>}
  n25 = Add(
   A=LinearInterpolate(B=Add(A=@n20, B=@n21), Alpha=@n22.R)
   B=LinearInterpolate(
    A=LinearInterpolate(B=@n15.RGB, Alpha=@n23)
    B=LinearInterpolate(B=@n24.RGB, Alpha=@n23)
    Alpha=ScalarParameter() {defaultValue=1, parameterName=UseBorder}
   )
  )
  n26 = TextureCoordinate()
  n27 = Constant() {r=-0.05}
  n28 = Constant() {r=-0.04}
  n29 = ScalarParameter() {defaultValue=1, parameterName=AnimateLeftCenter}
  n30 = ComponentMask(None=Rotator(Coordinate=@n26, Time=LinearInterpolate(A=@n27, B=@n28, Alpha=@n29)) {speed=6.2832}) {r=true}
  n31 = LinearInterpolate(A=Constant() {r=-0.5}, B=ScalarParameter() {defaultValue=-0.44, parameterName=LeftCenter_Offset}, Alpha=@n29)
  n32 = Clamp(None=Multiply(A=Add(A=@n30, B=@n31), B=Constant() {r=80}))
  n33 = Constant() {r=0.45}
  n34 = Constant() {r=0.445}
  n35 = ScalarParameter() {defaultValue=1, parameterName=AnimateRightCenter}
  n36 = ComponentMask(None=Rotator(Coordinate=@n26, Time=LinearInterpolate(A=@n33, B=@n34, Alpha=@n35)) {speed=6.2832}) {r=true}
  n37 = LinearInterpolate(A=Constant() {r=-0.5}, B=ScalarParameter() {defaultValue=-0.485, parameterName=RightCenter_Offset}, Alpha=@n35)
  n38 = Clamp(None=Multiply(A=Add(A=@n36, B=@n37), B=Constant() {r=80}))
  n39 = DotProduct(A=@n22.G, B=OneMinus(None=DotProduct(A=@n32, B=@n38)))
  n40 = ScalarParameter() {parameterName=UseMask}
  n41 = Multiply(A=@n22.B, B=ScalarParameter() {parameterName=BlockedState, sliderMax=1})
  LinearInterpolate(
   A=LinearInterpolate(A=@n25, B=LinearInterpolate(B=@n25, Alpha=@n39), Alpha=@n40)
   B=VectorParameter() {parameterName=BlockedIconColor, defaultValue=<rgba 1 1 1 0>}.RGB
   Alpha=@n41
  )
  n42 = Add(A=@n23, B=LinearInterpolate(B=@n22.R, Alpha=@n19))
  LinearInterpolate(A=Multiply(A=LinearInterpolate(A=@n42, B=@n39, Alpha=@n40), B=ScalarParameter() {defaultValue=1, parameterName=OverallAlpha}), Alpha=@n41)
 ]
}
