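# Floating value-bar widget for the Creative Capture Area device, showing capture progress, state and team on an objective badge.
# On init it binds to the area's capture component; updates set a percent text and the icon material's PercentFill.
# Capture state drives pulse opacity and direction on the dynamic material, with colors chosen by friendly or enemy affiliation.
# It also toggles a blocked look and shows team name or per-affiliation override text.
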WidgetBlueprint /Game/Creative/Devices/CaptureArea/Creative_CaptureProgressBar
graphs [
 EventGraph {nodes [
   n1 = OnProgressUpdated!() -> (OutputDelegate, then, New Progress:float=~) {is=title K2Node_CustomEvent}
   > Update!(execute, self:Creative Floating Value Bar Object Reference, CurrentValue:int=0, MaxValue:int=1, ForcedValuePercent:double=@n1.New Progress) -> (then)
   n2 = Event InitFromObject!() -> (OutputDelegate, then, InitObject:Object Reference) {is=title K2Node_Event}
   > SetBadgeStyle()
   > Parent: InitFromObject!(execute, InitObject:Object Reference=@n2.InitObject) -> (then) {is=title K2Node_CallParentFunction}
   > n3 = Get!(execute:Exec) -> (then:Exec, else:Exec, My Actor:Actor Object Reference)
   > n4 = Cast To BuildingActor!(execute, Object=@n3) -> (then, CastFailed, AsBuilding Actor:Building Actor Object Reference) {is=title K2Node_DynamicCast}
   > n5 = Set My Building(My Building:Building Actor Object Reference=@n4) -> (Output_Get:Building Actor Object Reference)
   > n6 = Cast To BP_Creative_CaptureArea!(execute, Object=@n5) -> (then, CastFailed, AsBP Creative Capture Area:BP Creative Capture Area Object Reference) {is=title K2Node_DynamicCast}
   > Bind Event to On Capture Progress Update(
    self:Capture Component Object Reference=Get CaptureComponent!(self:Building Prop Capture Area Object Reference=@n6) -> (CaptureComponent:Capture Component Object Reference)
    Delegate=Create Event()
   ) {is=title K2Node_AddDelegate}
   > Show Objective Text(self:Creative Floating Value Bar Object Reference)
   n7 = Event Update!() -> (OutputDelegate, then, CurrentValue:int=0, MaxValue:int=0, ForcedValuePercent:double) {is=title K2Node_Event}
   n8 = int * float!(A:int=100, B:double=@n7.ForcedValuePercent) -> (ReturnValue:double)
   > "SetText (Text)"(self=Get CurrentValue() -> (CurrentValue:Text Object Reference), InText="To Text (Integer)"(Value=Truncate(A=@n8)))
   > n9 = GetDynamicMaterial(self=Get ProgressIcon() -> (ProgressIcon:Image Object Reference))
   > SetScalarParameterValue(self=@n9, ParameterName=PercentFill, Value=@n7.ForcedValuePercent)
   n10 = Event UpdateTextByAffiliation!() -> (OutputDelegate, then, TeamAffiliation:EFortTeamAffiliation Enum=Friendly) {is=title K2Node_Event}
   > n11 = Branch(Condition=Get bShowTeamName() -> (bShowTeamName:bool))
   > "SetText (Text)"(self=Get Text!() -> (Text:Text Object Reference), InText=Get TeamName() -> (TeamName:Text))
   n12 = Get!(execute=@n11.else) -> (then, else, Value Bar Component:Creative Value Bar Component Object Reference)
   n13 = Select!(Friendly:int=0, Neutral:int=1, Hostile:int=2, Index:EFortTeamAffiliation Enum=@n10.TeamAffiliation) -> (ReturnValue:int)
   n14 = Get Text Overrides!(self:Creative Value Bar Component Object Reference=@n12) -> (Text Overrides:Array of Strings)
   > "SetText (Text)"(self=Get Text(), InText="To Text (String)"(InString="Get (a copy)"(Array=@n14, Dimension 1=@n13)))
  ]}
 GetMyBuildingTeamIndex {nodes [
   GetMyBuildingTeamIndex() {is=title K2Node_FunctionEntry}
   n1 = GetOwningPlayer()
   n2 = Get My Building() -> (My Building:Building Actor Object Reference)
   > n3 = Branch(Condition=AND Boolean(A=IsValid(Object=@n1), B=IsValid(Object=@n2)))
   > Return Node!(execute, MyBuildingTeamIndex:Byte=Get TeamIndex!(self:Building Actor Object Reference=Get My Building()) -> (TeamIndex:Byte=0))
   Return Node(
    execute=@n3.else
    MyBuildingTeamIndex=Literal enum EFortTeam!(Enum:EFortTeam Enum=Spectator) -> (ReturnValue:EFortTeam Enum) {is=title K2Node_EnumLiteral}
   )
  ]}
 ResetIconColors {nodes [
   ResetIconColors() {is=title K2Node_FunctionEntry}
   > n1 = GetDynamicMaterial(self=Get ProgressIcon())
   > Set MID_ProgressIcon(MID_ProgressIcon:Material Instance Dynamic Object Reference=@n1) -> (Output_Get:Material Instance Dynamic Object Reference)
   > n2 = GetLocalPlayerAffiliation!(execute, self:Creative Floating Value Bar Object Reference) -> (then, LocalTeamAffiliation:EFortTeamAffiliation Enum=Friendly)
   n3 = Find!(
    TargetMap:Map of EFortTeamAffiliation Enums to Creative Floating Health Color Structures=Get TeamColors() -> (TeamColors:Map of EFortTeamAffiliation Enums to Creative Floating Health Color Structures=Friendly)
    Key:"EFortTeamAffiliation Enum (by ref)"=@n2
   ) -> (
    Value_BackingColor_2_#1:Linear Color Structure
    Value_BorderColor_4_#2:Linear Color Structure
    Value_FillColor_6_#3:Linear Color Structure
    Value_ContentColorAndOpacity_8_#4:Linear Color Structure
    ReturnValue
   )
   > SetVectorParameterValue(
    self=Get MID_ProgressIcon() -> (MID_ProgressIcon:Material Instance Dynamic Object Reference)
    ParameterName=Get MID_BackgroundColor() -> (MID_BackgroundColor:Name=None)
    Value=@n3
   )
   > SetVectorParameterValue(
    self=Get MID_ProgressIcon()
    ParameterName=Get MID_BorderColor() -> (MID_BorderColor:Name=None)
    Value=@n3.Value_BorderColor_4_#2
   )
   > SetVectorParameterValue(
    self=Get MID_ProgressIcon()
    ParameterName=Get MID_FillColor() -> (MID_FillColor:Name=None)
    Value=@n3.Value_FillColor_6_#3
   )
  ]}
 SetBadgeStyle {nodes [
   SetBadgeStyle() {is=title K2Node_FunctionEntry}
   > SetBrushFromMaterial(self=Get ProgressIcon(), Material=Get HUD_CaptureObjective_Material() -> (HUD_CaptureObjective_Material:Material Interface Object Reference))
  ]}
 SetIsAllowedInteraction {nodes [
   SetIsAllowedInteraction!() -> (then, IsAllowedInteraction:bool=~) {is=title K2Node_FunctionEntry}
   > n1 = GetDynamicMaterial(self=Get ProgressIcon())
   > Set MID_ProgressIcon(MID_ProgressIcon=@n1)
   > n2 = Branch(Condition=Get IsAllowedInteraction() -> (IsAllowedInteraction:bool))
   > SetScalarParameterValue(self=Get MID_ProgressIcon(), ParameterName=Get MID_BlockedParameter() -> (MID_BlockedParameter:Name=None), Value=0)
   > SetVisibility(self=Get Text_IndicatorLabel() -> (Text_IndicatorLabel:Text Object Reference), InVisibility=SelfHitTestInvisible)
   SetScalarParameterValue(execute=@n2.else, self=Get MID_ProgressIcon(), ParameterName=Get MID_BlockedParameter() -> (MID_BlockedParameter=None), Value=1)
   > SetVisibility(self=Get Text_IndicatorLabel(), InVisibility=Collapsed)
   > ResetIconColors() {is=title K2Node_CallFunction}
   > SetScalarParameterValue(self=Get MID_ProgressIcon(), ParameterName=Get MID_CaptureEffectParameter() -> (MID_CaptureEffectParameter:Name=None), Value=0)
  ]}
 UpdateCaptureState {nodes [
   UpdateCaptureState!() -> (then, NewCaptureState:ECaptureState Enum=Neutral, PreviousCaptureState:ECaptureState Enum=Neutral, StateOwner:Byte) {is=title K2Node_FunctionEntry}
   > n1 = GetDynamicMaterial(self=Get ProgressIcon())
   > Set MID_ProgressIcon(MID_ProgressIcon=@n1)
   n2 = Get NewCaptureState() -> (NewCaptureState:ECaptureState Enum=Neutral)
   > n3 = Branch(Condition=OR Boolean(A="Equal (Enum)"(A:ECaptureState Enum=@n2, B:ECaptureState Enum=Neutral), B="Equal (Enum)"(A=@n2, B=Captured)))
   > SetScalarParameterValue(self=Get MID_ProgressIcon(), ParameterName=Get MID_CaptureEffectParameter() -> (MID_CaptureEffectParameter=None), Value=0)
   > ResetIconColors()
   n4 = Branch(execute=@n3.else, Condition="Equal (Enum)"(A=Get NewCaptureState() -> (NewCaptureState=Neutral), B=Capturing))
   > n5 = SetScalarParameterValue(
    self=Get MID_ProgressIcon()
    ParameterName=Get MID_CaptureEffectDirectionParameter() -> (MID_CaptureEffectDirectionParameter:Name=None)
    Value=0
   )
   SetScalarParameterValue(
    execute=[@n5.then, @n7.then]
    self=Get MID_ProgressIcon()
    ParameterName=Get MID_CaptureEffectParameter() -> (MID_CaptureEffectParameter=None)
    Value=1
   )
   > UpdateInteractingIconColors!(execute, self, CaptureState:Byte=Get NewCaptureState() -> (NewCaptureState=Neutral), StateOwner:Byte=Get StateOwner() -> (StateOwner:Byte=0)) -> (then)
   n6 = Branch(execute=@n4.else, Condition="Equal (Enum)"(A=Get NewCaptureState() -> (NewCaptureState=Neutral), B=Neutralizing))
   > n7 = SetScalarParameterValue(self=Get MID_ProgressIcon(), ParameterName=Get MID_CaptureEffectDirectionParameter() -> (MID_CaptureEffectDirectionParameter=None), Value=1)
   Branch(execute=@n6.else, Condition="Equal (Enum)"(A=Get NewCaptureState() -> (NewCaptureState=Neutral), B=Disabled))
   > SetScalarParameterValue(self=Get MID_ProgressIcon(), ParameterName=Get MID_CaptureEffectParameter() -> (MID_CaptureEffectParameter=None), Value=0)
   > ResetIconColors()
  ]}
 UpdateInteractingIconColors {nodes [
   UpdateInteractingIconColors!() -> (then, CaptureState:Byte, StateOwner:Byte) {is=title K2Node_FunctionEntry}
   > n1 = GetDynamicMaterial(self=Get ProgressIcon())
   > Set MID_Backplate(MID_Backplate:Material Instance Dynamic Object Reference=@n1) -> (Output_Get:Material Instance Dynamic Object Reference)
   > n2 = GetLocalPlayerAffiliation() -> (LocalTeamAffiliation=Friendly)
   > Set IsFriendly(IsFriendly:bool="Equal (Enum)"(A:EFortTeamAffiliation Enum=@n2, B:EFortTeamAffiliation Enum=Friendly)) -> (Output_Get:bool)
   n3 = Byte to Enum ECaptureState!(Byte:Byte=Get CaptureState() -> (CaptureState:Byte=0)) -> (ReturnValue:ECaptureState Enum) {is=title K2Node_CastByteToEnum}
   > n4 = Branch(Condition=OR Boolean(A="Equal (Enum)"(A:ECaptureState Enum=@n3, B:ECaptureState Enum=Neutralizing), B="Equal (Enum)"(A=@n3, B=Deneutralizing)))
   n5 = Get BackgroundColor_Friendly() -> (BackgroundColor_Friendly:Color Structure)
   n6 = Get BackgroundColor_Enemy() -> (BackgroundColor_Enemy:Color Structure)
   > SetVectorParameterValue(
    self=Get MID_Backplate() -> (MID_Backplate:Material Instance Dynamic Object Reference)
    ParameterName=Get MID_BackgroundColor() -> (MID_BackgroundColor=None)
    Value=SelectColor(
     A="To LinearColor (Color)"!(InColor:Color Structure=@n5) -> (ReturnValue:Linear Color Structure)
     B="To LinearColor (Color)"(InColor=@n6)
     bPickA=Get IsFriendly() -> (IsFriendly:bool)
    )
   )
   n7 = Get FillColor_Friendly() -> (FillColor_Friendly:Color Structure)
   n8 = Get FillColor_Enemy() -> (FillColor_Enemy:Color Structure)
   n9 = GetTeamAffiliationToLocalPlayer!(
    Team:Byte=GetMyBuildingTeamIndex!(self) -> (MyBuildingTeamIndex:Byte=0) {is=title K2Node_CallFunction}
    bIncludeNonSpectators:bool=true
    bIncludeSpectators:bool=false
   ) -> (ReturnValue:EFortTeamAffiliation Enum=Friendly)
   > SetVectorParameterValue(
    self=Get MID_Backplate()
    ParameterName=Get MID_FillColor() -> (MID_FillColor=None)
    Value=SelectColor(
     A="To LinearColor (Color)"(InColor=@n7)
     B="To LinearColor (Color)"(InColor=@n8)
     bPickA="Equal (Enum)"(A:EFortTeamAffiliation Enum=@n9, B:EFortTeamAffiliation Enum=Friendly)
    )
   )
   n10 = Sequence(execute=@n4.else)
   > n11 = Branch(Condition="Equal (Byte)"!(A:Byte=GetMyBuildingTeamIndex() -> (MyBuildingTeamIndex=0), B:Byte=Literal enum EFortTeam(Enum=Spectator)) -> (ReturnValue:bool))
   n12 = Get BackgroundColor_Friendly()
   n13 = Get BackgroundColor_Enemy()
   > SetVectorParameterValue(
    self=Get MID_Backplate()
    ParameterName=Get MID_BackgroundColor() -> (MID_BackgroundColor=None)
    Value=SelectColor(A="To LinearColor (Color)"(InColor=@n12), B="To LinearColor (Color)"(InColor=@n13), bPickA=Get IsFriendly())
   )
   n14 = Get BorderColor_Friendly() -> (BorderColor_Friendly:Color Structure)
   n15 = Get BorderColor_Enemy() -> (BorderColor_Enemy:Color Structure)
   > SetVectorParameterValue(
    self=Get MID_Backplate()
    ParameterName=Get MID_BorderColor() -> (MID_BorderColor=None)
    Value=SelectColor(A="To LinearColor (Color)"(InColor=@n14), B="To LinearColor (Color)"(InColor=@n15), bPickA=Get IsFriendly())
   )
   n16 = Get FillColor_Friendly()
   n17 = Get FillColor_Enemy()
   SetVectorParameterValue(
    execute=@n11.else
    self=Get MID_Backplate()
    ParameterName=Get MID_BackgroundColor() -> (MID_BackgroundColor=None)
    Value=SelectColor(A="To LinearColor (Color)"(InColor=@n16), B="To LinearColor (Color)"(InColor=@n17), bPickA=Get IsFriendly())
   )
   n18 = Get FillColor_Friendly()
   n19 = Get FillColor_Enemy()
   n20 = GetTeamAffiliationToLocalPlayer(Team=Get StateOwner() -> (StateOwner=0), bIncludeNonSpectators=true, bIncludeSpectators=false) -> (ReturnValue=Friendly)
   SetVectorParameterValue(
    execute=@n10.then_1
    self=Get MID_Backplate()
    ParameterName=Get MID_FillColor() -> (MID_FillColor=None)
    Value=SelectColor(A="To LinearColor (Color)"(InColor=@n18), B="To LinearColor (Color)"(InColor=@n19), bPickA="Equal (Enum)"(A=@n20, B=Friendly))
   )
  ]}
]
variables [
 BackgroundColor_Enemy {type=Color, default=<rgba 0.4667 0 0 1>}
 BackgroundColor_Friendly {type=Color, default=<rgba 0.051 0.1647 0.502 1>}
 BorderColor_Enemy {type=Color, default=<rgba 1 0.6667 0.349 1>}
 BorderColor_Friendly {type=Color, default=<rgba 0.251 0.8745 1 1>}
 bShowTeamName {type=boolean}
 FillColor_Enemy {type=Color, default=<rgba 0.9451 0.3608 0.1569 1>}
 FillColor_Friendly {type=Color, default=<rgba 0 0.502 1 1>}
 HUD_CaptureObjective_Material {type=MaterialInterface, default=/Game/Athena/HUD/LTM/Playground/Materials/M_UI_HUD_CaptureObjective_Progress}
 MID_BackgroundColor {type=string, default=BackingColor}
 MID_BlockedParameter {type=string, default=BlockedState}
 MID_BorderColor {type=string, default=BorderColor}
 MID_CaptureEffectDirectionParameter {type=string, default=InvertCapturePulseDirection}
 MID_CaptureEffectParameter {type=string, default=CapturePulseOpacity}
 MID_FillColor {type=string, default=FillColor}
 My Building {type=BuildingActor, default=null}
 TeamName {type=string, default=False}
]
widgets {
 info {
  inheritedWidgetCount=0
  parentClass=/Game/Creative/Devices/Common/Components/Creative_FloatingValueBar.Creative_FloatingValueBar_C
  rootWidgetClass=null
  widgetCount=0
 }
 widgets []
}
