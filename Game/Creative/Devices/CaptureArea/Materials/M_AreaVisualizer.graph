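# Translucent, unlit, two-sided material that visualizes a Creative Capture Area device's volume.
# A depth-based distance function raised to a power forms an intersection highlight, and back faces add a bottom-to-top fading gradient from bounds-normalized local height.
# The intersection mask also blends between two color parameters.
# A time-driven sine with offset, speed and minimum-brightness parameters pulses that color, with a debug scalar readout attached.
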Material /Game/Creative/Devices/CaptureArea/Materials/M_AreaVisualizer
graph {
 properties {bAutomaticallySetUsageInEditor=false, blendMode=BLEND_Translucent, shadingModel=MSM_Unlit, twoSided=true, usedShadingModels=Unlit}
 nodes [
  n1 = Power(
   Base=MaterialFunctionCall!(
    Distance Bias=ScalarParameter() {defaultValue=5, parameterName=Distance Bias}
    Distance Divisor=ScalarParameter() {defaultValue=30, parameterName=Distance Divisor}
    Scene Depth
    Pixel Depth
   ) {materialFunction=/Game/Effects/Fort_Effects/Materials/_Functions/MF_DistanceBehindTranslucency_01}.Inverted
   Exp=ScalarParameter() {defaultValue=10, parameterName=IntersectionExponent}
  )
  n2 = ScalarParameter() {defaultValue=1, parameterName=BaseGradientScale}
  n3 = Transform(None=WorldPosition().XYZ) {transformSourceType=TRANSFORMSOURCE_World, transformType=TRANSFORM_Local}
  n4 = Transform(None=ObjectPositionWS()) {transformSourceType=TRANSFORMSOURCE_World, transformType=TRANSFORM_Local}
  n5 = Add(A=Divide(A=Subtract(A=@n3, B=@n4), B=ObjectLocalBounds().Extents)) {constB=0.5}
  n6 = OneMinus(None=Saturate(None=ComponentMask(None=@n5) {b=true}))
  LinearInterpolate(A=Add(A=@n1, B=Multiply(A=@n2, B=@n6)), B=@n1, Alpha=Saturate(None=TwoSidedSign()))
  n7 = Time()
  n8 = ScalarParameter() {defaultValue=0.5, group=Pulse, parameterName=Pulse Time Offset}
  n9 = Sine(None=Multiply(A=Add(A=@n7, B=@n8), B=ScalarParameter() {defaultValue=0.5, group=Pulse, parameterName=Pulse Speed}))
  n10 = ScalarParameter() {defaultValue=0.2, group=Pulse, parameterName=Pulse Min Brightness}
  n11 = Add(A=Multiply(A=ConstantBiasScale(None=@n9), B=OneMinus(None=@n10)), B=@n10)
  MaterialFunctionCall(Number=@n11, MaximumNumberOfDigits=Constant() {r=3}) {materialFunction=&DebugScalarValues}
  Multiply(
   A=@n11
   B=LinearInterpolate(
    A=VectorParameter() {parameterName=ColorHigh, defaultValue=<rgba 2 0 0 1>}.RGB
    B=VectorParameter() {parameterName=IntersectionColor, defaultValue=<rgba 10 0 0 0>}.RGB
    Alpha=@n1
   )
  )
 ]
}
