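# Translucent, unlit, two-sided material for the Creative Capture Area device's volume, showing team color, capture progress, pulses and blocked state.
# Scale-tiled grid, noise blocks, fresnel and depth intersection form the base, tinted by LUT or direct colors.
# Static switches add progress waves, scrolling lock icons and timestamp-driven pulse and spawn/despawn animation, using stepped masks and curve atlas for WPO.
# Mobile shading paths use cheaper progress masks and icon sampling.
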Material /Game/Creative/Devices/CaptureArea/Materials/M_CaptureArea_CaptureVolume
graph {
 properties {
  bEnableResponsiveAA=true
  bUsedWithNanite=true
  bUsedWithSplineMeshes=true
  blendMode=BLEND_Translucent
  pixelDepthOffsetMode=PDOM_Legacy
  shadingModel=MSM_Unlit
  twoSided=true
  usedShadingModels=Unlit
 }
 nodes [
  n1 = Multiply(A=Time(), B=ScalarParameter() {defaultValue=1, group=Animation - Capture, parameterName=Capture Waves Speed}) {constA=3, constB=0.5}
  n2 = ComponentMask(None=TextureCoordinate()) {g=true}
  n3 = LinearInterpolate(A=@n2, B=OneMinus(None=@n2), Alpha=ScalarParameter() {group=Animation - Capture, parameterName=Invert Capture Direction})
  n4 = Saturate(None=Sine(None=Add(A=@n3, B=@n1)) {period=0.2})
  n5 = Multiply(A=@n4, B=@n4) {desc=Flow waves}
  n6 = ScalarParameter() {defaultValue=0.6, group=Animation - Capture, parameterName=Capture Waves Opacity}
  n7 = ScalarParameter() {group=Animation - Capture, parameterName=Capture Progress, sortPriority=-1}
  n8 = LinearInterpolate(Alpha=@n7) {constA=-0.6, constB=0.6}
  n9 = MaterialFunctionCall(In=Add(A=@n3, B=@n8) {constA=0.6}, Contrast=Constant() {r=5}) {materialFunction=&CheapContrast}
  n10 = Sign(None=Subtract(A=@n9.Result) {constB=0.5})
  n11 = Saturate(None=Multiply(A=OneMinus(None=@n9.Result), B=Sign(None=@n9.Result)))
  n12 = Multiply(A=@n11, B=Ceil(None=@n11))
  n13 = Add(
   A=Multiply(A=Multiply(A=@n5, B=@n6), B=Saturate(None=@n10) {desc=Progress Mask})
   B=Multiply(
    A=Power(Base=@n12) {constExponent=5}
    B=ScalarParameter() {defaultValue=5, group=Animation - Capture, parameterName=Capture Interface Intensity}
   ) {constB=10}
  )
  n14 = ScalarParameter() {desc="Turn on during capture, turn off after", group=Animation - Capture, parameterName=Capture Effect Opacity}
  n15 = VectorParameter() {desc=Use Direct Color rather than the LUT if set to do so., parameterName=Color, defaultValue=<rgba 1 1 1 1>}
  n16 = StaticSwitchParameter(
   True=@n15.RGB
   False=MaterialFunctionCall!(Mask=Constant() {r=1}, Index) {materialFunction=/Game/Creative/Materials/MF_CreativeColors}.Result
  ) {parameterName=UsesDirectColorInput}
  n17 = VectorParameter() {desc=Use Direct Color rather than the LUT if set to do so., parameterName=Color Capturing, defaultValue=<rgba 1 1 1 1>}
  n18 = StaticSwitchParameter(
   True=@n17.RGB
   False=MaterialFunctionCall(Mask=Constant() {r=1}, Index=ScalarParameter() {defaultValue=5, group=Animation - Capture, parameterName=TeamColor 2, sliderMax=64}) {materialFunction=&MF_CreativeColors}.Result
  ) {parameterName=UsesDirectColorInput}
  n19 = Multiply(A=Add(A=Multiply(A=@n13, B=@n14)), B=LinearInterpolate(A=@n16, B=@n18, Alpha=Saturate(None=@n9.Result)))
  n20 = Sign(None=Subtract(A=Add(A=@n8, B=@n3)) {constB=0.5})
  n21 = ShadingPathSwitch(Default=@n19, Deferred=@n19, Mobile=LinearInterpolate(A=@n16, B=@n18, Alpha=Saturate(None=@n20) {desc=Progress Mask})) {desc=Migrated from FeatureLevelSwitch}
  n22 = LinearInterpolate(A=@n16, B=@n18, Alpha=@n7)
  n23 = ScalarParameter() {group=Animation - Capture, parameterName=Simple Capture}
  n24 = VectorParameter() {
   desc="R = Spawn TimeStamp (sec)\r\nG = Transition Length (sec)\r\nB = Despawn TimeStamp (sec)\r\nA =  Switcher (0=Spawn / 0.5=Despawn / 1=pulse)\r\n"
   group="Animation - Spawn && Despawn"
   parameterName=TimeStampInfo
  }
  n25 = ComponentMask(None=TextureCoordinate()) {g=true}
  n26 = Multiply(A=OneMinus(None=@n25)) {constB=4}
  n27 = ScalarParameter() {defaultValue=0.5, group=Animation - Pulse, parameterName=Pulse Radius}
  n28 = Floor(None=@n24.A)
  n29 = Saturate(None=Subtract(A=@n24.A, B=@n28))
  n30 = Add(A=Multiply(A=Add(A=@n27, B=@n29)) {constB=2})
  n31 = MaterialFunctionCall(
   TimestampA=@n24.R
   TransitionLength_In=@n24.R
   TimestampB=@n24.R
   TransitionLength_Out=Multiply(A=@n24.G) {constB=0.75}
   Transition In Only?=StaticBool()
  ) {materialFunction=&MF_Timestamp_InOut_01}
  n32 = SphereMask(A=Divide(A=Floor(None=@n26)) {constB=4}, B=Subtract(A=Multiply(A=@n30, B=@n31.Linear), B=@n27), Radius=@n27, Hardness=Constant()) {desc=Stepped Pulse}
  n33 = Multiply(A=Multiply(A=@n32, B=@n25), B=@n28)
  n34 = Multiply(A=@n33, B=@n24.A)
  n35 = Constant()
  n36 = ScalarParameter() {parameterName=Pulse}
  n37 = TextureCoordinate()
  n38 = Multiply(
   A=ScalarParameter() {defaultValue=1, parameterName=BaseGradientScale}
   B=SphereMask(
    A=ComponentMask(None=@n37) {g=true}
    B=ScalarParameter() {defaultValue=1, parameterName=BaseGradientCenter}
    Radius=ScalarParameter() {defaultValue=0.25, parameterName=BaseGradientRadius}
    Hardness=ScalarParameter() {defaultValue=0.1, parameterName=BaseGradientHardness}
   )
  )
  n39 = SphereMask(A=ComponentMask(None=TextureCoordinate()) {g=true}, B=Constant() {r=0.01}, Radius=Constant() {r=0.008}, Hardness=Constant())
  n40 = ScalarParameter() {defaultValue=5, parameterName=TopLineExponent}
  n41 = Power(
   Base=MaterialFunctionCall!(
    Distance Bias=ScalarParameter() {defaultValue=5, parameterName=DistanceBias}
    Distance Divisor=ScalarParameter() {defaultValue=30, parameterName=Distance Divisor}
    Scene Depth
    Pixel Depth
   ) {materialFunction=/Game/Effects/Fort_Effects/Materials/_Functions/MF_DistanceBehindTranslucency_01}.Inverted
   Exp=ScalarParameter() {defaultValue=10, parameterName=IntersectionExponent}
  )
  n42 = StaticSwitchParameter(True=Add(A=Power(Base=@n39, Exp=@n40), B=@n41), False=@n41) {defaultValue=true, parameterName=UseTopLine}
  n43 = OneMinus(None=MaterialFunctionCall(
    Brightness=ScalarParameter() {defaultValue=1.25, parameterName=FresnelBrightness}
    Exponent=ScalarParameter() {defaultValue=3, parameterName=FresnelExponent}
   ) {materialFunction=&MF_Fresnel_FX}.Fresnel Double Sided)
  n44 = MaterialFunctionCall() {materialFunction=/Engine/Functions/Engine_MaterialFunctions02/WorldPositionOffset/ObjectScale}
  n45 = Divide(A=Divide(A=Add(A=@n44.Scale X, B=@n44.Scale X)), B=ScalarParameter() {defaultValue=30, desc=.75, parameterName=GridTileU})
  n46 = Divide(A=@n44.Scale Z, B=ScalarParameter() {defaultValue=100, parameterName=GridTileV})
  n47 = VertexInterpolator(VS=Ceil(None=AppendVector(A=@n45, B=@n46)))
  n48 = TextureSample(UVs=Multiply(A=TextureCoordinate(), B=@n47.PS)) {texture=/Game/Weapons/Prototype/ObjectMover/Textures/T_Movetool_Transition}
  n49 = ScalarParameter() {defaultValue=0.25, parameterName=GridOpacityInner}
  n50 = ScalarParameter() {defaultValue=0.9, parameterName=GridOpacityOuter}
  n51 = Multiply(A=TextureCoordinate(), B=Multiply(A=@n47.PS) {constB=4})
  n52 = Time()
  n53 = ScalarParameter() {defaultValue=0.6, parameterName=BlockSpeed}
  n54 = Sine(None=Add(
    A=MaterialFunctionCall!(PrimaryTextureCoordinates=@n51, 2ndOrderTextureCoordinates, RandomSeed, NoiseDensityBias, Bypass Noise Influence Crosstalk) {materialFunction=/Game/Effects/Fort_Effects/Materials/_Functions/MF_RandomUVSample}.NoiseWithRange
    B=Multiply(A=@n52, B=@n53)
   ))
  n55 = ScalarParameter() {parameterName=BlocksOpacityInnerFaces}
  n56 = ScalarParameter() {defaultValue=0.25, parameterName=BlocksOpacity}
  n57 = Saturate(None=TwoSidedSign())
  n58 = Multiply(A=OneMinus(None=@n48.R), B=Multiply(A=ConstantBiasScale(None=@n54), B=LinearInterpolate(A=@n55, B=@n56, Alpha=@n57)))
  n59 = Add(A=Multiply(A=@n48.R, B=LinearInterpolate(A=@n49, B=@n50, Alpha=@n43)), B=@n58)
  n60 = Add(A=@n38, B=Add(A=@n42, B=Add(A=@n43, B=@n59)))
  n61 = LinearInterpolate(
   A=LinearInterpolate(
    A=VectorParameter() {parameterName=ColorLow, defaultValue=<rgba 0.3028 0.1103 1 1>}.RGB
    B=VectorParameter() {parameterName=ColorHigh, defaultValue=<rgba 0.1281 0.7405 1 1>}.RGB
    Alpha=Add(A=@n36, B=@n60)
   )
   B=VectorParameter() {parameterName=ColorBlocks, defaultValue=<rgba 0.9253 0.0636 0.5483 1>}.RGB
   Alpha=@n58
  )
  n62 = VectorParameter() {parameterName=IntersectionColor, defaultValue=<rgba 1.6503 2.2282 10 0>}
  n63 = TextureCoordinate()
  n64 = AppendVector(A=Constant(), B=ScalarParameter() {parameterName=VBias})
  n65 = ScalarParameter() {defaultValue=4, parameterName=RingSteps}
  n66 = Multiply(A=Multiply(A=@n47.PS, B=Add(A=@n63, B=@n64)), B=@n65)
  n67 = SphereMask(
   A=Divide(A=Floor(None=@n66), B=@n65)
   B=Divide(A=@n47.PS)
   Radius=ScalarParameter() {defaultValue=0.3, parameterName=ImpactParam}
   Hardness=Constant()
  )
  n68 = ScalarParameter() {defaultValue=0.3, parameterName=RingCenter}
  n69 = ScalarParameter() {defaultValue=0.3, parameterName=RingRadius}
  n70 = ScalarParameter() {defaultValue=0.2, parameterName=RingHardness}
  n71 = Multiply(A=Multiply(A=@n67, B=SphereMask(A=@n67, B=@n68, Radius=@n69, Hardness=@n70)), B=ScalarParameter() {parameterName=RingOpacity})
  n72 = Saturate(None=Add(A=@n42, B=@n71))
  n73 = Add(
   A=Desaturation(None=LinearInterpolate(A=@n61, B=@n62.RGB, Alpha=@n72), Fraction=ScalarParameter() {defaultValue=0.75, parameterName=DesaturationPercentage})
   B=StaticSwitchParameter(True=@n34, False=@n35) {desc=Color Ad, group=Animation - Pulse, parameterName=Enable Pulse Animations, sortPriority=-10}
  ) {desc=Add pulse to color intensity}
  n74 = Multiply(A=Ceil(None=@n29), B=ScalarParameter() {defaultValue=10, group=Animation - Capture, parameterName=Capture Pulse Intensity Add}) {desc="Increase intensity of effect when capturing, but not when pulsing"}
  n75 = StaticSwitchParameter(True=Add(A=Multiply(A=@n34, B=@n74)), False=Constant() {r=1}) {desc=Color Multiply, group=Animation - Pulse, parameterName=Enable Pulse Animations, sortPriority=-10}
  n76 = MaterialFunctionCall() {materialFunction=&ObjectScale}
  n77 = Divide(A=@n76.Scale X) {constB=512}
  n78 = Multiply(A=Multiply(A=@n77, B=Constant() {r=1.333}) {desc=Aspect Ratio correction}) {constB=3.1416, desc=PI}
  n79 = Ceil(None=@n78)
  n80 = Constant() {r=1}
  n81 = MaterialFunctionCall(Float2=Multiply(A=TextureCoordinate(), B=AppendVector(A=@n79, B=@n80))) {materialFunction=&BreakOutFloat2Components}
  n82 = AppendVector(A=@n81.R, B=@n81.R)
  n83 = Panner(Coordinate=@n82) {speedX=0.1}
  n84 = ScalarParameter() {defaultValue=1.25, group=Blocked State, parameterName=Spacing}
  n85 = Max(
   A=Floor(None=Multiply(A=@n78, B=@n84))
   B=ScalarParameter() {defaultValue=4, desc=Minimum number of instances, group=Blocked State, parameterName=Minimum Instances}
  )
  n86 = Divide(A=ScalarParameter() {defaultValue=-0.05, group=Blocked State, parameterName=Blocked Icon Scrolling Speed}, B=Max(A=@n77))
  n87 = Constant()
  n88 = MaterialFunctionCall(Float2=Panner(Speed=AppendVector(A=@n86, B=@n87))) {materialFunction=&BreakOutFloat2Components}
  n89 = Divide(A=Frac(None=Multiply(A=@n85, B=@n88.R) {constB=4}), B=@n84)
  n90 = Multiply(A=Divide(B=Clamp(None=@n76.Scale X) {maxDefault=512}) {constA=512, constB=512}) {constB=0.75, desc=Operations for keeping the icon centered}
  n91 = Subtract(A=Divide(B=@n90) {constA=0.5}) {constB=0.5}
  n92 = Subtract(A=@n76.Scale X) {constB=511, desc="Must be 511, not 512"}
  n93 = LinearInterpolate(A=Multiply(A=Add(A=@n88.G, B=@n91) {constA=-0.5, constB=-0.25}, B=@n90) {constB=12}, B=@n88.G, Alpha=Saturate(None=Sign(None=@n92))) {desc=Scales UV's Y channel when capture area radius is smaller than 512}
  n94 = TextureSample(UVs=AppendVector(A=Saturate(None=@n89), B=Saturate(None=@n93))) {samplerType=SAMPLERTYPE_LinearColor, texture=/Game/Creative/Devices/CaptureArea/Textures/T_CaptureAreaLockedState_M}
  n95 = Multiply(A=@n82, B=Constant2Vector() {g=1.5, r=1})
  n96 = Constant2Vector() {g=-0.25}
  n97 = TextureSample(UVs=Panner(Coordinate=Add(A=@n95, B=@n96)) {speedX=-0.1}) {
   desc="Cheap version, for mobile"
   samplerType=SAMPLERTYPE_LinearColor
   texture=&T_CaptureAreaLockedState_M
  }
  n98 = Time()
  n99 = VectorParameter() {desc="On StartTime // On State (0-1)  // Off StartTime // AnimationDuration (sec)", parameterName=Blocked State Animation Info}
  n100 = Subtract(A=@n98, B=@n99.R)
  n101 = Add(A=@n99.B, B=@n99.B)
  n102 = Max(A=Multiply(A=Multiply(A=@n100, B=@n99.G), B=Divide(B=@n99.A)), B=Divide(A=Subtract(A=@n101, B=@n98), B=@n99.A))
  n103 = Multiply(
   A=Add(
    A=TextureSample(UVs=@n83) {samplerType=SAMPLERTYPE_LinearColor, texture=&T_CaptureAreaLockedState_M}.G
    B=ShadingPathSwitch(Default=@n94.R, Deferred=@n94.R, Mobile=@n97.R) {desc=Migrated from FeatureLevelSwitch}
   )
   B=StaticSwitchParameter(True=ScalarParameter() {group=Blocked State, parameterName=BlockedState}, False=Saturate(None=@n102)) {group="<DEBUG MODE>", parameterName=DEBUG MODE}
  )
  n104 = Constant()
  Multiply(
   A=StaticSwitchParameter(True=LinearInterpolate(A=@n21, B=@n22, Alpha=@n23), False=@n16) {group=Animation - Capture, parameterName=Can Be Captured, sortPriority=-10}
   B=Add(
    A=Multiply(A=@n73, B=@n75)
    B=StaticSwitchParameter(True=@n103, False=@n104) {group=Blocked State, parameterName=Enable Blocked State Functionality, sortPriority=-10}
   ) {desc=Add blocked state to color intensity}
  )
  n105 = Time()
  Multiply(A=ConstantBiasScale(None=Sine(None=@n105) {period=4})) {constB=10}
  n106 = Add(
   A=ScalarParameter() {defaultValue=0.1, group=Animation - Pulse, parameterName=Pulse Opacity Add}
   B=Multiply(A=@n29) {constB=0.6, desc=Increased Opacity}
  )
  n107 = Add(A=@n38, B=LinearInterpolate(A=ScalarParameter() {defaultValue=0.25, parameterName=FresnelInnerOpacity}, Alpha=@n43))
  n108 = Add(A=@n59, B=ScalarParameter() {defaultValue=0.1, parameterName=BlocksBase})
  n109 = Add(A=Add(A=@n42, B=Multiply(A=@n107, B=@n108)), B=@n71)
  n110 = SphereMask(
   A=ComponentMask(None=TextureCoordinate()) {g=true}
   B=ScalarParameter() {defaultValue=1, parameterName=FalloffGradientCenter}
   Radius=ScalarParameter() {defaultValue=1, parameterName=FalloffGradientRadius}
   Hardness=ScalarParameter() {defaultValue=0.6, parameterName=FalloffGradientHardness}
  )
  n111 = ScalarParameter() {defaultValue=100, parameterName=NearCameraFadeDistance}
  n112 = Add(
   A=Multiply(
    A=Multiply(A=@n109, B=@n110)
    B=MaterialFunctionCall(Distance=@n111) {materialFunction=&Near_Camera_Fade}.Result
   )
   B=StaticSwitchParameter(True=Multiply(A=@n33, B=@n106), False=@n35) {desc=Opacity Add, group=Animation - Pulse, parameterName=Enable Pulse Animations, sortPriority=-10}
  ) {desc=Add pulse to opacity}
  n113 = Add(A=Multiply(A=@n43) {constB=0.5}) {constB=0.5}
  n114 = Divide(A=Subtract(A=ScalarParameter() {defaultValue=0.2, group=Blocked State, parameterName=Inward Opacity})) {constB=-2}
  n115 = Multiply(A=TwoSidedSign(), B=@n114) {constB=0.4}
  n116 = OneMinus(None=@n114)
  n117 = StaticSwitchParameter(True=Multiply(A=Multiply(A=@n103, B=@n113), B=Add(A=@n115, B=@n116) {constB=0.6}), False=Constant()) {group=Blocked State, parameterName=Enable Blocked State Functionality, sortPriority=-10}
  n118 = Multiply(A=@n12, B=OneMinus(None=@n23))
  n119 = Constant()
  n120 = Constant() {r=1}
  n121 = If(A=@n24.A, B=Constant() {r=0.5}, "A > B"=@n120, "A == B"=@n31.Linear, "A < B"=@n120)
  Multiply(
   A=Add(
    A=Add(A=@n112, B=@n117) {desc=Add blocked state to opacity}
    B=StaticSwitchParameter(True=@n118, False=@n119) {group=Animation - Capture, parameterName=Can Be Captured, sortPriority=-10}
   )
   B=StaticSwitchParameter(True=VertexInterpolator(VS=@n121).PS, False=Constant() {r=1}) {desc=Sapwn / Despawn WPO, group="Animation - Spawn && Despawn", parameterName=Enable Spawn Animations, sortPriority=-10}
  )
  n122 = MaterialFunctionCall(Float3=TransformPosition(None=WorldPosition().XYZ) {transformSourceType=TRANSFORMPOSSOURCE_World}) {materialFunction=&BreakOutFloat3Components}
  n123 = Ceil(None=@n24.A)
  n124 = Multiply(A=Saturate(None=Subtract(A=@n123)), B=Sine(None=Min(A=@n31.Linear) {constB=0.5}))
  n125 = OneMinus(None=@n25)
  n126 = MaterialFunctionCall(
   X=Multiply(A=@n122.R, B=@n32)
   Y=Multiply(A=@n122.G, B=@n32)
   Z=Multiply(A=Multiply(A=@n124, B=@n125), B=ScalarParameter() {defaultValue=2, group=Animation - Capture, parameterName=Capture Animation Height})
  ) {materialFunction=&MakeFloat3}
  n127 = ScalarParameter() {defaultValue=0.05, group=Animation - Pulse, parameterName=Pulse WPO Strength}
  n128 = Divide(B=MaterialFunctionCall() {materialFunction=&ObjectScale}.Scale X) {constA=512, constB=512, desc=Scale WPO based on size}
  n129 = StaticSwitchParameter(True=Multiply(A=Multiply(A=@n126.Result, B=@n28), B=Multiply(A=@n127, B=@n128)), False=Constant3Vector()) {group=Animation - Pulse, parameterName=Enable Pulse Animations, sortPriority=-10}
  n130 = CurveAtlasRowParameter(CurveTime=@n31.Linear) {
   atlas=/Game/Creative/Devices/CaptureArea/Curves/CapturePointAnimationAtlas
   curve=/Game/Creative/Devices/CaptureArea/Curves/Curve_CaptureAreaAnimationSmallAreas
   desc=Spawn Animation Curves
   group="Animation - Spawn && Despawn"
   parameterName=Animation Curve
  }
  n131 = Multiply(A=@n24.A) {constB=2}
  n132 = OneMinus(None=@n131)
  n133 = Subtract(A=Add(A=Multiply(A=@n130.G, B=@n132), B=Multiply(A=@n130.A, B=@n131)))
  n134 = Multiply(A=@n122.R, B=@n133)
  n135 = Multiply(A=@n122.G, B=@n133)
  n136 = Multiply(A=@n130.R, B=@n132)
  n137 = Multiply(A=@n130.B, B=@n131)
  n138 = Multiply(A=@n122.B, B=Subtract(A=Add(A=@n136, B=@n137)))
  n139 = StaticSwitchParameter(
   True=Multiply(
    A=OneMinus(None=@n28)
    B=MaterialFunctionCall(X=@n134, Y=@n135, Z=@n138) {materialFunction=&MakeFloat3}.Result
   )
   False=Constant3Vector()
  ) {desc=Sapwn / Despawn WPO, group="Animation - Spawn && Despawn", parameterName=Enable Spawn Animations, sortPriority=-10}
  Reroute(None=Transform(None=Add(A=@n129, B=@n139)) {transformSourceType=TRANSFORMSOURCE_Local})
 ]
}
