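# Translucent unlit material for the Creative Capture Area device's preview bubble, with a tintable emissive color, blue by default.
# A radial UV gradient is inverted, raised to a power and sphere-masked into a band with adjustable center and radius.
# A sine-over-time sweep, modulated by that radial gradient, is added with its own opacity control.
# AlignMeshToTheCamera, fed a custom object basis, turns the mesh to face the viewer.
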Material /Game/Creative/Devices/CaptureArea/Materials/M_CaptureArea_PreviewBubble
graph {
 properties {
  bEnableResponsiveAA=true
  bUsedWithNanite=true
  blendMode=BLEND_Translucent
  pixelDepthOffsetMode=PDOM_Legacy
  shadingModel=MSM_Unlit
  usedShadingModels=Unlit
 }
 nodes [
  n1 = Constant() {r=0.5}
  n2 = SphereMask(A=TextureCoordinate(), B=@n1, Radius=@n1, Hardness=Constant())
  n3 = OneMinus(None=@n2)
  n4 = ScalarParameter() {defaultValue=4, parameterName=Exp}
  n5 = Add(A=TextureCoordinate()) {constB=0.5}
  n6 = Sine(None=Time()) {period=2}
  n7 = OneMinus(None=Multiply(A=@n2, B=Multiply(A=@n6) {constB=4}))
  n8 = ScalarParameter() {defaultValue=0.5, parameterName=SweepRadius}
  n9 = ScalarParameter() {defaultValue=0.7, parameterName=SweepHardness}
  Add(
   A=SphereMask(
    A=Power(Base=@n3, Exp=@n4)
    B=ScalarParameter() {defaultValue=0.8, parameterName=Center}
    Radius=ScalarParameter() {defaultValue=0.2, parameterName=Radius}
    Hardness=Constant()
   )
   B=Multiply(A=SphereMask(A=@n5, B=@n7, Radius=@n8, Hardness=@n9), B=ScalarParameter() {defaultValue=0.2, parameterName=SweepOpacity})
  )
  MaterialFunctionCall!(
   WorldPosition=WorldPosition().XYZ
   PivotLocation
   WS Normals
   Custom Object Basis 1=Constant3Vector() {constant=<rgba 0 1 0 1>}
   Custom Object Basis 2=Constant3Vector() {constant=<rgba 0 0 -1 0>}
   Custom Object Basis 3
  ) {materialFunction=/Engine/Functions/Engine_MaterialFunctions02/WorldPositionOffset/AlignMeshToTheCamera}
  VectorParameter() {parameterName=EmissiveColor, defaultValue=<rgba 0.1128 0.8091 5 1>}
 ]
}
