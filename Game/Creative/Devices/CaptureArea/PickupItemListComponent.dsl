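; Item list component for the Creative Capture Area device that turns dropped pickups into stored item entries.
; AddItemToQueue accepts valid pickups from players who can create before the minigame starts, ignoring repeat bounce hits, and sets a timer.
; On timeout, the authority copies each pickup's entries into the list, broadcasts the addition and destroys the world actor.
; After spawning it imports persisted spawn items, then clears those legacy arrays.
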(event EventPostFinishSpawning
  (ItemData|AddItemsfromSpawnParams (Variables|Persisted|GetItemsToSpawn))
  (|CallParentFunction)
  (Utilities|Array|Clear (Variables|Persisted|GetItemsToSpawn))
  (Utilities|Array|Clear (Variables|Persisted|GetItemList)))

(event ItemData|EventNotifyItemListChanged
  (Default|CallOnItemListUpdated))

(fn AddItemsToList
  (bind _returnvalue (Components|GetOwner))
  (Utilities|IsValid _returnvalue
    (:Is%20Valid
      (if (Networking|HasAuthority _returnvalue)
        (for _array_element (Variables|Default|GetItemsToBeAddedQueue)
          (bind _output_get (LocalVariables|SetCurrentPickup _array_element))
          (bind _asfort_capturable_item_interface (Utilities|Casting|CastToFortCapturableItemInterface _output_get))
          (for _array_element_2 (CapturableItemInterface|GetCaptureItemEntries _asfort_capturable_item_interface)
            (ItemData|AddItemEntry _array_element_2))
          (Default|CallOnItemAdded (LocalVariables|GetCurrentPickup))
          (Actor|DestroyActor (LocalVariables|GetCurrentPickup) :!pinned true) ; Destroy the pick-up that is in the world.  Hidden by FX
        )
        (Utilities|Time|InvalidateTimerHandle (Variables|Default|GetAddItemsTimerHandle))
        (Utilities|Array|Clear (Variables|Default|GetItemsToBeAddedQueue))))))

(fn AddItemToQueue
  (params (NewItem :type Actor%20Object%20Reference))
  (bind _asfort_capturable_item_interface (Utilities|Casting|CastToFortCapturableItemInterface NewItem))
  (Utilities|FlowControl|Sequence
    (:then_0
      (if (and (CapturableItemInterface|CanbeAddedtoCaptureList _asfort_capturable_item_interface (Variables|Default|GetInvalidItemTags)) (Interact|CanInteractwithOwner (CapturableItemInterface|GetPawnTryingtoCaptureItem _asfort_capturable_item_interface)))
        ;[0.9913 1 0 0.5] Every bounce of the pickup causes a hit trigger.  This ensures rest of script only fires once
        (if (not (Utilities|Array|ContainsItem (Variables|Default|GetItemsToBeAddedQueue) (Variables|GetNewItem)))
          (Utilities|Array|Add (Variables|Default|GetItemsToBeAddedQueue) (Variables|GetNewItem))
          (LocalVariables|SetbSuccess true)
          (bind _returnvalue (Utilities|Time|SetTimerbyEvent (EventDispatchers|CreateEvent :!event AddItemsToList) (Variables|Default|GetAddItemsTimeout)))
          (Variables|Default|SetAddItemsTimerHandle _returnvalue)
          (CapturableItemInterface|OnAddedtoCaptureList _asfort_capturable_item_interface))))
    (:then_1
      (return (LocalVariables|GetbSuccess)))))

(fn CanBeDeletedFromCreativeVolume
  (return false))

(fn CanBeSavedToCreativeVolume
  (return false))

(fn CanInteractWithOwner
  (params (InteractingPawn :type Fort%20Pawn%20Object%20Reference))
  (bind _interactingpawn (Variables|GetInteractingPawn))
  (bind _returnvalue (Pawn|GetController _interactingpawn))
  (Utilities|FlowControl|Sequence
    (:then_0
      (Utilities|IsValid InteractingPawn
        (:Is%20Valid
          (Utilities|IsValid _returnvalue
            (:Is%20Valid
              (bind _asfort_player_controller_athena (Utilities|Casting|CastToFortPlayerControllerAthena _returnvalue))
              (if (Creative|CanPlayerCreateinCurrentExperience _asfort_player_controller_athena)
                (Utilities|IsValid (Creative|GetMinigame _asfort_player_controller_athena)
                  (:Is%20Valid
                    (LocalVariables|SetbSuccess (not (Minigame|HasStarted (Creative|GetMinigame _asfort_player_controller_athena) :!class /Script/FortniteGame.FortMinigame))))
                  (:Is%20Not%20Valid
                    (LocalVariables|SetbSuccess true)))))))))
    (:then_1
      (return (LocalVariables|GetbSuccess)))))

(fn ClearItemList
  (Utilities|FlowControl|Sequence
    (:then_0
      (ItemData|ClearItems)
      (Utilities|Array|Clear (Variables|Default|GetItemsToBeAddedQueue)))
    (:then_1
      (Utilities|Time|ClearandInvalidateTimerbyHandle (Variables|Default|GetAddItemsTimerHandle)))))

(fn GetItemListLength
  (return (Utilities|Array|Length (ItemData|GetContainedItems))))

(fn IsItemInList
  (params (ItemDefinition :type Fort%20World%20Item%20Definition%20Object%20Reference))
  (Utilities|FlowControl|Sequence
    (:then_0
      (Utilities|IsValid ItemDefinition
        (:Is%20Valid
          (LocalVariables|SetbSuccess (Utilities|Array|ContainsItem (ItemData|GetContainedItems) (Variables|GetItemDefinition))))))
    (:then_1
      (return (LocalVariables|GetbSuccess)))))

(fn IsItemInQueue
  (params (NewItem :type Actor%20Object%20Reference))
  (return (and (Utilities|IsValid NewItem :!class /Script/Engine.KismetSystemLibrary) (Utilities|Array|ContainsItem (Variables|Default|GetItemsToBeAddedQueue) NewItem))))

(fn TryOverrideInitialWeaponAmmunition
  (params (StructRef :type Minigame%20Spawner%20Spawn%20Params%20Structure%20%28by%20ref%29) (Override :type Boolean) (InitialWeaponAmmo :type Integer))
  (if Override
    (bind _structout (Utilities|Struct|SetmembersinMinigameSpawnerSpawnParams StructRef InitialWeaponAmmo))
    (return _structout)
    (else
      (return StructRef))))
